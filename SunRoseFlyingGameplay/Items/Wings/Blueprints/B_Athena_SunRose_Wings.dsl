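; Sun Rose wings item Blueprint that filters damage to the flying wings through ModifyIncomingDamage.
; It zeroes damage from friendly sources, then checks each wing that has a positive skim distance in the FlyingPawnComponent.
; To block hits from behind a skimmed wall, signed-distance dot products zero damage when the impact point and wing follicle socket share a side of the skim plane.
; All other damage passes through unchanged.
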(fn ModifyIncomingDamage
  (params (Damage :type Float%20%28single-precision%29) (DamageTags :type Gameplay%20Tag%20Container%20Structure%20%28by%20ref%29) (DamageHit :type Hit%20Result%20Structure%20%28by%20ref%29) (EventInstigator :type Controller%20Object%20Reference) (DamageCauser :type Actor%20Object%20Reference))
  (bind (_bblockinghit _binitialoverlap _time _distance _location _impactpoint) (Collision|BreakHitResult DamageHit))
  (bind _flyingpawncomponent (Variables|Default|GetFlyingPawnComponent))
  (bind _flyingpawncomponent_2 (Variables|Default|GetFlyingPawnComponent))
  (Utilities|FlowControl|Sequence
    (:then_0
      ;[1 1 1] Ignore damage from friendly ally
      (if (Utilities|Enum|Equal(Enum) (Team|GetTeamAffiliation (Game|GetInstigator) DamageCauser) "Friendly")
        (return 0.0)))
    (:then_1
      ;[1 1 1] Check if damaged from behind left wing's skimming wall
      (if (> (Variables|/SunRoseFlyingGameplay/Items/Wings/Blueprints/FlyingPawnComponent.FlyingPawnComponent_C|GetSkim_Distance_Left _flyingpawncomponent_2) 0)
        (if (>= (* (- (Math|Vector|DotProduct (Transformation|GetSocketLocation (Variables|Visual|GetWeaponMesh) "follicle_10_l") (Variables|/SunRoseFlyingGameplay/Items/Wings/Blueprints/FlyingPawnComponent.FlyingPawnComponent_C|GetSkim_Normal_Left _flyingpawncomponent_2)) (Math|Vector|DotProduct (Variables|/SunRoseFlyingGameplay/Items/Wings/Blueprints/FlyingPawnComponent.FlyingPawnComponent_C|GetSkim_Location_Left _flyingpawncomponent_2) (Variables|/SunRoseFlyingGameplay/Items/Wings/Blueprints/FlyingPawnComponent.FlyingPawnComponent_C|GetSkim_Normal_Left _flyingpawncomponent_2))) (- (Math|Vector|DotProduct _impactpoint (Variables|/SunRoseFlyingGameplay/Items/Wings/Blueprints/FlyingPawnComponent.FlyingPawnComponent_C|GetSkim_Normal_Left _flyingpawncomponent_2)) (Math|Vector|DotProduct (Variables|/SunRoseFlyingGameplay/Items/Wings/Blueprints/FlyingPawnComponent.FlyingPawnComponent_C|GetSkim_Location_Left _flyingpawncomponent_2) (Variables|/SunRoseFlyingGameplay/Items/Wings/Blueprints/FlyingPawnComponent.FlyingPawnComponent_C|GetSkim_Normal_Left _flyingpawncomponent_2)))) 0)
          (return 0.0))))
    (:then_2
      ;[1 1 1] Check if damaged from behind right wing's skimming wall
      (if (> (Variables|/SunRoseFlyingGameplay/Items/Wings/Blueprints/FlyingPawnComponent.FlyingPawnComponent_C|GetSkim_Distance_Right _flyingpawncomponent) 0)
        (if (>= (* (- (Math|Vector|DotProduct (Transformation|GetSocketLocation (Variables|Visual|GetWeaponMesh) "follicle_r_10") (Variables|/SunRoseFlyingGameplay/Items/Wings/Blueprints/FlyingPawnComponent.FlyingPawnComponent_C|GetSkim_Normal_Right _flyingpawncomponent)) (Math|Vector|DotProduct (Variables|/SunRoseFlyingGameplay/Items/Wings/Blueprints/FlyingPawnComponent.FlyingPawnComponent_C|GetSkim_Location_Right _flyingpawncomponent) (Variables|/SunRoseFlyingGameplay/Items/Wings/Blueprints/FlyingPawnComponent.FlyingPawnComponent_C|GetSkim_Normal_Right _flyingpawncomponent))) (- (Math|Vector|DotProduct _impactpoint (Variables|/SunRoseFlyingGameplay/Items/Wings/Blueprints/FlyingPawnComponent.FlyingPawnComponent_C|GetSkim_Normal_Right _flyingpawncomponent)) (Math|Vector|DotProduct (Variables|/SunRoseFlyingGameplay/Items/Wings/Blueprints/FlyingPawnComponent.FlyingPawnComponent_C|GetSkim_Location_Right _flyingpawncomponent) (Variables|/SunRoseFlyingGameplay/Items/Wings/Blueprints/FlyingPawnComponent.FlyingPawnComponent_C|GetSkim_Normal_Right _flyingpawncomponent)))) 0)
          (return 0.0))))
    (:then_3
      (return (Variables|GetDamage)))))

(fn ConstructionScript)
