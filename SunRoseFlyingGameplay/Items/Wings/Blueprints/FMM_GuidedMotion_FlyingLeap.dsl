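; Guided-motion movement mode extension for the SunRose wings' leap, which launches the pawn into flying.
; Activation spends weapon ammo as energy or sends an abort event, and applies cooldown effects.
; Updates end the leap on timeout, cancel tags, weapon change or unpaid energy; relevant collisions force-end it, tracing world and water below to block flight when too low.
; Deactivation sends the start-flying event if still allowed and applies any energy penalty.
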(event EventOnExtensionActivate
  (params (RuntimeContext :type MMERuntime%20Context%20Structure%20%28by%20ref%29))
  (bind _ownerpawn (Variables|FortMovementMode_BaseExtLogic|GetOwnerPawn))
  (bind _currentweapon (Variables|/Script/FortniteGame.FortPawn|GetCurrentWeapon _ownerpawn))
  (|CallParentFunction RuntimeContext)
  (Variables|Default|SetCanStartFlying true)
  (Variables|Default|SetApplyEnergyPenalty false)
  (Variables|Default|SetPaidEnergyToLeap false)
  (Utilities|FlowControl|Sequence
    (:then_0
      ;[1 1 1] Energy Consumption
      (if (ScalableFloat|GetValueasBool (Variables|Default|GetRow_CostLeapEnergyFromMME))
        (if (Utilities|Enum|NotEqual(Enum) (Networking|GetLocalRole _ownerpawn) "ROLE_SimulatedProxy")
          (if (<= (FortWeapon|GetLocalMagazineAmmoCount _currentweapon) 0)
            (Ability|SendGameplayEventtoActor (Variables|FortMovementMode_BaseExtLogic|GetOwnerPawn) (Variables|Default|GetTag_AbortLeap) (Utilities|Struct|MakeGameplayEventData (Variables|Default|GetTag_AbortLeap) (Variables|FortMovementMode_BaseExtLogic|GetOwnerPawn) (Variables|FortMovementMode_BaseExtLogic|GetOwnerPawn)))
            (else
              (FortWeapon|Ammo|ConsumeAmmo _currentweapon (Math|Float|Truncate (ScalableFloat|GetValueatLevel (Variables|Default|GetRow_LeapAmmoCount))) true) ; Use this to apply ammo cost since it applies even if the current ammo is below the cost threshold
              (Variables|Default|SetPaidEnergyToLeap true))))))
    (:then_1
      ;[1 1 1] Apply fixed duation cooldown
      (if (ScalableFloat|GetValueasBool (Variables|Default|GetRow_ApplyLeapCooldown))
        (bind _returnvalue (Ability|GameplayEffect|SetDuration (GameplayEffects|MakeOutgoingSpec :self (Ability|GetAbilitySystemComponent (Variables|FortMovementMode_BaseExtLogic|GetOwnerPawn)) :GameplayEffectClass (Variables|Default|GetGELeapCooldown) :Level 0.0) (ScalableFloat|GetValueatLevel (Variables|Default|GetRow_LeapStartCooldownDuration))))
        (GameplayEffects|ApplyGameplayEffectSpecToSelf (Ability|GetAbilitySystemComponent (Variables|FortMovementMode_BaseExtLogic|GetOwnerPawn)) _returnvalue)))
    (:then_2
      ;[1 1 1] Apply till land or till leap end cooldown
      (GameplayEffects|ApplyGameplayEffectSpecToSelf (Ability|GetAbilitySystemComponent (Variables|FortMovementMode_BaseExtLogic|GetOwnerPawn)) (GameplayEffects|MakeOutgoingSpec :self (Ability|GetAbilitySystemComponent (Variables|FortMovementMode_BaseExtLogic|GetOwnerPawn)) :GameplayEffectClass "/SunRoseFlyingGameplay/Items/Wings/Blueprints/GE_Athena_SunRose_Wings_Cooldown_Leaping.GE_Athena_SunRose_Wings_Cooldown_Leaping_C" :Level 0.0)))))

(event EventOnExtensionDeactivate
  (params (RuntimeContext :type MMERuntime%20Context%20Structure%20%28by%20ref%29))
  (bind _row_energypenalty (Variables|Default|GetRow_EnergyPenalty))
  (bind _returnvalue (ScalableFloat|GetValueatLevel _row_energypenalty))
  (bind _returnvalue_2 (Math|Float|Truncate _returnvalue))
  (|CallParentFunction RuntimeContext)
  (Utilities|FlowControl|Sequence
    (:then_0
      (if (Variables|Default|GetCanStartFlying)
        (Ability|SendGameplayEventtoActor (Variables|FortMovementMode_BaseExtLogic|GetOwnerPawn) (Variables|Default|GetTag_StartFlying) (Utilities|Struct|MakeGameplayEventData (Variables|Default|GetTag_StartFlying) (Variables|FortMovementMode_BaseExtLogic|GetOwnerPawn) (Variables|FortMovementMode_BaseExtLogic|GetOwnerPawn)))))
    (:then_1
      (if (and (Variables|Default|GetApplyEnergyPenalty) (>= _returnvalue_2 0))
        (FortWeapon|Ammo|ConsumeAmmo (Variables|/Script/FortniteGame.FortPawn|GetCurrentWeapon (Variables|FortMovementMode_BaseExtLogic|GetOwnerPawn)) _returnvalue_2 true) ; Use this to apply ammo cost since it applies even if the current ammo is below the cost threshold
      ))
    (:then_2
      (Variables|Default|SetPaidEnergyToLeap false))))

(fn HandleHitResponse
  (params (RuntimeContext :type MMERuntime%20Context%20Structure%20%28by%20ref%29) (Hit :type Hit%20Result%20Structure%20%28by%20ref%29) (InVelocity :type Vector%20%28by%20ref%29 :default "0, 0, 0") (OldPosition :type Vector%20%28by%20ref%29 :default "0, 0, 0") (Adjustment :type Vector%20%28by%20ref%29 :default "0, 0, 0"))
  (bind (_bblockinghit _binitialoverlap _time _distance _location _impactpoint _normal _impactnormal _physmat _hitactor _hitcomponent) (Collision|BreakHitResult Hit))
  (bind _returnvalue (Class|FortMovementModeExtBlueprintLibrary|GetRuntimeDatafromContext RuntimeContext "/Script/FortniteGame.FortMovementMode_GuidedMotionRuntimeData"))
  (Utilities|IsValid _returnvalue
    (:Is%20Valid
      (bind _returnvalue_2 (CallFunction|IsHitRelevant Hit InVelocity OldPosition Adjustment))
      (if _returnvalue_2
        (if _bblockinghit
          (switch Utilities|FlowControl|Switch|SwitchonECollisionChannel (Collision|GetCollisionObjectType _hitcomponent)
            (:ECC_WorldStatic
              _)
            (:ECC_WorldDynamic
              (bind _returnvalue_3 (GameplayTags|HasMatchingGameplayTag(Message) _hitactor "(TagName=\"Athena.Gameplay.Train.Car\")"))
              (if _returnvalue_3
                _))
            (:ECC_Pawn
              _)
            (:ECC_PhysicsBody
              _)
            (:ECC_Vehicle
              _)
            (:ECC_Destructible
              _))
          (Utilities|FlowControl|Sequence
            (:then_0
              (Development|PrintString (Variables|FortMovementMode_BaseExtLogic|GetOwnerPawn) "FlyingLeap MME going to end because of collision")
              (bind (_bisblockinghit _closesthit) (Class|FortWingedFlyingBlueprintLibrary|TraceforWorldandWater (Variables|FortMovementMode_BaseExtLogic|GetOwnerPawn) (Utilities|Operators|Subtract _location (Math|Vector|MakeVector 0.0 0.0 (Components|Capsule|GetScaledCapsuleHalfHeight (Variables|/Script/Engine.Character|GetCapsuleComponent (Variables|FortMovementMode_BaseExtLogic|GetOwnerPawn))))) (Utilities|Operators|Subtract (Utilities|Operators|Subtract _location (Math|Vector|MakeVector 0.0 0.0 (Components|Capsule|GetScaledCapsuleHalfHeight (Variables|/Script/Engine.Character|GetCapsuleComponent (Variables|FortMovementMode_BaseExtLogic|GetOwnerPawn))))) (Math|Vector|MakeVector 0.0 0.0 (ScalableFloat|GetValueatLevel (Variables|Default|GetRow_MinLeapHeightToStartFlying))))))
              (if (and _bisblockinghit (Collision|BreakHitResult _closesthit))
                (Variables|Default|SetCanStartFlying false)
                (Ability|SendGameplayEventtoActor (Variables|FortMovementMode_BaseExtLogic|GetOwnerPawn) (Variables|Default|GetTag_AbortLeap) (Utilities|Struct|MakeGameplayEventData (Variables|Default|GetTag_AbortLeap) (Variables|FortMovementMode_BaseExtLogic|GetOwnerPawn) (Variables|FortMovementMode_BaseExtLogic|GetOwnerPawn)))))
            (:then_1
              (Class|FortMovementModeBaseExtRuntimeData|ForceManualEnd _returnvalue)
              (return))))))))

(fn SafeMoveUpdatedComponentHitResponse
  (params (RuntimeContext :type MMERuntime%20Context%20Structure%20%28by%20ref%29) (InVelocity :type Vector%20%28by%20ref%29 :default "0, 0, 0") (OldPosition :type Vector%20%28by%20ref%29 :default "0, 0, 0") (Adjustment :type Vector%20%28by%20ref%29 :default "0, 0, 0") (Hit :type Hit%20Result%20Structure%20%28by%20ref%29))
  (bind (_bblockinghit _binitialoverlap _time _distance _location _impactpoint _normal _impactnormal _physmat _hitactor _hitcomponent) (Collision|BreakHitResult Hit))
  (bind _returnvalue (Class|FortMovementModeExtBlueprintLibrary|GetRuntimeDatafromContext RuntimeContext "/Script/FortniteGame.FortMovementMode_GuidedMotionRuntimeData"))
  (Utilities|IsValid _returnvalue
    (:Is%20Valid
      (bind _returnvalue_2 (CallFunction|IsHitRelevant Hit InVelocity OldPosition Adjustment))
      (if _returnvalue_2
        (if _bblockinghit
          (switch Utilities|FlowControl|Switch|SwitchonECollisionChannel (Collision|GetCollisionObjectType _hitcomponent)
            (:ECC_WorldStatic
              _)
            (:ECC_WorldDynamic
              (bind _returnvalue_3 (GameplayTags|HasMatchingGameplayTag(Message) _hitactor "(TagName=\"Athena.Gameplay.Train.Car\")"))
              (if _returnvalue_3
                _))
            (:ECC_Pawn
              _)
            (:ECC_PhysicsBody
              _)
            (:ECC_Vehicle
              _)
            (:ECC_Destructible
              _))
          (Utilities|FlowControl|Sequence
            (:then_0
              (Development|PrintString (Variables|FortMovementMode_BaseExtLogic|GetOwnerPawn) "FlyingLeap MME going to end because of collision")
              (bind (_bisblockinghit _closesthit) (Class|FortWingedFlyingBlueprintLibrary|TraceforWorldandWater (Variables|FortMovementMode_BaseExtLogic|GetOwnerPawn) (Utilities|Operators|Subtract _location (Math|Vector|MakeVector 0.0 0.0 (Components|Capsule|GetScaledCapsuleHalfHeight (Variables|/Script/Engine.Character|GetCapsuleComponent (Variables|FortMovementMode_BaseExtLogic|GetOwnerPawn))))) (Utilities|Operators|Subtract (Utilities|Operators|Subtract _location (Math|Vector|MakeVector 0.0 0.0 (Components|Capsule|GetScaledCapsuleHalfHeight (Variables|/Script/Engine.Character|GetCapsuleComponent (Variables|FortMovementMode_BaseExtLogic|GetOwnerPawn))))) (Math|Vector|MakeVector 0.0 0.0 (ScalableFloat|GetValueatLevel (Variables|Default|GetRow_MinLeapHeightToStartFlying))))))
              (if (and _bisblockinghit (Collision|BreakHitResult _closesthit))
                (Variables|Default|SetCanStartFlying false)
                (Ability|SendGameplayEventtoActor (Variables|FortMovementMode_BaseExtLogic|GetOwnerPawn) (Variables|Default|GetTag_AbortLeap) (Utilities|Struct|MakeGameplayEventData (Variables|Default|GetTag_AbortLeap) (Variables|FortMovementMode_BaseExtLogic|GetOwnerPawn) (Variables|FortMovementMode_BaseExtLogic|GetOwnerPawn)))))
            (:then_1
              (Class|FortMovementModeBaseExtRuntimeData|ForceManualEnd _returnvalue)
              (return))))))))

(fn UpdateBeforeCharacterMovement
  (params (NativeResult :type EFortMovementModeExt_UpdateResult%20Enum :default "Updating") (RuntimeContext :type MMERuntime%20Context%20Structure%20%28by%20ref%29))
  (bind _returnvalue (Class|FortMovementModeExtBlueprintLibrary|GetRuntimeDatafromContext RuntimeContext "/Script/FortniteGame.FortMovementMode_GuidedMotionRuntimeData"))
  (if (or (not (Utilities|IsValid _returnvalue :!class /Script/Engine.KismetSystemLibrary)) (>= (Variables|/Script/FortniteGame.FortMovementMode_GuidedMotionRuntimeData|GetLocalEndElapsedTime _returnvalue) (ScalableFloat|GetValueatLevel (Variables|Default|GetRow_TimeBeforeStartingFlight))))
    (return "Ended")
    (elif (and (Utilities|Enum|Equal(Enum) NativeResult "Active_Layered") (not (GameplayTags|HasAnyMatchingGameplayTags (Variables|FortMovementMode_BaseExtLogic|GetOwnerPawn) (Variables|Default|GetTagsToCancelFlight))))
      ;[1 1 1] Check if started leaping with 0 energy
      (bind _asb_athena_sun_rose_wings (Utilities|Casting|CastToB_Athena_SunRose_Wings (Variables|/Script/FortniteGame.FortPawn|GetCurrentWeapon (Variables|FortMovementMode_BaseExtLogic|GetOwnerPawn)))
        (:then
          (if (not (Math|Boolean|ANDBoolean (ScalableFloat|GetValueasBool (Variables|Default|GetRow_ForceEndLeapForNoEnergy)) (Networking|HasAuthority (Variables|FortMovementMode_BaseExtLogic|GetOwnerPawn)) (not (Variables|Default|GetPaidEnergyToLeap)) (<= (FortWeapon|GetLocalMagazineAmmoCount _asb_athena_sun_rose_wings) 0)))
            (return "Active_Layered")))
        (:CastFailed
          (Variables|Default|SetApplyEnergyPenalty true)))))
  (Variables|Default|SetCanStartFlying false)
  (return "Ended"))
