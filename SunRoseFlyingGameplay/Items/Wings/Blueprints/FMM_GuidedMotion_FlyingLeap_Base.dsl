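; Base guided-motion movement mode extension for a wings flying leap, applying a constant force for a scalable duration.
; On activation it caches a launch velocity, whose direction and length drive that force, and prints debug messages on start and end.
; Launch velocity sums retained velocity and an acceleration- or facing-based impulse, with vertical sign set by leap type, times a predictive multiplier.
; A networked optional float can scale horizontal and vertical magnitudes as charge.
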(event EventOnExtensionActivate
  (params (RuntimeContext :type MMERuntime%20Context%20Structure%20%28by%20ref%29))
  (Development|PrintString (Variables|FortMovementMode_BaseExtLogic|GetOwnerPawn) "Player just activated FMM_GuidedMotion_Leap" false)
  (bind _returnvalue (Class|FortMovementModeExtBlueprintLibrary|GetRuntimeDatafromContext RuntimeContext "/Script/FortniteGame.FortMovementMode_GuidedMotionRuntimeData"))
  (Variables|Initialization|RuntimeVars|SetCalculatedForce (Initialization|GetLaunchVelocity _returnvalue)))

(event EventOnExtensionDeactivate
  (params (RuntimeContext :type MMERuntime%20Context%20Structure%20%28by%20ref%29))
  (Development|PrintString (Variables|FortMovementMode_BaseExtLogic|GetOwnerPawn) "player just ended FMM_GuidedMotion_Leap"))

(fn GetAccelerationValue
  (return (Math|Vector|SelectVector (Math|Vector|Normalize (Pawn|Components|CharacterMovement|GetCurrentAcceleration (Variables|FortMovementMode_BaseExtLogic|GetOwnerMovementComponent))) (Transformation|GetActorForwardVector (Variables|FortMovementMode_BaseExtLogic|GetOwnerPawn)) (and (Variables|PowerLeap|UpwardOrDownward|GetZeroAccelerationAssumesForward) (Math|Vector|VectorIsNearlyZero (Math|Vector|Normalize (Pawn|Components|CharacterMovement|GetCurrentAcceleration (Variables|FortMovementMode_BaseExtLogic|GetOwnerMovementComponent))))))))

(fn GetDirection_ConstantForce
  (params (GuidedMotionRuntimeData :type Fort%20Movement%20Mode%20Guided%20Motion%20Runtime%20Data%20Object%20Reference))
  (return (Math|Vector|Normalize (Variables|Initialization|RuntimeVars|GetCalculatedForce))))

(fn GetLaunchVelocity
  (params (Runtime%20Data :type Fort%20Movement%20Mode%20Guided%20Motion%20Runtime%20Data%20Object%20Reference))
  (bind (_returnvalue_x _returnvalue_y) (Math|Vector|BreakVector (Math|Vector|SelectVector (Initialization|GetAccelerationValue) (Transformation|GetActorForwardVector (Variables|FortMovementMode_BaseExtLogic|GetOwnerPawn)) (Variables|PowerLeap|UpwardOrDownward|GetBaseLaunchDirectionOnAcceleration))))
  (return (* (+ (Initialization|GetVelocityPercent) (Math|Vector|MakeVector (* _returnvalue_x (Initialization|ModifyLaunchMagwithCharge Runtime%20Data (ScalableFloat|GetValueatLevel (Variables|PowerLeap|UpwardorDownward|GetRow_HorizontalLaunchMagnitude)) (Variables|PowerLeap|UpwardorDownward|GetHorizontalLaunchMagnitudeUsesCharge))) (* _returnvalue_y (Initialization|ModifyLaunchMagwithCharge Runtime%20Data (ScalableFloat|GetValueatLevel (Variables|PowerLeap|UpwardorDownward|GetRow_HorizontalLaunchMagnitude)) (Variables|PowerLeap|UpwardorDownward|GetHorizontalLaunchMagnitudeUsesCharge))) (Utilities|Select :Index (Variables|PowerLeap|GetLeapType) :NewEnumerator0 0.0 :NewEnumerator3 0.0 :NewEnumerator2 (Initialization|ModifyLaunchMagwithCharge Runtime%20Data (ScalableFloat|GetValueatLevel (Variables|PowerLeap|UpwardorDownward|GetRow_VerticalLaunchMagnitude)) (Variables|PowerLeap|UpwardOrDownward|GetVerticalLaunchMagnitudeUsesCharge)) :NewEnumerator1 (* (Initialization|ModifyLaunchMagwithCharge Runtime%20Data (ScalableFloat|GetValueatLevel (Variables|PowerLeap|UpwardorDownward|GetRow_VerticalLaunchMagnitude)) (Variables|PowerLeap|UpwardOrDownward|GetVerticalLaunchMagnitudeUsesCharge)) -1.0)))) (Variables|PowerLeap|PredictiveLaunchVelocity|RuntimeVars|GetPredictiveLaunchVelocityMult))))

(fn GetMaxDuration
  (params (GuidedMotionRuntimeData :type Fort%20Movement%20Mode%20Guided%20Motion%20Runtime%20Data%20Object%20Reference))
  (return (ScalableFloat|GetValueatLevel (Variables|PowerLeap|GetRow_LeapDuration))))

(fn GetStrength_ConstantForce
  (params (GuidedMotionRuntimeData :type Fort%20Movement%20Mode%20Guided%20Motion%20Runtime%20Data%20Object%20Reference))
  (return (Math|Vector|VectorLength (Variables|Initialization|RuntimeVars|GetCalculatedForce))))

(fn GetVelocityPercent
  (bind (_velocity_x _velocity_y _velocity_z) (Math|Vector|BreakVector (Variables|/Script/Engine.MovementComponent|GetVelocity_X (Variables|FortMovementMode_BaseExtLogic|GetOwnerMovementComponent))))
  (return (Math|Vector|MakeVector (* _velocity_x (ScalableFloat|GetValueatLevel (Variables|PowerLeap|UpwardOrDownward|GetRow_HorizontalVelocityPercentMaintained))) (* _velocity_y (ScalableFloat|GetValueatLevel (Variables|PowerLeap|UpwardOrDownward|GetRow_HorizontalVelocityPercentMaintained))) (select (or (and (Utilities|Select :Index (Variables|PowerLeap|GetLeapType) :NewEnumerator0 false :NewEnumerator3 false :NewEnumerator2 (> (* _velocity_z (ScalableFloat|GetValueatLevel (Variables|PowerLeap|UpwardOrDownward|GetRow_VertitcalVelocityPercentMaintained))) 0.0) :NewEnumerator1 (< (* _velocity_z (ScalableFloat|GetValueatLevel (Variables|PowerLeap|UpwardOrDownward|GetRow_VertitcalVelocityPercentMaintained))) 0.0)) (Variables|PowerLeap|UpwardOrDownward|GetVerticalVelocityMustMatchImpulse)) (not (Variables|PowerLeap|UpwardOrDownward|GetVerticalVelocityMustMatchImpulse))) (* _velocity_z (ScalableFloat|GetValueatLevel (Variables|PowerLeap|UpwardOrDownward|GetRow_VertitcalVelocityPercentMaintained))) 0.0))))

(fn ModifyLaunchMagWithCharge
  (params (Runtime%20Data :type Fort%20Movement%20Mode%20Guided%20Motion%20Runtime%20Data%20Object%20Reference) (Mag :type Float%20%28double-precision%29) (Modify :type Boolean))
  (bind _netoptionalfloats (Variables|/Script/FortniteGame.FortMovementMode_GuidedMotionRuntimeData|GetNetOptionalFloats Runtime%20Data))
  (bind _output_value (Utilities|Struct|BreakFortMovementModeFloatPropertyNetworked (Utilities|Array|Get(aref) _netoptionalfloats)))
  (if (Utilities|Array|IsValidIndex _netoptionalfloats)
    (return (Math|Float|SelectFloat (* Mag _output_value) Mag Modify))
    (else
      (return Mag))))
