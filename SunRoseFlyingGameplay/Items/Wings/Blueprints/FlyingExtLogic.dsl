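; Movement mode extension logic for SunRose winged flight, handling boost, ascend, crash and left/right dodge.
; On activation it toggles the flying input context around inventory and map use, turns FlyingPawnComponent input events into intent flags, and ends flight when the required tag is removed.
; Start, continue and end checks let crashing override other statuses and apply balance-driven durations.
; Status changes, ignored during resimulation, send gameplay events or cancel tagged abilities.
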(event Custom|ResetMoveInputValues
  (params (bResetCrashInput :type Boolean :default true))
  ;[1 1 1] Reset Move Input Values
  (Class|FlyingPawnComponent|ResetInputStatus (Variables|Default|GetFlyingPawnComponent) bResetCrashInput))

(event EventOnExtensionActivate
  (params (RuntimeContext :type MMERuntime%20Context%20Structure%20%28by%20ref%29))
  (bind _cachedplayercontroller (Variables|Default|GetCachedPlayerController))
  (bind _returnvalue (Actor|GetComponentbyClass _cachedplayercontroller "/Script/FortniteGame.FortControllerComponent_Minimap"))
  (bind _ownerpawn (Variables|FortMovementMode_BaseExtLogic|GetOwnerPawn))
  (Development|PrintString (Variables|FortMovementMode_BaseExtLogic|GetOwnerPawn) (Utilities|String|Append (Utilities|GetDisplayName (Variables|FortMovementMode_BaseExtLogic|GetOwnerPawn)) " activated flying movement") true true "(R=1.000000,G=0.000000,B=0.825650,A=1.000000)" 5.0)
  (Utilities|FlowControl|Sequence
    (:then_0
      (Variables|Default|SetTotalUnpaidCeilingHeightAdded 0.0)
      (Variables|Default|SetCeilingHeightToAddThisTick 0.0)
      (Variables|Default|SetAscendHeightPer1Cost (ScalableFloat|GetValueatLevel (Variables|Balance|EnergyCosts|GetRow_HeightIncreasePer1Cost)))
      (Variables|Default|SetPotentialCeilingHeightDecayThisTickFromFlying 0.0)
      (Variables|Default|SetActualCeilingHeightDecayThisTick 0.0)
      (Variables|Default|SetForwardSpeed (ScalableFloat|GetValueatLevel (Variables|Balance|Flying|GetRow_FlyingSpeed_Base)))
      (bind _assun_rose_player_anim_bp (Utilities|Casting|CastToSunRose_Player_AnimBP (Animation|LinkedAnimGraphs|GetLinkedAnimLayerInstancebyClass (Components|SkeletalMesh|GetAnimInstance (Variables|/Script/Engine.Character|GetMesh (Variables|FortMovementMode_BaseExtLogic|GetOwnerPawn))) "/SunRoseFlyingGameplay/Items/Wings/Blueprints/SunRose_Player_AnimBP.SunRose_Player_AnimBP_C" false)))
      (Variables|Default|SetFlyingPlayerAnim _assun_rose_player_anim_bp)
      (Variables|Default|SetFlyingPawnComponent (Actor|GetComponentbyClass (Variables|FortMovementMode_BaseExtLogic|GetOwnerPawn) "/SunRoseFlyingGameplay/Items/Wings/Blueprints/FlyingPawnComponent.FlyingPawnComponent_C")))
    (:then_1
      (if (Utilities|Enum|Equal(Enum) (Networking|GetLocalRole _ownerpawn) "ROLE_AutonomousProxy")
        (bind _asplayer_controller (Utilities|Casting|CastToPlayerController (Pawn|GetController _ownerpawn)))
        (Variables|Default|SetCachedPlayerController _asplayer_controller)
        (CallFunction|ResetMoveInputValues false) ; Keep crash input to allow crash from falling mode
        (CallFunction|ToggleInputMappingContext true)))
    (:then_2
      (bind _asfort_ability_system_component (Utilities|Casting|CastToFortAbilitySystemComponent (Ability|GetAbilitySystemComponent (Variables|FortMovementMode_BaseExtLogic|GetOwnerPawn))))
      (Variables|Default|SetOwnerASC _asfort_ability_system_component))
    (:then_3
      (Ability|SendGameplayEventtoActor (Variables|FortMovementMode_BaseExtLogic|GetOwnerPawn) "(TagName=\"Athena.ContextualTutorial.SunRose.Wings.FlyForward.Start\")")
      (CallFunction|BindInputActions))
    (:then_4
      (bind _asynctaskproxy (Ability|Tasks|WaitGameplayTagRemoveFromActor (Variables|FortMovementMode_BaseExtLogic|GetOwnerPawn) (Variables|Default|GetTag_RequiredToContinueFlying))
        (:then
          (Variables|Default|SetTask_WaitRemoveTagRequiredToContinueFlying _asynctaskproxy))
        (:Removed
          (Class|FortPawnComponentWingedFlying|EndFlying (Variables|Default|GetFlyingPawnComponent)))))
    (:then_5
      (bind _asfort_player_controller (Utilities|Casting|CastToFortPlayerController _cachedplayercontroller))
      (EventDispatchers|BindEventtoOnInventoryActiveChangedEvent _asfort_player_controller (EventDispatchers|CreateEvent :!event InputConditionChanged))
      (Utilities|IsValid _returnvalue
        (:Is%20Valid
          (PlayerController|BindEventtoOnToggleFullscreenMap _returnvalue (EventDispatchers|CreateEvent :!event InputConditionChanged)))))))

(event Custom|Bind%20Input%20Actions
  (bind _flyingpawncomponent (Variables|Default|GetFlyingPawnComponent))
  ;[1 1 1] Inputs
  ;;[1 1 1] Temporary, to be replaced with input action events when MME's can accesst them
  (Default|BindEventtoBoostActionStarted _flyingpawncomponent (delegate Custom|BoostAction_Started))
  (Default|BindEventtoBoostActionEnded _flyingpawncomponent (delegate Custom|BoostAction_Ended))
  (Default|BindEventtoAscendActionStarted _flyingpawncomponent (delegate Custom|AscendAction_Started))
  (Default|BindEventtoAscendActionEnded _flyingpawncomponent (delegate Custom|AscendAction_Ended))
  (Default|BindEventtoCrashActionTriggered _flyingpawncomponent (delegate Custom|CrashAction_Triggered))
  (Default|BindEventtoRollLeftActionTriggered _flyingpawncomponent (delegate Custom|RollLeftAction_Triggered))
  (Default|BindEventtoRollRightActionTriggered _flyingpawncomponent (delegate Custom|RollRightAction_Triggered)))

(event Custom|Unbind%20Input%20Actions
  (bind _flyingpawncomponent (Variables|Default|GetFlyingPawnComponent))
  ;[1 1 1] Inputs
  ;;[1 1 1] Temporary, to be replaced with input action events when MME's can accesst them
  (Default|UnbindEventfromBoostActionStarted _flyingpawncomponent (delegate Custom|BoostAction_Started))
  (Default|UnbindEventfromBoostActionEnded _flyingpawncomponent (delegate Custom|BoostAction_Ended))
  (Default|UnbindEventfromAscendActionStarted _flyingpawncomponent (delegate Custom|AscendAction_Started))
  (Default|UnbindEventfromAscendActionEnded _flyingpawncomponent (delegate Custom|AscendAction_Ended))
  (Default|UnbindEventfromCrashActionTriggered _flyingpawncomponent (delegate Custom|CrashAction_Triggered))
  (Default|UnbindEventfromRollLeftActionTriggered _flyingpawncomponent (delegate Custom|RollLeftAction_Triggered))
  (Default|UnbindEventfromRollRightActionTriggered _flyingpawncomponent (delegate Custom|RollRightAction_Triggered)))

(event EventOnExtensionDeactivate
  (params (RuntimeContext :type MMERuntime%20Context%20Structure%20%28by%20ref%29))
  (bind _cachedplayercontroller (Variables|Default|GetCachedPlayerController))
  (bind _returnvalue (Actor|GetComponentbyClass _cachedplayercontroller "/Script/FortniteGame.FortControllerComponent_Minimap"))
  (Development|PrintString (Variables|FortMovementMode_BaseExtLogic|GetOwnerPawn) (Utilities|String|Append (Utilities|GetDisplayName (Variables|FortMovementMode_BaseExtLogic|GetOwnerPawn)) " stopped flying movement"))
  (Utilities|FlowControl|Sequence
    (:then_0
      (GameplayCue|RemoveGameplayCueOnActor(Looping) (Variables|FortMovementMode_BaseExtLogic|GetOwnerPawn) (Variables|GameplayCues|GetGC_CrashEnabled) (Ability|GameplayCue|MakeGameplayCueParameters 0.0))
      (Variables|Default|SetCrashEnabled false)
      (GameplayCue|RemoveGameplayCueOnActor(Looping) (Variables|FortMovementMode_BaseExtLogic|GetOwnerPawn) (Variables|GameplayCues|GetGC_EnergyDrain) (Ability|GameplayCue|MakeGameplayCueParameters 0.0))
      (Variables|Default|SetEnergyDrainEnabled false)
      (CallFunction|UnbindInputActions))
    (:then_1
      (CallFunction|ToggleInputMappingContext false)
      (bind _asfort_player_controller (Utilities|Casting|CastToFortPlayerController _cachedplayercontroller))
      (EventDispatchers|UnbindEventfromOnInventoryActiveChangedEvent _asfort_player_controller (EventDispatchers|CreateEvent :!event InputConditionChanged))
      (Utilities|IsValid _returnvalue
        (:Is%20Valid
          (PlayerController|UnbindEventfromOnToggleFullscreenMap _returnvalue (EventDispatchers|CreateEvent :!event InputConditionChanged)))))
    (:then_2
      (CallFunction|ResetMoveInputValues true))
    (:then_3
      (Ability|Async|EndAction (Variables|Default|GetTask_WaitRemoveTagRequiredToContinueFlying)))))

(event Custom|BoostAction_Started
  (bind _flyingpawncomponent (Variables|Default|GetFlyingPawnComponent))
  ;[1 1 1] Inputs
  (Variables|/Script/SunRoseFlyingRuntime.FortPawnComponent_WingedFlying|SetbIntendsToBoost true _flyingpawncomponent)
  (Variables|/SunRoseFlyingGameplay/Items/Wings/Blueprints/FlyingPawnComponent.FlyingPawnComponent_C|SetBoostPressed true _flyingpawncomponent))

(event Custom|BoostAction_Ended
  (bind _flyingpawncomponent (Variables|Default|GetFlyingPawnComponent))
  (bind _flyingpawncomponent_2 (Variables|Default|GetFlyingPawnComponent))
  ;[1 1 1] Inputs
  (Variables|/Script/SunRoseFlyingRuntime.FortPawnComponent_WingedFlying|SetbIntendsToBoost false _flyingpawncomponent)
  (Variables|/SunRoseFlyingGameplay/Items/Wings/Blueprints/FlyingPawnComponent.FlyingPawnComponent_C|SetBoostPressed false _flyingpawncomponent)
  (if (Variables|/SunRoseFlyingGameplay/Items/Wings/Blueprints/FlyingPawnComponent.FlyingPawnComponent_C|GetAscendPressed (Variables|Default|GetFlyingPawnComponent))
    (Variables|/Script/SunRoseFlyingRuntime.FortPawnComponent_WingedFlying|SetbIntendsToAscend true _flyingpawncomponent_2)
    (Variables|/SunRoseFlyingGameplay/Items/Wings/Blueprints/FlyingPawnComponent.FlyingPawnComponent_C|SetAscendPressed true _flyingpawncomponent_2)))

(event Custom|AscendAction_Started
  (bind _flyingpawncomponent (Variables|Default|GetFlyingPawnComponent))
  ;[1 1 1] Inputs
  (Variables|/Script/SunRoseFlyingRuntime.FortPawnComponent_WingedFlying|SetbIntendsToAscend true _flyingpawncomponent)
  (Variables|/SunRoseFlyingGameplay/Items/Wings/Blueprints/FlyingPawnComponent.FlyingPawnComponent_C|SetAscendPressed true _flyingpawncomponent))

(event Custom|AscendAction_Ended
  (bind _flyingpawncomponent (Variables|Default|GetFlyingPawnComponent))
  (bind _flyingpawncomponent_2 (Variables|Default|GetFlyingPawnComponent))
  ;[1 1 1] Inputs
  (Variables|/Script/SunRoseFlyingRuntime.FortPawnComponent_WingedFlying|SetbIntendsToAscend false _flyingpawncomponent_2)
  (Variables|/SunRoseFlyingGameplay/Items/Wings/Blueprints/FlyingPawnComponent.FlyingPawnComponent_C|SetAscendPressed false _flyingpawncomponent_2)
  (if (Variables|/SunRoseFlyingGameplay/Items/Wings/Blueprints/FlyingPawnComponent.FlyingPawnComponent_C|GetBoostPressed (Variables|Default|GetFlyingPawnComponent))
    (Variables|/Script/SunRoseFlyingRuntime.FortPawnComponent_WingedFlying|SetbIntendsToBoost true _flyingpawncomponent)
    (Variables|/SunRoseFlyingGameplay/Items/Wings/Blueprints/FlyingPawnComponent.FlyingPawnComponent_C|SetBoostPressed true _flyingpawncomponent)))

(event Custom|CrashAction_Triggered
  ;[1 1 1] Inputs
  (Variables|/Script/SunRoseFlyingRuntime.FortPawnComponent_WingedFlying|SetbIntendsToCrash true (Variables|Default|GetFlyingPawnComponent)))

(event Custom|RollLeftAction_Triggered
  (bind _flyingpawncomponent (Variables|Default|GetFlyingPawnComponent))
  ;[1 1 1] Inputs
  (Variables|/Script/SunRoseFlyingRuntime.FortPawnComponent_WingedFlying|SetbIntendsToRollRight false _flyingpawncomponent)
  (Variables|/Script/SunRoseFlyingRuntime.FortPawnComponent_WingedFlying|SetbIntendsToRollLeft true _flyingpawncomponent))

(event Custom|RollRightAction_Triggered
  (bind _flyingpawncomponent (Variables|Default|GetFlyingPawnComponent))
  ;[1 1 1] Inputs
  (Variables|/Script/SunRoseFlyingRuntime.FortPawnComponent_WingedFlying|SetbIntendsToRollLeft false _flyingpawncomponent)
  (Variables|/Script/SunRoseFlyingRuntime.FortPawnComponent_WingedFlying|SetbIntendsToRollRight true _flyingpawncomponent))

(event EventHandleFlyingStatusChanged
  (params (MoveStatus :type Fort%20Movement%20Mode%20Winged%20Flying%20Runtime%20Data%20Object%20Reference) (NewStatus :type EWingedFlyingStatus%20Enum :default "Default") (bIsActive :type Boolean) (FailureTag :type Gameplay%20Tag%20Structure%20%28by%20ref%29))
  (if (not (Movement|IsResimulatingMove MoveStatus))
    (switch Utilities|FlowControl|Switch|SwitchonEWingedFlyingStatus NewStatus
      (:Boosting
        ;[1 1 1] Boosting status change
        (if bIsActive
          (Ability|SendGameplayEventtoActor (Variables|FortMovementMode_BaseExtLogic|GetOwnerPawn) "(TagName=\"Granted.Athena.SunRose.Wings.Boosting\")" (Utilities|Struct|MakeGameplayEventData "(TagName=\"Granted.Athena.SunRose.Wings.Boosting\")" (Variables|FortMovementMode_BaseExtLogic|GetOwnerPawn) (Variables|FortMovementMode_BaseExtLogic|GetOwnerPawn)))
          (else
            (Tags|CancelAbilitiesWithTags (Variables|Default|GetOwnerASC) (GameplayTags|MakeGameplayTagContainerfromTag "(TagName=\"Granted.Athena.SunRose.Wings.Boosting\")")))))
      (:Ascending
        ;[1 1 1] Ascending status change
        (if bIsActive
          (Ability|SendGameplayEventtoActor (Variables|FortMovementMode_BaseExtLogic|GetOwnerPawn) "(TagName=\"Granted.Athena.SunRose.Wings.Ascending\")" (Utilities|Struct|MakeGameplayEventData "(TagName=\"Granted.Athena.SunRose.Wings.Ascending\")" (Variables|FortMovementMode_BaseExtLogic|GetOwnerPawn) (Variables|FortMovementMode_BaseExtLogic|GetOwnerPawn)))
          (else
            (Tags|CancelAbilitiesWithTags (Variables|Default|GetOwnerASC) (GameplayTags|MakeGameplayTagContainerfromTag "(TagName=\"Granted.Athena.SunRose.Wings.Ascending\")")))))
      (:Crashing
        ;[1 1 1] Crashing Status Change
        (if bIsActive
          (Ability|SendGameplayEventtoActor (Variables|FortMovementMode_BaseExtLogic|GetOwnerPawn) "(TagName=\"Granted.Athena.SunRose.Wings.Crashing\")" (Utilities|Struct|MakeGameplayEventData "(TagName=\"Granted.Athena.SunRose.Wings.Crashing\")" (Variables|FortMovementMode_BaseExtLogic|GetOwnerPawn) (Variables|FortMovementMode_BaseExtLogic|GetOwnerPawn)))
          (else
            (Tags|CancelAbilitiesWithTags (Variables|Default|GetOwnerASC) (GameplayTags|MakeGameplayTagContainerfromTag "(TagName=\"Granted.Athena.SunRose.Wings.Crashing\")")))))
      (:DodgingLeft
        ;[1 1 1] Dodging Left Status Change
        (if bIsActive
          (Ability|SendGameplayEventtoActor (Variables|FortMovementMode_BaseExtLogic|GetOwnerPawn) "(TagName=\"Granted.Athena.SunRose.Wings.Dodging.Left\")" (Utilities|Struct|MakeGameplayEventData "(TagName=\"Granted.Athena.SunRose.Wings.Dodging.Left\")" (Variables|FortMovementMode_BaseExtLogic|GetOwnerPawn) (Variables|FortMovementMode_BaseExtLogic|GetOwnerPawn)))
          (GameplayCue|ExecuteGameplayCue%28Burst%29Local (Variables|FortMovementMode_BaseExtLogic|GetOwnerPawn) (Variables|GameplayCues|GetGC_Dodge_Left) (Ability|GameplayCue|MakeGameplayCueParameters 0.0))
          (else
            (Tags|CancelAbilitiesWithTags (Variables|Default|GetOwnerASC) (GameplayTags|MakeGameplayTagContainerfromTag "(TagName=\"Granted.Athena.SunRose.Wings.Dodging.Left\")")))))
      (:DodgingRight
        ;[1 1 1] Dodging Right Status Change
        (if bIsActive
          (Ability|SendGameplayEventtoActor (Variables|FortMovementMode_BaseExtLogic|GetOwnerPawn) "(TagName=\"Granted.Athena.SunRose.Wings.Dodging.Right\")" (Utilities|Struct|MakeGameplayEventData "(TagName=\"Granted.Athena.SunRose.Wings.Dodging.Right\")" (Variables|FortMovementMode_BaseExtLogic|GetOwnerPawn) (Variables|FortMovementMode_BaseExtLogic|GetOwnerPawn)))
          (GameplayCue|ExecuteGameplayCue%28Burst%29Local (Variables|FortMovementMode_BaseExtLogic|GetOwnerPawn) (Variables|GameplayCues|GetGC_Dodge_Right) (Ability|GameplayCue|MakeGameplayCueParameters 0.0))
          (else
            (Tags|CancelAbilitiesWithTags (Variables|Default|GetOwnerASC) (GameplayTags|MakeGameplayTagContainerfromTag "(TagName=\"Granted.Athena.SunRose.Wings.Dodging.Right\")"))))))))

(event EventHandleFlyingStatusActivationFailed
  (params (FailedStatus :type EWingedFlyingStatus%20Enum :default "Default") (FailureTag :type Gameplay%20Tag%20Structure%20%28by%20ref%29))
  ;[1 1 1] Handle flying status activation failed
  (if (GameplayTags|HasTag (Variables|Default|GetTags_ToAllowUnableText) FailureTag true)
    (if (ScalableFloat|GetValueasBool (Variables|Balance|StartFlying|GetRow_StartFlying_DisplayUnableText))
      (switch Utilities|FlowControl|Switch|SwitchonEWingedFlyingStatus FailedStatus
        (:Boosting
          _)
        (:Ascending
          _)
        (:Crashing
          _)
        (:DodgingLeft
          _)
        (:DodgingRight
          _))
      (bind _asfort_player_pawn (Utilities|Casting|CastToFortPlayerPawn (Variables|FortMovementMode_BaseExtLogic|GetOwnerPawn)))
      (FailedAction|UnabletoPerformAction _asfort_player_pawn false (GameplayTags|MakeGameplayTagContainerfromTag FailureTag)))))

(fn CanBeginFlying
  (return (GameplayTags|HasMatchingGameplayTag (Variables|FortMovementMode_BaseExtLogic|GetOwnerPawn) (Variables|Default|GetTag_RequiredToContinueFlying))))

(fn CanContinueFlyingStatus
  (params (MoveStatus :type Fort%20Movement%20Mode%20Winged%20Flying%20Runtime%20Data%20Object%20Reference) (StatusToContinue :type EWingedFlyingStatus%20Enum :default "Default") (ElapsedDuration :type Float%20%28single-precision%29))
  (bind _movestatus (Variables|GetMoveStatus))
  (bind _flyingstatus (Variables|/Script/SunRoseFlyingRuntime.FortMovementMode_WingedFlyingRuntimeData|GetFlyingStatus _movestatus))
  (bind _movestatus_2 (Variables|GetMoveStatus))
  (bind _flyingstatus_2 (Variables|/Script/SunRoseFlyingRuntime.FortMovementMode_WingedFlyingRuntimeData|GetFlyingStatus _movestatus_2))
  (Utilities|IsValid MoveStatus
    (:Is%20Valid
      (switch Utilities|FlowControl|Switch|SwitchonEWingedFlyingStatus StatusToContinue
        (:Default
          _)
        (:Boosting
          ;[1 1 1] Boost can continue check
          (if (Class|FortWingedFlyingBlueprintLibrary|FlyingStatusHasAnyFlags _flyingstatus 4)
            _
            (elif (Math|Boolean|ANDBoolean (Class|FortWingedFlyingBlueprintLibrary|FlyingStatusHasAnyFlags _flyingstatus 2) (> (Class|FortWingedFlyingBlueprintLibrary|GetFlyingStatusElapsedTime "Ascending" _movestatus) 0) (<= (Class|FortWingedFlyingBlueprintLibrary|GetFlyingStatusElapsedTime "Ascending" _movestatus) (Variables|GetElapsedDuration)))
              (Variables|/Script/SunRoseFlyingRuntime.FortPawnComponent_WingedFlying|SetbIntendsToBoost false (Variables|Default|GetFlyingPawnComponent))
              (else
                (return true 0))))
          (return false 0))
        (:Ascending
          ;[1 1 1] Ascend can continue check
          (if (Class|FortWingedFlyingBlueprintLibrary|FlyingStatusHasAnyFlags _flyingstatus_2 4)
            _
            (elif (Math|Boolean|ANDBoolean (Class|FortWingedFlyingBlueprintLibrary|FlyingStatusHasAnyFlags _flyingstatus_2 1) (> (Class|FortWingedFlyingBlueprintLibrary|GetFlyingStatusElapsedTime "Boosting" _movestatus_2) 0) (<= (Class|FortWingedFlyingBlueprintLibrary|GetFlyingStatusElapsedTime "Boosting" _movestatus_2) (Variables|GetElapsedDuration)))
              (Variables|/Script/SunRoseFlyingRuntime.FortPawnComponent_WingedFlying|SetbIntendsToAscend false (Variables|Default|GetFlyingPawnComponent))
              (else
                (return true 0))))
          (return false 0))
        (:Crashing
          _)
        (:DodgingLeft
          ;[1 1 1] Dodge can continue check
          (return (not (Class|FortWingedFlyingBlueprintLibrary|FlyingStatusHasAnyFlags (Variables|/Script/SunRoseFlyingRuntime.FortMovementMode_WingedFlyingRuntimeData|GetFlyingStatus (Variables|GetMoveStatus)) 4)) 0))
        (:DodgingRight
          (return (not (Class|FortWingedFlyingBlueprintLibrary|FlyingStatusHasAnyFlags (Variables|/Script/SunRoseFlyingRuntime.FortMovementMode_WingedFlyingRuntimeData|GetFlyingStatus (Variables|GetMoveStatus)) 4)) 0)))
      (return true 0))))

(fn CanEndFlyingStatus
  (params (MoveStatus :type Fort%20Movement%20Mode%20Winged%20Flying%20Runtime%20Data%20Object%20Reference) (StatusToEnd :type EWingedFlyingStatus%20Enum :default "Default") (ElapsedDuration :type Float%20%28single-precision%29))
  (Utilities|IsValid MoveStatus
    (:Is%20Valid
      (LocalVariables|SetIsCrashing (Class|FortWingedFlyingBlueprintLibrary|FlyingStatusHasAnyFlags (Variables|/Script/SunRoseFlyingRuntime.FortMovementMode_WingedFlyingRuntimeData|GetFlyingStatus MoveStatus) 4))
      (switch Utilities|FlowControl|Switch|SwitchonEWingedFlyingStatus StatusToEnd
        (:Default
          _)
        (:Boosting
          ;[1 1 1] Boost End Condition
          (return (or (LocalVariables|GetIsCrashing) (>= (Variables|GetElapsedDuration) (ScalableFloat|GetValueatLevel (Variables|Balance|Boost|GetRow_Boost_MinDuration))))))
        (:Ascending
          ;[1 1 1] Ascend End Condition
          (return (or (LocalVariables|GetIsCrashing) (>= (Variables|GetElapsedDuration) (ScalableFloat|GetValueatLevel (Variables|Balance|Ascend|GetRow_Ascend_MinDuration))))))
        (:Crashing
          _)
        (:DodgingLeft
          ;[1 1 1] Dodge End condition
          (return (or (LocalVariables|GetIsCrashing) (>= (Variables|GetElapsedDuration) (ScalableFloat|GetValueatLevel (Variables|Balance|Dodge|GetRow_Dodge_Duration))))))
        (:DodgingRight
          (return (or (LocalVariables|GetIsCrashing) (>= (Variables|GetElapsedDuration) (ScalableFloat|GetValueatLevel (Variables|Balance|Dodge|GetRow_Dodge_Duration)))))))
      (return false))
    (:Is%20Not%20Valid
      (return false))))

(fn CanStartFlyingStatus
  (params (MoveStatus :type Fort%20Movement%20Mode%20Winged%20Flying%20Runtime%20Data%20Object%20Reference) (StatusToStart :type EWingedFlyingStatus%20Enum :default "Default"))
  (bind _movestatus (Variables|GetMoveStatus))
  (Utilities|IsValid MoveStatus
    (:Is%20Valid
      (switch Utilities|FlowControl|Switch|SwitchonEWingedFlyingStatus StatusToStart
        (:Default
          _)
        (:Boosting
          (if (Class|FortWingedFlyingBlueprintLibrary|FlyingStatusHasAnyFlags (Variables|/Script/SunRoseFlyingRuntime.FortMovementMode_WingedFlyingRuntimeData|GetFlyingStatus _movestatus) 4) :!label _label
            (return false "(TagName=\"\")")
            (else
              (return true "(TagName=\"\")"))))
        (:Ascending
          (goto _label))
        (:Crashing
          _)
        (:DodgingLeft
          (if (Variables|/Script/SunRoseFlyingRuntime.FortMovementMode_WingedFlyingRuntimeData|GetbIsStartingFlying _movestatus) :!label _label_2
            (return false "(TagName=\"Athena.Ability.Failed.CannotDoThatNow\")")
            (else
              (goto _label))))
        (:DodgingRight
          (goto _label_2))))
    (:Is%20Not%20Valid
      _))
  (return true 0))

(fn InputConditionChanged
  (params (InventoryActive :type Boolean))
  (CallFunction|ToggleInputMappingContext (not InventoryActive)))

(fn ToggleInputMappingContext
  (params (Enable :type Boolean))
  (bind _cachedplayercontroller (Variables|Default|GetCachedPlayerController))
  (bind _returnvalue (PlayerController|LocalPlayerSubsystems|GetEnhancedInputLocalPlayerSubsystem _cachedplayercontroller))
  (bind _ownerpawn (Variables|FortMovementMode_BaseExtLogic|GetOwnerPawn))
  (bind _flyinginputcontext (Variables|Default|GetFlyingInputContext))
  (bind _flyinginputcontext_2 (Variables|Default|GetFlyingInputContext))
  (Utilities|IsValid _cachedplayercontroller
    (:Is%20Valid
      (if Enable
        (Input|AddMappingContext _returnvalue _flyinginputcontext 1000 "(bIgnoreAllPressedKeysUntilRelease=True,bForceImmediately=False,bNotifyUserSettings=False)")
        (Utilities|IsValid _ownerpawn
          (:Is%20Valid
            (Class|FortWingedFlyingBlueprintLibrary|SetInputActionDisplayContext _ownerpawn _flyinginputcontext_2)))
        (else
          (Input|RemoveMappingContext _returnvalue _flyinginputcontext "(bIgnoreAllPressedKeysUntilRelease=True,bForceImmediately=False,bNotifyUserSettings=False)")
          (Utilities|IsValid _ownerpawn
            (:Is%20Valid
              (Class|FortWingedFlyingBlueprintLibrary|ClearInputActionDisplayContext _ownerpawn _flyinginputcontext_2))))))))

(fn UpdateForHitResponse
  (params (MoveStatus :type Fort%20Movement%20Mode%20Winged%20Flying%20Runtime%20Data%20Object%20Reference) (DeltaTime :type Float%20%28single-precision%29) (InVelocity :type Vector%20%28by%20ref%29 :default "0, 0, 0") (OldPosition :type Vector%20%28by%20ref%29 :default "0, 0, 0") (Adjustment :type Vector%20%28by%20ref%29 :default "0, 0, 0") (Hit :type Hit%20Result%20Structure%20%28by%20ref%29)))

(fn UpdateForRotation
  (params (MoveStatus :type Fort%20Movement%20Mode%20Winged%20Flying%20Runtime%20Data%20Object%20Reference) (DeltaTime :type Float%20%28single-precision%29)))
