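; Flying pawn component for SunRose wings in Fortnite, extending the native winged flying component to track a player pawn's flight state.
; Movement status updates feed flying energy and status, while forward speed and added downward velocity mirror into replicated variables that only simulated proxies apply.
; Once the ability system is ready, a damage ForceCrash tag triggers crashing.
; Left and right trace results store skim location, normal and distance; input action events are empty.
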(event Custom|UpdateCurrentFlyingEnergyRemaining
  (params (NewEnergyRemaining :type Float%20%28double-precision%29) (bForceUpdate :type Boolean))
  ;[1 1 1] Update Current Flying Energy Remaining
  (CallFunction|UpdateCurrentFlyingEnergy NewEnergyRemaining bForceUpdate))

(event Custom|OnASCReady
  (params (AbilitySystemComponent :type Fort%20Ability%20System%20Component%20Object%20Reference) (AffectedPawn :type Fort%20Player%20Pawn%20Object%20Reference))
  ;[1 1 1] On ASC Ready
  (Ability|Tasks|WaitGameplayTagAddToActor (Variables|Default|GetOwnerPawn) "(TagName=\"Granted.Athena.SunRose.Wings.ForceCrash.Damaged\")"
    (:Added
      (CallFunction|TryStartCrashing true))))

(event EventBeginPlay
  (bind _ownerpawn (Variables|Default|GetOwnerPawn))
  (bind _returnvalue (Ability|GetFortAbilitySystemComponent _ownerpawn))
  ;[1 1 1] Begin Play
  (bind _asfort_player_pawn (Utilities|Casting|CastToFortPlayerPawn (Components|GetOwner)))
  (Variables|Default|SetOwnerPawn _asfort_player_pawn)
  (Utilities|IsValid _returnvalue
    (:Is%20Valid
      (CallFunction|OnASCReady _returnvalue _ownerpawn))
    (:Is%20Not%20Valid
      (EventDispatchers|BindEventtoOnAbilitySystemComponentInitialized _ownerpawn (EventDispatchers|CreateEvent :!event Custom|OnASCReady) :!class /Script/FortniteGame.FortPlayerPawn))))

(event EnhancedInputActionWingedFlyingBoostAction
  (params (ActionValue :type Boolean) (ElapsedSeconds :type Float%20%28double-precision%29) (TriggeredSeconds :type Float%20%28double-precision%29) (InputAction :type Input%20Action%20Object%20Reference :default "/SunRoseFlyingGameplay/Input/WingedFlyingBoostAction.WingedFlyingBoostAction")))

(event EnhancedInputActionWingedFlyingAscendAction
  (params (ActionValue :type Boolean) (ElapsedSeconds :type Float%20%28double-precision%29) (TriggeredSeconds :type Float%20%28double-precision%29) (InputAction :type Input%20Action%20Object%20Reference :default "/SunRoseFlyingGameplay/Input/WingedFlyingAscendAction.WingedFlyingAscendAction")))

(event EnhancedInputActionWingedFlyingCrashAction
  (params (ActionValue :type Boolean) (ElapsedSeconds :type Float%20%28double-precision%29) (TriggeredSeconds :type Float%20%28double-precision%29) (InputAction :type Input%20Action%20Object%20Reference :default "/SunRoseFlyingGameplay/Input/WingedFlyingCrashAction.WingedFlyingCrashAction")))

(event EnhancedInputActionWingedFlyingRollLeftAction
  (params (ActionValue :type Boolean) (ElapsedSeconds :type Float%20%28double-precision%29) (TriggeredSeconds :type Float%20%28double-precision%29) (InputAction :type Input%20Action%20Object%20Reference :default "/SunRoseFlyingGameplay/Input/WingedFlyingRollLeftAction.WingedFlyingRollLeftAction")))

(event EnhancedInputActionWingedFlyingRollRightAction
  (params (ActionValue :type Boolean) (ElapsedSeconds :type Float%20%28double-precision%29) (TriggeredSeconds :type Float%20%28double-precision%29) (InputAction :type Input%20Action%20Object%20Reference :default "/SunRoseFlyingGameplay/Input/WingedFlyingRollRightAction.WingedFlyingRollRightAction")))

(event EventHandleFlyingMoveStatusUpdated
  (params (MoveStatus :type Fort%20Movement%20Mode%20Winged%20Flying%20Runtime%20Data%20Object%20Reference))
  (Utilities|IsValid MoveStatus
    (:Is%20Valid
      (|CallParentFunction MoveStatus)
      (CallFunction|UpdateCurrentFlyingEnergyRemaining (Variables|/Script/SunRoseFlyingRuntime.FortMovementMode_WingedFlyingRuntimeData|GetFlyingEnergy MoveStatus))
      (Variables|FortPawnComponent_WingedFlying|SetFlyingStatus (Variables|/Script/SunRoseFlyingRuntime.FortMovementMode_WingedFlyingRuntimeData|GetFlyingStatus MoveStatus)))))

(event EventHandleLeftSkimCheckComplete
  (params (HitResult :type Hit%20Result%20Structure%20%28by%20ref%29))
  (bind (_bblockinghit _binitialoverlap _time _distance _location _impactpoint _normal _impactnormal _physmat _hitactor _hitcomponent _hitbonename _bonename _hititem _elementindex _faceindex _tracestart) (Collision|BreakHitResult HitResult))
  ;[1 1 1] Skimming
  (if _bblockinghit
    (Variables|Default|SetSkim_Location_Left _location)
    (Variables|Default|SetSkim_Normal_Left _normal)
    (Variables|Default|SetSkim_Distance_Left (Math|Vector|Distance(Vector) (Variables|Default|GetSkim_Location_Left) _tracestart))
    (else
      (Variables|Default|SetSkim_Distance_Left -1.0))))

(event EventHandleAddedDownZVelocityUpdated
  (params (NewAddedDownZVelocity :type Float%20%28single-precision%29))
  (CallFunction|SetAddedDownZVelocity NewAddedDownZVelocity))

(event EventHandleRightSkimCheckComplete
  (params (HitResult :type Hit%20Result%20Structure%20%28by%20ref%29))
  (bind (_bblockinghit _binitialoverlap _time _distance _location _impactpoint _normal _impactnormal _physmat _hitactor _hitcomponent _hitbonename _bonename _hititem _elementindex _faceindex _tracestart) (Collision|BreakHitResult HitResult))
  ;[1 1 1] Skimming
  (if _bblockinghit
    (Variables|Default|SetSkim_Location_Right _location)
    (Variables|Default|SetSkim_Normal_Right _normal)
    (Variables|Default|SetSkim_Distance_Right (Math|Vector|Distance(Vector) (Variables|Default|GetSkim_Location_Right) _tracestart))
    (else
      (Variables|Default|SetSkim_Distance_Right -1.0))))

(event EventHandleForwardSpeedUpdated
  (params (NewForwardSpeed :type Float%20%28single-precision%29))
  (CallFunction|SetForwardSpeed NewForwardSpeed))

(fn OnRep_ReplicatedAddedDownZVelocity
  (if (Utilities|Enum|Equal(Enum) (Networking|GetLocalRole (Variables|Default|GetOwnerPawn)) "ROLE_SimulatedProxy")
    (Variables|FortPawnComponent_WingedFlying|SetAddedDownZVelocity (Variables|Default|GetReplicatedAddedDownZVelocity))))

(fn OnRep_ReplicatedForwardSpeed
  (if (Utilities|Enum|Equal(Enum) (Networking|GetLocalRole (Variables|Default|GetOwnerPawn)) "ROLE_SimulatedProxy")
    (Variables|FortPawnComponent_WingedFlying|SetForwardSpeed (Variables|Default|GetReplicatedForwardSpeed))))

(fn ResetInputStatus
  (params (bResetCrashInput :type Boolean :default true))
  (Utilities|FlowControl|Sequence
    (:then_0
      (Variables|FortPawnComponent_WingedFlying|SetbIntendsToBoost false)
      (Variables|FortPawnComponent_WingedFlying|SetbIntendsToAscend false)
      (Variables|FortPawnComponent_WingedFlying|SetbIntendsToRollLeft false)
      (Variables|FortPawnComponent_WingedFlying|SetbIntendsToRollRight false)
      (Variables|Default|SetAscendPressed false)
      (Variables|Default|SetBoostPressed false))
    (:then_1
      (if (Variables|GetbResetCrashInput)
        (Variables|FortPawnComponent_WingedFlying|SetbIntendsToCrash false)))))

(fn SetAddedDownZVelocity
  (params (NewAddedDownZVelocity :type Float%20%28double-precision%29))
  (if (!= NewAddedDownZVelocity (Variables|FortPawnComponent_WingedFlying|GetAddedDownZVelocity))
    (Variables|FortPawnComponent_WingedFlying|SetAddedDownZVelocity NewAddedDownZVelocity)
    (Variables|Default|SetReplicatedAddedDownZVelocity NewAddedDownZVelocity)))

(fn SetForwardSpeed
  (params (NewForwardSpeed :type Float%20%28double-precision%29))
  (if (!= NewForwardSpeed (Variables|FortPawnComponent_WingedFlying|GetForwardSpeed))
    (Variables|FortPawnComponent_WingedFlying|SetForwardSpeed NewForwardSpeed)
    (Variables|Default|SetReplicatedForwardSpeed NewForwardSpeed)))

(fn UpdateMoveXInput
  (params (NewMoveInputX :type Float%20%28double-precision%29)))
