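# Player animation Blueprint for the SunRose wings item, supplying item-layer overrides that animate winged flight and landing.
# The update event reads flying status flags, gameplay tags, velocity, height and magazine-ammo fuel into state booleans, smoothed glide pitch/roll and curve-driven alphas.
# The full-body override blends a conduit-hubbed flying state machine (leap, glide, boost, dodge, dive), layered with pitch/roll/yaw additives, against a ground landing state machine, finished with dead blending.
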AnimBlueprint /SunRoseFlyingGameplay/Items/Wings/Blueprints/SunRose_Player_AnimBP_PreNativized
target_skeleton=&Fortnite_M_Avg_Player_Skeleton
graphs [
 AnimGraph {nodes [Output Pose() {layerGroup=None}]}
 EventGraph {nodes [
   n1 = Event BlueprintUpdateAnimation()
   > Set Delta Time X(Delta Time X=@n1.DeltaTimeX)
   > Set RandomInteger(RandomInteger:int=RandomInteger!(Max:int=200) -> (ReturnValue:int=0)) -> (Output_Get:int=0)
   > n2 = Sequence!(execute) -> (then_0, then_1, then_2:Exec, then_3:Exec, then_4:Exec)
   n3 = Get FlyingPawnComponent() -> (FlyingPawnComponent:Flying Pawn Component Object Reference)
   > Is Valid(InputObject=@n3)
   > Set Current Height Above World(Current Height Above World:double=Get NativeCurrentHeightAboveWorld!(self:Fort Pawn Component Winged Flying Object Reference=@n3) -> (NativeCurrentHeightAboveWorld:float)) -> (Output_Get:double)
   n4 = Set FlyingPawnComponent(
    execute=@n2.then_1
    FlyingPawnComponent:Flying Pawn Component Object Reference=GetComponentByClass!(
     self:Actor Object Reference=Get FortPlayerPawn() -> (FortPlayerPawn:Fort Player Pawn Object Reference)
     ComponentClass:Actor Component Class Reference=&/FlyingPawnComponent.FlyingPawnComponent_C
    ) -> (ReturnValue:Flying Pawn Component Object Reference)
   ) -> (Output_Get:Flying Pawn Component Object Reference)
   n5 = Get FlyingStatus!(self:Fort Pawn Component Winged Flying Object Reference=@n4) -> (FlyingStatus:EWingedFlyingStatus Enum=Default)
   > Set IsWingsBoosting(IsWingsBoosting:bool=FlyingStatusHasAnyFlags!(Status:EWingedFlyingStatus Enum=@n5, Flags:Bitmask Integer=1) -> (ReturnValue:bool)) -> (Output_Get:bool)
   > Set IsDodgeLeft(IsDodgeLeft:bool=FlyingStatusHasAnyFlags(Status=@n5, Flags=8)) -> (Output_Get:bool)
   > Set IsDodgeRight(IsDodgeRight:bool=FlyingStatusHasAnyFlags(Status=@n5, Flags=16)) -> (Output_Get:bool)
   > Set IsAscending(IsAscending:bool=FlyingStatusHasAnyFlags(Status=@n5, Flags=2)) -> (Output_Get:bool)
   > Set IsWingsCrashing(IsWingsCrashing:bool=FlyingStatusHasAnyFlags(Status=@n5, Flags=4)) -> (Output_Get:bool)
   n6 = Get FortPlayerPawn()
   > Set IsWingsFlying(IsWingsFlying:bool=HasMatchingGameplayTag!(self:Gameplay Tag Asset Interface Interface=@n6, TagToCheck:Gameplay Tag Structure=(TagName="MovementMode.Flying.WingedFlight")) -> (ReturnValue:bool)) -> (Output_Get:bool)
   > Set IsWingLeaping(IsWingLeaping:bool=HasMatchingGameplayTag(self=@n6, TagToCheck=(TagName="Granted.Athena.SunRose.Wings.Leaping"))) -> (Output_Get:bool)
   > Set IsForceDive(IsForceDive:bool=HasMatchingGameplayTag(self=@n6, TagToCheck=(TagName="Granted.Athena.SunRose.Wings.ForceCrash"))) -> (Output_Get:bool)
   > Set IsOnGrindRail(IsOnGrindRail:bool=HasMatchingGameplayTag(self=@n6, TagToCheck=(TagName="Granted.Athena.GrindRails.IsGrindingIncludingInAir"))) -> (Output_Get:bool)
   > Set IsOnAscender(IsOnAscender:bool=HasMatchingGameplayTag(self=@n6, TagToCheck=(TagName="Gameplay.Action.Player.Ascender"))) -> (Output_Get:bool)
   n7 = Get CharacterMovement!(self:Character Object Reference=Get FortPlayerPawn()) -> (CharacterMovement:Character Movement Component Object Reference)
   > Set GlidePitch(GlidePitch:double=GetYawPitchFromVector!(InVec:Vector=GetLastUpdateVelocity!(self:Character Movement Component Object Reference=@n7) -> (ReturnValue:Vector)) -> (Yaw:float, Pitch:float).Pitch) -> (Output_Get:double)
   n8 = Get FortPlayerPawn()
   n9 = Normalize!(A:Vector=GetVelocity!(self:Actor Object Reference=@n8) -> (ReturnValue:Vector), Tolerance:float=0.0001) -> (ReturnValue_X:double, ReturnValue_Y:double, ReturnValue_Z:double) {is=title K2Node_CallFunction}
   > Set GlideRoll(GlideRoll:double=Dot Product!(A:Vector=MakeVector(X=@n9, Y=@n9.ReturnValue_Y), B:Vector=GetActorRightVector!(self:Actor Object Reference=@n8) -> (ReturnValue:Vector)) -> (ReturnValue:double)) -> (Output_Get:double)
   > "Set Pawn Rotation Rate Z (Yaw)"("Pawn Rotation Rate Z (Yaw)":float=Get!() -> (PawnRotationRate_Roll:float, PawnRotationRate_Pitch:float, PawnRotationRate_Yaw:float).PawnRotationRate_Yaw) -> (Output_Get:float)
   n10 = GetCurrentWeapon!(self:Fort Pawn Object Reference=Get FortPlayerPawn()) -> (ReturnValue:Fort Weapon Object Reference)
   > Set Fuel(Fuel:double="To Float (Integer)"(InInt=GetMagazineAmmoCount!(self:Fort Weapon Object Reference=@n10) -> (ReturnValue:int=0))) -> (Output_Get:double)
   n11 = "Get Pawn Rotation Rate Z (Yaw)"() -> ("Pawn Rotation Rate Z (Yaw)":float)
   Set GlideRollCombined(
    execute=@n2.then_2
    GlideRollCombined:double=float + float(A=Get GlideRoll() -> (GlideRoll:double), B=float * float(A=@n11, B=10) {is=title K2Node_PromotableOperator})
   ) -> (Output_Get:double)
   > Set RollAlpha(RollAlpha:float=MapRangeClamped(Value=GetCurveValue(CurveName=Roll), InRangeA=1, InRangeB=0)) -> (Output_Get:float)
   > Set PitchAlpha(PitchAlpha:double=MapRangeClamped(Value=GetCurveValue(CurveName=Pitch), InRangeA=1, InRangeB=0)) -> (Output_Get:double)
   > Set YawAlpha(YawAlpha:double=MapRangeClamped(Value=GetCurveValue(CurveName=Yaw), InRangeA=1, InRangeB=0)) -> (Output_Get:double)
   n12 = Get IsDodgeLeft() -> (IsDodgeLeft:bool)
   n13 = Get IsDodgeRight() -> (IsDodgeRight:bool)
   n14 = Get IsWingsCrashing() -> (IsWingsCrashing:bool)
   n15 = Get IsWingsBoosting() -> (IsWingsBoosting:bool)
   n16 = GetCurveValue(CurveName=interruptible)
   > Set IsLeapInterruptible(IsLeapInterruptible:bool=AND Boolean(A="float >= float"(A=@n16, B=0.1), B=OR Boolean!(A=@n14, B=@n15, C:bool=@n13, D:bool=@n12) -> (ReturnValue))) -> (Output_Get:bool)
   n17 = "float < float"(A=Get GlidePitch() -> (GlidePitch:double), B=-20)
   n18 = Get GlideRollCombined() -> (GlideRollCombined:double)
   n19 = "float < float"(A=@n18, B=2.2)
   n20 = "float >= float"(A=@n18, B=-2.2)
   n21 = "float >= float"(A=Get PawnVelocityZ(), B=100)
   n22 = Get bIsAccelerating2D()
   > Set IsAllowWingFlap(IsAllowWingFlap:bool=AND Boolean!(A=OR Boolean!(A=@n21, B=@n22) -> (ReturnValue), B=AND Boolean!(A=@n19, B=@n20) -> (ReturnValue), C:bool=NOT Boolean(A=@n17)) -> (ReturnValue)) -> (Output_Get:bool)
   > Set YawBlend(YawBlend:double=MapRangeClamped(Value=Get GlidePitch(), InRangeA=30, InRangeB=-45)) -> (Output_Get:double)
   > Set GlidePitchAlpha(GlidePitchAlpha:double=MapRangeClamped(Value=Get GlidePitch(), InRangeA=-21, InRangeB=-65)) -> (Output_Get:double)
   > n23 = Branch(Condition="Equal (Float)"(A=GetCurveValue(CurveName=BothHandIK), B=1))
   > Set GroundBlendTime(GroundBlendTime:double=0.3) -> (Output_Get:double)
   Set GroundBlendTime(execute=@n23.else, GroundBlendTime=0)
   n24 = Sequence!(execute=@n2.then_3) -> (then_0, then_1, then_2, then_3)
   > n25 = Sequence!(execute) -> (then_0, then_1, then_2)
   > n26 = Sequence()
   n27 = Get Speed2D() -> (Speed2D:float)
   n28 = AND Boolean!(
    A=AND Boolean!(A="float < float"(A=@n27, B=800), B="float >= float"(A=@n27, B=300)) -> (ReturnValue)
    B="float >= float"(A=Get Current Height Above World() -> (Current Height Above World:double), B=225)
    C:bool="float > float"(A=Get Fuel() -> (Fuel:double))
   ) -> (ReturnValue)
   n29 = Get Current Height Above World()
   > n30 = Branch(Condition=OR Boolean(A="float < float"(A=@n29, B=225), B=@n28))
   > n31 = Branch(Condition=@n28)
   > Set SlowDownOrLandTarget(SlowDownOrLandTarget:double=0.25) -> (Output_Get:double)
   Set SlowDownOrLandTarget(execute=@n31.else, SlowDownOrLandTarget=1)
   Set SlowDownOrLandTarget(execute=@n30.else, SlowDownOrLandTarget=0)
   Set SlownDownOrLandAlpha(
    execute=@n26.then_1
    SlownDownOrLandAlpha:double=FInterpTo(
     Current=Get SlownDownOrLandAlpha() -> (SlownDownOrLandAlpha:double)
     Target=Get SlowDownOrLandTarget() -> (SlowDownOrLandTarget:double)
     DeltaTime=Get Delta Time X()
     InterpSpeed=3
    )
   ) -> (Output_Get:double)
   n32 = Branch(execute=@n25.then_1, Condition=Get IsAscending() -> (IsAscending:bool))
   > Set GlidePitchTarget(GlidePitchTarget:double=float * float(A=Get GlidePitch(), B=0.25)) -> (Output_Get:double)
   n33 = Get SlowDownOrLandTarget()
   n34 = Branch(execute=@n32.else, Condition=AND Boolean(A=Get IsWingLeaping() -> (IsWingLeaping:bool), B="float < float"(A=@n33, B=200)))
   > Set GlidePitchTarget(GlidePitchTarget=0)
   n35 = Get IsWingsCrashing()
   n36 = Get Speed2D()
   n37 = "float < float"(A=@n36, B=800)
   n38 = "float >= float"(A=@n36, B=300)
   n39 = Get Current Height Above World()
   n40 = GetCurveValue(CurveName=UserCurve01)
   n41 = Get Fuel()
   n42 = Branch(
    execute=@n34.else
    Condition=AND Boolean!(
     A=AND Boolean!(A=@n37, B=@n38) -> (ReturnValue)
     B="float >= float"(A=@n39, B=225)
     C:bool="float > float"(A=@n41)
     D:bool=NOT Boolean(A=@n35)
     E:bool="Equal (Float)"(A=@n40)
    ) -> (ReturnValue)
   )
   > Set GlidePitchTarget(GlidePitchTarget=20)
   Set GlidePitchTarget(execute=@n42.else, GlidePitchTarget=Get GlidePitch())
   n43 = Branch(execute=@n25.then_2, Condition=Get IsWingsCrashing())
   > Set GlidePitchInterp(GlidePitchInterp:double=Get GlidePitch()) -> (Output_Get:double)
   Set GlidePitchInterp(
    execute=@n43.else
    GlidePitchInterp=FInterpTo(
     Current=Get GlidePitchInterp() -> (GlidePitchInterp:double)
     Target=Get GlidePitchTarget() -> (GlidePitchTarget:double)
     DeltaTime=Get Delta Time X()
     InterpSpeed=3
    )
   )
   n44 = Get Fuel()
   n45 = Get Current Height Above World()
   n46 = Branch(execute=@n24.then_1, Condition=OR Boolean(A="float < float"(A=@n45, B=800), B="Equal (Float)"(A=@n44)))
   n47 = Get Current Height Above World()
   > Set GlideRollCombinedTarget(GlideRollCombinedTarget:double=float * float(A=Get GlideRollCombined(), B=MapRangeClamped(Value=@n47, InRangeA=800, InRangeB=0, OutRangeA=1, OutRangeB=0))) -> (Output_Get:double)
   Set GlideRollCombinedTarget(execute=@n46.else, GlideRollCombinedTarget=Get GlideRollCombined())
   Set GlideRollInterp(
    execute=@n24.then_2
    GlideRollInterp:double=FInterpTo(
     Current=Get GlideRollInterp() -> (GlideRollInterp:double)
     Target=Get GlideRollCombinedTarget() -> (GlideRollCombinedTarget:double)
     DeltaTime=Get Delta Time X()
     InterpSpeed=20
    )
   ) -> (Output_Get:double)
   Set IsBoostStartEnd(execute=@n24.then_3, IsBoostStartEnd:bool="float >= float"(A=GetCurveValue(CurveName=Boosting), B=0.8)) -> (Output_Get:bool)
   n48 = GetCurveValue(CurveName=Dive)
   n49 = GetCurveValue(CurveName=Hurdle_Land)
   > Set IsPlayLandAdditive(IsPlayLandAdditive:bool=AND Boolean(A="float >= float"(A=@n49, B=0.5), B="float < float"(A=@n48, B=0.5))) -> (Output_Get:bool)
   > Set isPlayDiveHitToIdle(isPlayDiveHitToIdle:bool="float >= float"(A=GetCurveValue(CurveName=Dive), B=0.5)) -> (Output_Get:bool)
   > Set IsDiveStart(IsDiveStart:bool=OR Boolean(A=Get IsWingsCrashing(), B=Get IsForceDive() -> (IsForceDive:bool))) -> (Output_Get:bool)
   n50 = Get IsDodgeLeft()
   n51 = Get IsDodgeRight()
   > Set ToBoosting(ToBoosting:bool=AND Boolean!(A=NOT Boolean(A=@n51), B=NOT Boolean(A=@n50), C:bool=Get IsWingsBoosting()) -> (ReturnValue)) -> (Output_Get:bool)
   > Set IsExtBoostEarly(IsExtBoostEarly:bool=OR Boolean!(A=Get IsWingsCrashing(), B=Get IsWingsBoosting(), C:bool=Get IsDodgeRight(), D:bool=Get IsDodgeLeft()) -> (ReturnValue)) -> (Output_Get:bool)
   > Set IsFlyingStart(IsFlyingStart:bool=OR Boolean!(A=Get IsWingLeaping(), B=Get IsWingsFlying() -> (IsWingsFlying:bool)) -> (ReturnValue)) -> (Output_Get:bool)
   n52 = Event BlueprintProcessAnimInputEvent()
   > Branch(Condition=IsNamed(self=@n52.AnimInputEvent, InEventName=aie.WingReplenish))
   > n53 = Set isWingReplenish(isWingReplenish:bool=true) -> (Output_Get:bool)
   > Branch(Condition=@n53)
   > Delay(Duration=0.1)
   > Set isWingReplenish(isWingReplenish=false)
  ]}
 ItemAimOffset {nodes [
   Input Pose!() -> (Pose, AimOffsetAlpha:double, Yaw:double, Pitch:double) {name=InPose_AimOffset}
   Input Pose!() -> (Pose) {name=InPose_PreAimOffsets}
   Output Pose() {layerGroup=Item Override}
  ]}
 ItemFallAdditive {nodes [
   Input Pose() {name=InPoseItemFallAdditive}
   Output Pose() {layerGroup=Item Override}
  ]}
 ItemFallLandAdditive {nodes [
   Input Pose() {name=InPoseItemFallLandAdditive}
   Output Pose() {layerGroup=Item Override}
  ]}
 ItemFallLevel2Override {nodes [
   Input Pose() {name=InPose_FallLevel2}
   Output Pose() {layerGroup=Item Override}
  ]}
 ItemFinalPoseOverride {nodes [
   Input Pose() {name=InFinalPose}
   Output Pose() {layerGroup=Item Override}
  ]}
 ItemFullBodyAimOffsets {nodes [
   Input Pose!() -> (Pose, FullBody_AimOffsetAlpha:double, Aim_Yaw:double, Aim_Pitch:double) {name=InPoseFullBodyAimOffsets}
   Input Pose!() -> (Pose) {name=InPosePreFullBodyAimOffsets}
   Output Pose() {layerGroup=Item Override}
  ]}
 ItemFullBodyOverride {nodes [
   n1 = Get GlidePitchInterp()
   n2 = Get GlideRollInterp()
   n3 = Blend Poses by bool(
    BlendPose_0=Blendspace Player 'Olympus_Player_Glide_Boost_Pitch_Roll_BS'(X=@n2, Y=@n1) {blendSpace=/SunRoseFlyingGameplay/Items/Wings/Animations/Player/Olympus_Player_Glide_Boost_Pitch_Roll_BS}
    BlendPose_1=Blendspace Player 'Olympus_Player_Glide_Pitch_Roll_BS'(X=@n2, Y=@n1) {blendSpace=/SunRoseFlyingGameplay/Items/Wings/Animations/Player/Olympus_Player_Glide_Pitch_Roll_BS}
    BlendTime_0=0.25
    BlendTime_1=0.2
    bActiveValue=Get IsWingsBoosting()
   ) {blendType=Linear}
   AnimGraphNode_StateMachine_0 = FlyingStateMachine() {is=title AnimGraphNode_StateMachine, bAllowConduitEntryStates=true}
   n4 = Get PitchAlpha() -> (PitchAlpha:double)
   n5 = Blendspace Player 'Olympus_Player_Glide_Yaw_BS'!(X=Get GlideRollCombined()) -> (Pose) {blendSpace=/SunRoseFlyingGameplay/Items/Wings/Animations/Player/Olympus_Player_Glide_Yaw_BS}
   n6 = "Sequence Player (additive) 'Olympus_Player_Glide_Center'"() {bLoopAnimation=false, sequence=/SunRoseFlyingGameplay/Items/Wings/Animations/Player/Olympus_Player_Glide_Center}
   n7 = Get YawBlend() -> (YawBlend:double)
   n8 = Modify Curve!(
    SourcePose=Apply Additive(
     Base=Apply Additive(Base=@AnimGraphNode_StateMachine_0, Additive=@n3, Alpha=@n4) {alphaBoolBlend {blendOption=Linear}}
     Additive=Two Way Blend(A=@n6, B=@n5, Alpha=@n7) {blendNode {alphaBoolBlend {blendOption=Linear}}}
     Alpha=Get YawAlpha() -> (YawAlpha:double)
    ) {alphaBoolBlend {blendOption=Linear}}
    CurveValues_0=1
    CurveValues_1:float=1
    CurveValues_2:float=1
    Alpha=1
   ) -> (Pose) {curveNames [BothHandFK,BothLegFK,Hurdle_Land], curveValues [0,0,0]}
   AnimGraphNode_StateMachine_1 = GroundStateMachine() {is=title AnimGraphNode_StateMachine, bAllowConduitEntryStates=true}
   n9 = Get IsFlyingStart() -> (IsFlyingStart:bool)
   n10 = Get GroundBlendTime() -> (GroundBlendTime:double)
   Output Pose(Result=Dead Blending!(Source:Pose Link Structure=Blend Poses by bool(BlendPose_0=@n8, BlendPose_1=@AnimGraphNode_StateMachine_1, BlendTime_1=@n10, bActiveValue=@n9) {blendType=Sinusoidal, childUpateMode=ResetChildOnActivate}) -> (Pose:Pose Link Structure) {is=title AnimGraphNode_DeadBlending, defaultBlendDuration=0.1, defaultBlendMode=Linear}) {layerGroup=Item Override}
   Save cached pose 'SavedPose9831'(Pose=Input Pose() {name=InPoseFullBody}) {cacheName=SavedPose9831}
   Slot 'OverrideFullBody'() {slotName=OverrideFullBody}
   "Get Pawn Rotation Rate Z (Yaw)"()
  ]}
 FlyingStateMachine {
  entry=Conduit
  path=ItemFullBodyOverride.AnimGraphNode_StateMachine_0.FlyingStateMachine
  ids=id
  nodes [
   ToDodge() {is=title AnimStateAliasNode, also {aliases [BoostLoop, BoostStart, DodgeLeftEnd, DodgeRightEnd]}}
   ToBoosting() {is=title AnimStateAliasNode, also {aliases [DodgeLeftEnd, DodgeRightEnd]}}
   ToDive() {is=title AnimStateAliasNode, also {aliases [BoostEnd, DodgeLeftEnd, DodgeRightEnd, Gliding]}}
   AnimStateConduitNode_0 = Conduit()
   AnimStateNode_0 = Leap() {
    is=title AnimStateNode
    stateEntered {guid=#1}
    stateFullyBlended {guid=#2}
    stateLeft {guid=#3}
   }
   AnimStateNode_1 = DodgeRightStart() {is=title AnimStateNode}
   AnimStateNode_10 = DodgeLeftEnd() {is=title AnimStateNode}
   AnimStateNode_2 = DiveLoop() {is=title AnimStateNode}
   AnimStateNode_3 = Dive Start() {
    is=title AnimStateNode
    stateEntered {guid=#4}
    stateFullyBlended {guid=#5}
    stateLeft {guid=#6}
   }
   AnimStateNode_4 = DodgeLeftStart() {is=title AnimStateNode}
   AnimStateNode_5 = BoostStart() {is=title AnimStateNode}
   AnimStateNode_6 = BoostLoop() {
    is=title AnimStateNode
    stateEntered {guid=#7}
    stateFullyBlended {guid=#8}
    stateLeft {guid=#9}
   }
   AnimStateNode_7 = Gliding() {is=title AnimStateNode}
   AnimStateNode_8 = BoostEnd() {is=title AnimStateNode}
   AnimStateNode_9 = DodgeRightEnd() {is=title AnimStateNode}
   AnimStateTransitionNode_0 = AnimStateTransitionNode() {automaticRuleTriggerTime=0, bAutomaticRuleBasedOnSequencePlayerInState=true}
   AnimStateTransitionNode_1 = AnimStateTransitionNode() {crossfadeDuration=0, priorityOrder=4}
   AnimStateTransitionNode_10 = AnimStateTransitionNode() {crossfadeDuration=0}
   AnimStateTransitionNode_11 = AnimStateTransitionNode() {crossfadeDuration=0, priorityOrder=2}
   AnimStateTransitionNode_12 = AnimStateTransitionNode() {crossfadeDuration=0.4, logicType=TLT_Inertialization, priorityOrder=100}
   AnimStateTransitionNode_13 = AnimStateTransitionNode()
   AnimStateTransitionNode_14 = AnimStateTransitionNode() {crossfadeDuration=0.1, logicType=TLT_Inertialization}
   AnimStateTransitionNode_15 = AnimStateTransitionNode()
   AnimStateTransitionNode_18 = AnimStateTransitionNode() {automaticRuleTriggerTime=0, bAutomaticRuleBasedOnSequencePlayerInState=true, crossfadeDuration=0.1, logicType=TLT_Inertialization}
   AnimStateTransitionNode_19 = AnimStateTransitionNode()
   AnimStateTransitionNode_2 = AnimStateTransitionNode() {crossfadeDuration=0.1, logicType=TLT_Inertialization}
   AnimStateTransitionNode_20 = AnimStateTransitionNode()
   AnimStateTransitionNode_21 = AnimStateTransitionNode() {crossfadeDuration=0.4, logicType=TLT_Inertialization}
   AnimStateTransitionNode_22 = AnimStateTransitionNode() {bAutomaticRuleBasedOnSequencePlayerInState=true}
   AnimStateTransitionNode_23 = AnimStateTransitionNode()
   AnimStateTransitionNode_24 = AnimStateTransitionNode() {bAutomaticRuleBasedOnSequencePlayerInState=true, crossfadeDuration=0.1}
   AnimStateTransitionNode_25 = AnimStateTransitionNode() {logicType=TLT_Inertialization}
   AnimStateTransitionNode_27 = AnimStateTransitionNode() {bAutomaticRuleBasedOnSequencePlayerInState=true}
   AnimStateTransitionNode_28 = AnimStateTransitionNode() {bAutomaticRuleBasedOnSequencePlayerInState=true}
   AnimStateTransitionNode_3 = AnimStateTransitionNode()
   AnimStateTransitionNode_4 = AnimStateTransitionNode() {crossfadeDuration=0, priorityOrder=5}
   AnimStateTransitionNode_6 = AnimStateTransitionNode() {crossfadeDuration=0.1, logicType=TLT_Inertialization, priorityOrder=2}
   AnimStateTransitionNode_7 = AnimStateTransitionNode() {bAutomaticRuleBasedOnSequencePlayerInState=true, crossfadeDuration=0.1}
   AnimStateTransitionNode_8 = AnimStateTransitionNode()
   AnimStateTransitionNode_9 = AnimStateTransitionNode() {crossfadeDuration=0.1, logicType=TLT_Inertialization, priorityOrder=3}
  ]
  links [
   {
    dst=Conduit
    rule=ItemFullBodyOverride.AnimGraphNode_StateMachine_0.FlyingStateMachine.AnimStateTransitionNode_22.Transition
    src=BoostEnd
    via=AnimStateTransitionNode_22
   }
   {
    dst=Conduit
    rule=ItemFullBodyOverride.AnimGraphNode_StateMachine_0.FlyingStateMachine.AnimStateTransitionNode_23.Transition
    src=BoostEnd
    via=AnimStateTransitionNode_23
   }
   {
    dst=BoostEnd
    rule=ItemFullBodyOverride.AnimGraphNode_StateMachine_0.FlyingStateMachine.AnimStateTransitionNode_21.Transition
    src=BoostLoop
    via=AnimStateTransitionNode_21
   }
   {
    dst=BoostLoop
    rule=ItemFullBodyOverride.AnimGraphNode_StateMachine_0.FlyingStateMachine.AnimStateTransitionNode_14.Transition
    src=BoostStart
    via=AnimStateTransitionNode_14
   }
   {
    dst=Conduit
    rule=ItemFullBodyOverride.AnimGraphNode_StateMachine_0.FlyingStateMachine.AnimStateTransitionNode_15.Transition
    src=BoostStart
    via=AnimStateTransitionNode_15
   }
   {
    dst=BoostStart
    rule=ItemFullBodyOverride.AnimGraphNode_StateMachine_0.FlyingStateMachine.AnimStateTransitionNode_9.Transition
    src=Conduit
    via=AnimStateTransitionNode_9
   }
   {
    dst=Dive Start
    rule=ItemFullBodyOverride.AnimGraphNode_StateMachine_0.FlyingStateMachine.AnimStateTransitionNode_6.Transition
    src=Conduit
    via=AnimStateTransitionNode_6
   }
   {
    dst=DodgeLeftStart
    rule=ItemFullBodyOverride.AnimGraphNode_StateMachine_0.FlyingStateMachine.AnimStateTransitionNode_4.Transition
    src=Conduit
    via=AnimStateTransitionNode_4
   }
   {
    dst=DodgeRightStart
    rule=ItemFullBodyOverride.AnimGraphNode_StateMachine_0.FlyingStateMachine.AnimStateTransitionNode_1.Transition
    src=Conduit
    via=AnimStateTransitionNode_1
   }
   {
    dst=Gliding
    rule=ItemFullBodyOverride.AnimGraphNode_StateMachine_0.FlyingStateMachine.AnimStateTransitionNode_12.Transition
    src=Conduit
    via=AnimStateTransitionNode_12
   }
   {
    dst=Leap
    rule=ItemFullBodyOverride.AnimGraphNode_StateMachine_0.FlyingStateMachine.AnimStateTransitionNode_25.Transition
    src=Conduit
    via=AnimStateTransitionNode_25
   }
   {
    dst=Conduit
    rule=ItemFullBodyOverride.AnimGraphNode_StateMachine_0.FlyingStateMachine.AnimStateTransitionNode_8.Transition
    src=Dive Start
    via=AnimStateTransitionNode_8
   }
   {
    dst=DiveLoop
    rule=ItemFullBodyOverride.AnimGraphNode_StateMachine_0.FlyingStateMachine.AnimStateTransitionNode_18.Transition
    src=Dive Start
    via=AnimStateTransitionNode_18
   }
   {
    dst=Conduit
    rule=ItemFullBodyOverride.AnimGraphNode_StateMachine_0.FlyingStateMachine.AnimStateTransitionNode_19.Transition
    src=DiveLoop
    via=AnimStateTransitionNode_19
   }
   {
    dst=Conduit
    rule=ItemFullBodyOverride.AnimGraphNode_StateMachine_0.FlyingStateMachine.AnimStateTransitionNode_27.Transition
    src=DodgeLeftEnd
    via=AnimStateTransitionNode_27
   }
   {
    dst=DodgeLeftEnd
    rule=ItemFullBodyOverride.AnimGraphNode_StateMachine_0.FlyingStateMachine.AnimStateTransitionNode_24.Transition
    src=DodgeLeftStart
    via=AnimStateTransitionNode_24
   }
   {
    dst=Conduit
    rule=ItemFullBodyOverride.AnimGraphNode_StateMachine_0.FlyingStateMachine.AnimStateTransitionNode_28.Transition
    src=DodgeRightEnd
    via=AnimStateTransitionNode_28
   }
   {
    dst=DodgeRightEnd
    rule=ItemFullBodyOverride.AnimGraphNode_StateMachine_0.FlyingStateMachine.AnimStateTransitionNode_7.Transition
    src=DodgeRightStart
    via=AnimStateTransitionNode_7
   }
   {
    dst=Conduit
    rule=ItemFullBodyOverride.AnimGraphNode_StateMachine_0.FlyingStateMachine.AnimStateTransitionNode_13.Transition
    src=Gliding
    via=AnimStateTransitionNode_13
   }
   {
    dst=Conduit
    rule=ItemFullBodyOverride.AnimGraphNode_StateMachine_0.FlyingStateMachine.AnimStateTransitionNode_0.Transition
    src=Leap
    via=AnimStateTransitionNode_0
   }
   {
    dst=Conduit
    rule=ItemFullBodyOverride.AnimGraphNode_StateMachine_0.FlyingStateMachine.AnimStateTransitionNode_20.Transition
    src=Leap
    via=AnimStateTransitionNode_20
   }
   {
    dst=BoostStart
    rule=ItemFullBodyOverride.AnimGraphNode_StateMachine_0.FlyingStateMachine.AnimStateTransitionNode_3.Transition
    src=ToBoosting
    via=AnimStateTransitionNode_3
   }
   {
    dst=Dive Start
    rule=ItemFullBodyOverride.AnimGraphNode_StateMachine_0.FlyingStateMachine.AnimStateTransitionNode_2.Transition
    src=ToDive
    via=AnimStateTransitionNode_2
   }
   {
    dst=DodgeLeftStart
    rule=ItemFullBodyOverride.AnimGraphNode_StateMachine_0.FlyingStateMachine.AnimStateTransitionNode_11.Transition
    src=ToDodge
    via=AnimStateTransitionNode_11
   }
   {
    dst=DodgeRightStart
    rule=ItemFullBodyOverride.AnimGraphNode_StateMachine_0.FlyingStateMachine.AnimStateTransitionNode_10.Transition
    src=ToDodge
    via=AnimStateTransitionNode_10
   }
  ]
 }
 Conduit {
  path=ItemFullBodyOverride.AnimGraphNode_StateMachine_0.FlyingStateMachine.AnimStateConduitNode_0.Conduit
  nodes [Result(bCanEnterTransition=true)]
 }
 Leap {
  path=ItemFullBodyOverride.AnimGraphNode_StateMachine_0.FlyingStateMachine.AnimStateNode_0.Leap
  nodes [Output Animation Pose(Result=Sequence Player 'Olympus_Player_Takeoff'() {bLoopAnimation=false, sequence=/SunRoseFlyingGameplay/Items/Wings/Animations/Player/Olympus_Player_Takeoff})]
 }
 DodgeRightStart {
  path=ItemFullBodyOverride.AnimGraphNode_StateMachine_0.FlyingStateMachine.AnimStateNode_1.DodgeRightStart
  nodes [Output Animation Pose(Result=Sequence Player 'Olympus_Player_Dodge_Right'() {bLoopAnimation=false, sequence=/SunRoseFlyingGameplay/Items/Wings/Animations/Player/Olympus_Player_Dodge_Right})]
 }
 DodgeLeftEnd {
  path=ItemFullBodyOverride.AnimGraphNode_StateMachine_0.FlyingStateMachine.AnimStateNode_10.DodgeLeftEnd
  nodes [Output Animation Pose(Result=Sequence Player 'Olympus_Player_Dodge_Left_Recover'() {bLoopAnimation=false, sequence=/SunRoseFlyingGameplay/Items/Wings/Animations/Player/Olympus_Player_Dodge_Left_Recover})]
 }
 DiveLoop {
  path=ItemFullBodyOverride.AnimGraphNode_StateMachine_0.FlyingStateMachine.AnimStateNode_2.DiveLoop
  nodes [
   Modify Curve!(SourcePose, CurveValues_0=1, CurveValues_1=1, Alpha=1) -> (Pose) {curveNames [Roll,Yaw], curveValues [0,0]}
   Output Animation Pose(Result=Sequence Player 'Olympus_Player_Dive_Loop'() {sequence=/SunRoseFlyingGameplay/Items/Wings/Animations/Player/Olympus_Player_Dive_Loop})
  ]
 }
 Dive Start {
  path=ItemFullBodyOverride.AnimGraphNode_StateMachine_0.FlyingStateMachine.AnimStateNode_3.Dive Start
  nodes [
   Modify Curve(CurveValues_0=1, CurveValues_1=1, Alpha=1) {curveNames [Roll,Yaw], curveValues [0,0]}
   Output Animation Pose(Result=Blend Poses by bool(
     BlendPose_0=Sequence Player 'Olympus_Player_Dive_Start_HitReact'() {bLoopAnimation=false, sequence=/SunRoseFlyingGameplay/Items/Wings/Animations/Player/Olympus_Player_Dive_Start_HitReact}
     BlendPose_1=Sequence Player 'Olympus_Player_Dive'() {bLoopAnimation=false, sequence=/SunRoseFlyingGameplay/Items/Wings/Animations/Player/Olympus_Player_Dive}
     bActiveValue=Get IsForceDive()
    ) {blendType=Linear, transitionType=Inertialization})
  ]
 }
 DodgeLeftStart {
  path=ItemFullBodyOverride.AnimGraphNode_StateMachine_0.FlyingStateMachine.AnimStateNode_4.DodgeLeftStart
  nodes [Output Animation Pose(Result=Sequence Player 'Olympus_Player_Dodge_Left'() {bLoopAnimation=false, sequence=/SunRoseFlyingGameplay/Items/Wings/Animations/Player/Olympus_Player_Dodge_Left})]
 }
 BoostStart {
  path=ItemFullBodyOverride.AnimGraphNode_StateMachine_0.FlyingStateMachine.AnimStateNode_5.BoostStart
  nodes [Output Animation Pose(Result=Sequence Player 'Olympus_Player_Boost_start'() {bLoopAnimation=false, sequence=/SunRoseFlyingGameplay/Items/Wings/Animations/Player/Olympus_Player_Boost_start})]
 }
 BoostLoop {
  path=ItemFullBodyOverride.AnimGraphNode_StateMachine_0.FlyingStateMachine.AnimStateNode_6.BoostLoop
  nodes [Output Animation Pose(Result=Sequence Player 'Olympus_Player_Boost_Loop'() {sequence=/SunRoseFlyingGameplay/Items/Wings/Animations/Player/Olympus_Player_Boost_Loop})]
 }
 Gliding {
  path=ItemFullBodyOverride.AnimGraphNode_StateMachine_0.FlyingStateMachine.AnimStateNode_7.Gliding
  nodes [
   Apply Additive(
    Additive="Sequence Player (additive) 'Olympus_Player_Glide_Slow'"() {bLoopAnimation=false, sequence=/SunRoseFlyingGameplay/Items/Wings/Animations/Player/Olympus_Player_Glide_Slow}
    Alpha=Get SlowSpeedAlpha() -> (SlowSpeedAlpha:float)
   ) {alphaBoolBlend {blendOption=Linear}}
   n1 = Blendspace Player 'Olympus_Player_Glide_Turbulence_BS'!(X=Get Fuel()) -> (Pose) {blendSpace=/SunRoseFlyingGameplay/Items/Wings/Animations/Player/Olympus_Player_Glide_Turbulence_BS}
   n2 = Component To Local(ComponentPose="Transform (Modify) Bone - Bone: Root"!(
     ComponentPose=Local To Component(LocalPose=@n1)
     Alpha=1
     Translation_X:double=0
     Translation_Y:double=0
     Translation_Z:double=Get Pelvis() -> (Pelvis:double)
     Rotation="0,0,0"
     Scale="1,1,1"
    ) -> (Pose) {translationMode=BMM_Additive, alphaBoolBlend {blendOption=Linear}, boneToModify {boneName=Root}})
   n3 = Sequence Player 'Olympus_Player_Flap_Large'() {sequence=/SunRoseFlyingGameplay/Items/Wings/Animations/Player/Olympus_Player_Flap_Large}
   n4 = Get IsAllowWingFlap() -> (IsAllowWingFlap:bool)
   n5 = Get Fuel()
   n6 = Get Current Height Above World()
   n7 = Blend Poses by bool(
    BlendPose_0=Sequence Player 'Olympus_Player_NoFuel'() {sequence=/SunRoseFlyingGameplay/Items/Wings/Animations/Player/Olympus_Player_NoFuel}
    BlendPose_1=Two Way Blend(
     A=Blend Poses by bool(BlendPose_0=@n3, BlendPose_1=@n2, BlendTime_0=0.5, BlendTime_1=0.8, bActiveValue=@n4) {blendType=Sinusoidal, transitionType=Inertialization}
     B=Sequence Player 'Olympus_Player_Glide_Idle_High_15'() {bLoopAnimation=false, sequence=/SunRoseFlyingGameplay/Items/Wings/Animations/Player/Olympus_Player_Glide_Idle_High_15}
     Alpha=Get GlidePitchAlpha() -> (GlidePitchAlpha:double)
    ) {blendNode {alphaBoolBlend {blendOption=Linear}}}
    BlendTime_0=0.4
    BlendTime_1=0.5
    bActiveValue=AND Boolean!(A="Equal (Float)"(A=@n5), B="float >= float"(A=@n6, B=200)) -> (ReturnValue)
   ) {blendType=Sinusoidal, transitionType=Inertialization}
   n8 = Sequence Player 'Olympus_Player_Ascend_Loop'() {sequence=/SunRoseFlyingGameplay/Items/Wings/Animations/Player/Olympus_Player_Ascend_Loop}
   n9 = Get IsAscending()
   n10 = Get GlideRoll()
   Output Animation Pose(Result=Apply Additive(
     Base=Blend Poses by bool(BlendPose_0=@n8, BlendPose_1=@n7, BlendTime_0=0.3, BlendTime_1=0.5, bActiveValue=@n9) {blendType=Linear, transitionType=Inertialization}
     Additive=Blendspace Player 'Olympus_Player_Land_BS'!(X=@n10) -> (Pose) {blendSpace=/SunRoseFlyingGameplay/Items/Wings/Animations/Player/Olympus_Player_Land_BS}
     Alpha=Get SlownDownOrLandAlpha()
    ) {alphaBoolBlend {blendOption=Linear}})
  ]
 }
 BoostEnd {
  path=ItemFullBodyOverride.AnimGraphNode_StateMachine_0.FlyingStateMachine.AnimStateNode_8.BoostEnd
  nodes [Output Animation Pose(Result=Sequence Player 'Olympus_Player_Boost_End'() {bLoopAnimation=false, sequence=/SunRoseFlyingGameplay/Items/Wings/Animations/Player/Olympus_Player_Boost_End})]
 }
 DodgeRightEnd {
  path=ItemFullBodyOverride.AnimGraphNode_StateMachine_0.FlyingStateMachine.AnimStateNode_9.DodgeRightEnd
  nodes [Output Animation Pose(Result=Sequence Player 'Olympus_Player_Dodge_Right_Recover'() {bLoopAnimation=false, sequence=/SunRoseFlyingGameplay/Items/Wings/Animations/Player/Olympus_Player_Dodge_Right_Recover})]
 }
 Transition {path=ItemFullBodyOverride.AnimGraphNode_StateMachine_0.FlyingStateMachine.AnimStateTransitionNode_0.Transition, nodes [Result()]}
 Transition {
  path=ItemFullBodyOverride.AnimGraphNode_StateMachine_0.FlyingStateMachine.AnimStateTransitionNode_1.Transition
  nodes [Result(bCanEnterTransition=Get IsDodgeRight())]
 }
 Transition {
  path=ItemFullBodyOverride.AnimGraphNode_StateMachine_0.FlyingStateMachine.AnimStateTransitionNode_10.Transition
  nodes [Result(bCanEnterTransition=Get IsDodgeRight())]
 }
 Transition {
  path=ItemFullBodyOverride.AnimGraphNode_StateMachine_0.FlyingStateMachine.AnimStateTransitionNode_11.Transition
  nodes [Result(bCanEnterTransition=Get IsDodgeLeft())]
 }
 Transition {
  path=ItemFullBodyOverride.AnimGraphNode_StateMachine_0.FlyingStateMachine.AnimStateTransitionNode_12.Transition
  nodes [Result(bCanEnterTransition=true)]
 }
 Transition {
  path=ItemFullBodyOverride.AnimGraphNode_StateMachine_0.FlyingStateMachine.AnimStateTransitionNode_13.Transition
  nodes [Result(bCanEnterTransition=true)]
 }
 Transition {
  path=ItemFullBodyOverride.AnimGraphNode_StateMachine_0.FlyingStateMachine.AnimStateTransitionNode_14.Transition
  nodes [Result(bCanEnterTransition=Get IsBoostStartEnd() -> (IsBoostStartEnd:bool))]
 }
 Transition {
  path=ItemFullBodyOverride.AnimGraphNode_StateMachine_0.FlyingStateMachine.AnimStateTransitionNode_15.Transition
  nodes [Result(bCanEnterTransition=NOT Boolean(A=Get IsWingsBoosting()))]
 }
 Transition {path=ItemFullBodyOverride.AnimGraphNode_StateMachine_0.FlyingStateMachine.AnimStateTransitionNode_18.Transition, nodes [Result()]}
 Transition {
  path=ItemFullBodyOverride.AnimGraphNode_StateMachine_0.FlyingStateMachine.AnimStateTransitionNode_19.Transition
  nodes [Result(bCanEnterTransition=NOT Boolean(A=Get IsWingsCrashing()))]
 }
 Transition {
  path=ItemFullBodyOverride.AnimGraphNode_StateMachine_0.FlyingStateMachine.AnimStateTransitionNode_2.Transition
  nodes [Result(bCanEnterTransition=Get IsWingsCrashing())]
 }
 Transition {
  path=ItemFullBodyOverride.AnimGraphNode_StateMachine_0.FlyingStateMachine.AnimStateTransitionNode_20.Transition
  nodes [Result(bCanEnterTransition=Get IsLeapInterruptible() -> (IsLeapInterruptible:bool))]
 }
 Transition {
  path=ItemFullBodyOverride.AnimGraphNode_StateMachine_0.FlyingStateMachine.AnimStateTransitionNode_21.Transition
  nodes [Result(bCanEnterTransition=NOT Boolean(A=Get IsWingsBoosting()))]
 }
 Transition {path=ItemFullBodyOverride.AnimGraphNode_StateMachine_0.FlyingStateMachine.AnimStateTransitionNode_22.Transition, nodes [Result()]}
 Transition {
  path=ItemFullBodyOverride.AnimGraphNode_StateMachine_0.FlyingStateMachine.AnimStateTransitionNode_23.Transition
  nodes [Result(bCanEnterTransition=Get IsExtBoostEarly() -> (IsExtBoostEarly:bool))]
 }
 Transition {path=ItemFullBodyOverride.AnimGraphNode_StateMachine_0.FlyingStateMachine.AnimStateTransitionNode_24.Transition, nodes [Result()]}
 Transition {
  path=ItemFullBodyOverride.AnimGraphNode_StateMachine_0.FlyingStateMachine.AnimStateTransitionNode_25.Transition
  nodes [Result(bCanEnterTransition=Get IsWingLeaping())]
 }
 Transition {path=ItemFullBodyOverride.AnimGraphNode_StateMachine_0.FlyingStateMachine.AnimStateTransitionNode_27.Transition, nodes [Result()]}
 Transition {path=ItemFullBodyOverride.AnimGraphNode_StateMachine_0.FlyingStateMachine.AnimStateTransitionNode_28.Transition, nodes [Result()]}
 Transition {
  path=ItemFullBodyOverride.AnimGraphNode_StateMachine_0.FlyingStateMachine.AnimStateTransitionNode_3.Transition
  nodes [Result(bCanEnterTransition=Get ToBoosting() -> (ToBoosting:bool))]
 }
 Transition {
  path=ItemFullBodyOverride.AnimGraphNode_StateMachine_0.FlyingStateMachine.AnimStateTransitionNode_4.Transition
  nodes [Result(bCanEnterTransition=Get IsDodgeLeft())]
 }
 Transition {
  path=ItemFullBodyOverride.AnimGraphNode_StateMachine_0.FlyingStateMachine.AnimStateTransitionNode_6.Transition
  nodes [Result(bCanEnterTransition=Get IsDiveStart() -> (IsDiveStart:bool))]
 }
 Transition {path=ItemFullBodyOverride.AnimGraphNode_StateMachine_0.FlyingStateMachine.AnimStateTransitionNode_7.Transition, nodes [Result()]}
 Transition {
  path=ItemFullBodyOverride.AnimGraphNode_StateMachine_0.FlyingStateMachine.AnimStateTransitionNode_8.Transition
  nodes [Result(bCanEnterTransition=NOT Boolean(A=Get IsWingsCrashing()))]
 }
 Transition {
  path=ItemFullBodyOverride.AnimGraphNode_StateMachine_0.FlyingStateMachine.AnimStateTransitionNode_9.Transition
  nodes [Result(bCanEnterTransition=Get IsWingsBoosting())]
 }
 GroundStateMachine {
  entry=Conduit
  path=ItemFullBodyOverride.AnimGraphNode_StateMachine_1.GroundStateMachine
  ids=id
  nodes [
   AnimStateConduitNode_0 = Conduit()
   AnimStateNode_0 = LandAdditve() {is=title AnimStateNode}
   AnimStateNode_1 = Ground Navigation() {is=title AnimStateNode}
   AnimStateNode_2 = DiveHitToIdle() {is=title AnimStateNode}
   AnimStateTransitionNode_0 = AnimStateTransitionNode() {crossfadeDuration=0.3, logicType=TLT_Inertialization}
   AnimStateTransitionNode_1 = AnimStateTransitionNode() {logicType=TLT_Inertialization, priorityOrder=100}
   AnimStateTransitionNode_2 = AnimStateTransitionNode() {automaticRuleTriggerTime=0, bAutomaticRuleBasedOnSequencePlayerInState=true, logicType=TLT_Inertialization}
   AnimStateTransitionNode_3 = AnimStateTransitionNode() {automaticRuleTriggerTime=0, logicType=TLT_Inertialization, priorityOrder=2}
   AnimStateTransitionNode_4 = AnimStateTransitionNode() {automaticRuleTriggerTime=0, bAutomaticRuleBasedOnSequencePlayerInState=true, crossfadeDuration=0.25, logicType=TLT_Inertialization}
   AnimStateTransitionNode_5 = AnimStateTransitionNode() {crossfadeDuration=0.25, logicType=TLT_Inertialization, priorityOrder=2}
  ]
  links [
   {
    dst=DiveHitToIdle
    rule=ItemFullBodyOverride.AnimGraphNode_StateMachine_1.GroundStateMachine.AnimStateTransitionNode_3.Transition
    src=Conduit
    via=AnimStateTransitionNode_3
   }
   {
    dst=Ground Navigation
    rule=ItemFullBodyOverride.AnimGraphNode_StateMachine_1.GroundStateMachine.AnimStateTransitionNode_1.Transition
    src=Conduit
    via=AnimStateTransitionNode_1
   }
   {
    dst=LandAdditve
    rule=ItemFullBodyOverride.AnimGraphNode_StateMachine_1.GroundStateMachine.AnimStateTransitionNode_0.Transition
    src=Conduit
    via=AnimStateTransitionNode_0
   }
   {
    dst=Ground Navigation
    rule=ItemFullBodyOverride.AnimGraphNode_StateMachine_1.GroundStateMachine.AnimStateTransitionNode_4.Transition
    src=DiveHitToIdle
    via=AnimStateTransitionNode_4
   }
   {
    dst=Ground Navigation
    rule=ItemFullBodyOverride.AnimGraphNode_StateMachine_1.GroundStateMachine.AnimStateTransitionNode_5.Transition
    src=DiveHitToIdle
    via=AnimStateTransitionNode_5
   }
   {
    dst=Ground Navigation
    rule=ItemFullBodyOverride.AnimGraphNode_StateMachine_1.GroundStateMachine.AnimStateTransitionNode_2.Transition
    src=LandAdditve
    via=AnimStateTransitionNode_2
   }
  ]
 }
 Conduit {
  path=ItemFullBodyOverride.AnimGraphNode_StateMachine_1.GroundStateMachine.AnimStateConduitNode_0.Conduit
  nodes [Result(bCanEnterTransition=true)]
 }
 LandAdditve {
  path=ItemFullBodyOverride.AnimGraphNode_StateMachine_1.GroundStateMachine.AnimStateNode_0.LandAdditve
  nodes [
   Layered blend per bone() {
    blendMode=BlendMask
    curveBlendOption=UseMaxValue
    blendMasks [&Fortnite_M_Avg_Player_Skeleton:FirstPersonAdditives]
    layerSetup []
   }
   Output Animation Pose(Result=Apply Additive(
     Base=Use cached pose 'SavedPose9831'()
     Additive="Sequence Player (additive) 'Olympus_Player_LandingAdditive'"() {bLoopAnimation=false, sequence=/SunRoseFlyingGameplay/Items/Wings/Animations/Player/Olympus_Player_LandingAdditive}
    ) {alphaBoolBlend {blendOption=Linear}})
  ]
 }
 Ground Navigation {
  path=ItemFullBodyOverride.AnimGraphNode_StateMachine_1.GroundStateMachine.AnimStateNode_1.Ground Navigation
  nodes [Output Animation Pose(Result=Use cached pose 'SavedPose9831'())]
 }
 DiveHitToIdle {
  path=ItemFullBodyOverride.AnimGraphNode_StateMachine_1.GroundStateMachine.AnimStateNode_2.DiveHitToIdle
  nodes [Output Animation Pose(Result=Sequence Player 'Olympus_Player_Dive_Hit_To_Idle'() {bLoopAnimation=false, sequence=/SunRoseFlyingGameplay/Items/Wings/Animations/Player/Olympus_Player_Dive_Hit_To_Idle})]
 }
 Transition {
  path=ItemFullBodyOverride.AnimGraphNode_StateMachine_1.GroundStateMachine.AnimStateTransitionNode_0.Transition
  nodes [Result(bCanEnterTransition=Get IsPlayLandAdditive() -> (IsPlayLandAdditive:bool))]
 }
 Transition {
  path=ItemFullBodyOverride.AnimGraphNode_StateMachine_1.GroundStateMachine.AnimStateTransitionNode_1.Transition
  nodes [Result(bCanEnterTransition=true)]
 }
 Transition {path=ItemFullBodyOverride.AnimGraphNode_StateMachine_1.GroundStateMachine.AnimStateTransitionNode_2.Transition, nodes [Result()]}
 Transition {
  path=ItemFullBodyOverride.AnimGraphNode_StateMachine_1.GroundStateMachine.AnimStateTransitionNode_3.Transition
  nodes [Result(bCanEnterTransition=Get isPlayDiveHitToIdle() -> (isPlayDiveHitToIdle:bool))]
 }
 Transition {path=ItemFullBodyOverride.AnimGraphNode_StateMachine_1.GroundStateMachine.AnimStateTransitionNode_4.Transition, nodes [Result()]}
 Transition {
  path=ItemFullBodyOverride.AnimGraphNode_StateMachine_1.GroundStateMachine.AnimStateTransitionNode_5.Transition
  nodes [Result(bCanEnterTransition=Get bIsAccelerating2D())]
 }
 ItemIdleOverride {nodes [
   Input Pose() {name=InPose_Idle}
   Output Pose() {layerGroup=Item Override}
  ]}
 ItemJetPackJumpAdditive {nodes [Output Pose() {layerGroup=Item Override}]}
 ItemJetPackStartAdditive {nodes [Output Pose() {layerGroup=Item Override}]}
 ItemJumpLoopAdditive {nodes [
   Input Pose() {name=InPoseItemJumpLoopAdditive}
   Output Pose() {layerGroup=Item Override}
  ]}
 ItemJumpOverride {nodes [
   Input Pose() {name=InPoseJump}
   Output Pose() {layerGroup=Item Override}
  ]}
 ItemJumpUpAdditive {nodes [
   Input Pose() {name=InPoseItemJumpUpAdditive}
   Output Pose() {layerGroup=Item Override}
  ]}
 ItemPreAimOffsets {nodes [
   Input Pose() {name=InPosePreAimOffsets}
   Output Pose() {layerGroup=Item Override}
  ]}
 ItemPreIK {nodes [
   Input Pose() {name=InPose_PreIK}
   Output Pose() {layerGroup=Item Override}
  ]}
 ItemRightHandIK {nodes [
   Input Pose() {name=InPosePreRightHandIK}
   Input Pose() {name=InPosePostRightHandIK}
   Output Pose() {layerGroup=Item Override}
  ]}
 ItemSkeletalControl {nodes [
   Input Pose()
   Output Pose() {layerGroup=Item Override}
  ]}
 ItemSwimJumpFallLoop {nodes [Output Pose() {layerGroup=Item Override}]}
 ItemSwimJumpSurfaceEnd {nodes [Output Pose() {layerGroup=Item Override}]}
 ItemSwimJumpSurfaceLoop {nodes [Output Pose() {layerGroup=Item Override}]}
 ItemUpperBody {nodes [
   Input Pose!() -> (Pose, InputParam:Fort Anim Input Adjusted Aim Structure) {name=InPoseUpperBody}
   Output Pose() {layerGroup=Item Override}
  ]}
 ItemZipLineStartAdditive {nodes [Output Pose() {layerGroup=Item Override}]}
]
variables [
 Current Height Above World {type=number}
 Delta Time X {type=number}
 FlyingPawnComponent {type=FlyingPawnComponent_C, default=null}
 Fuel {type=number}
 GlidePitch {type=number}
 GlidePitchAlpha {type=number}
 GlidePitchInterp {type=number}
 GlidePitchTarget {type=number}
 GlideRoll {type=number}
 GlideRollCombined {type=number}
 GlideRollCombinedTarget {type=number}
 GlideRollInterp {type=number}
 GroundBlendTime {type=number}
 IsAllowWingFlap {type=boolean}
 IsAscending {type=boolean}
 IsBoostStartEnd {type=boolean}
 IsDiveStart {type=boolean}
 IsDodgeLeft {type=boolean}
 IsDodgeRight {type=boolean}
 IsExtBoostEarly {type=boolean}
 IsFlyingStart {type=boolean}
 IsForceDive {type=boolean}
 IsLeapInterruptible {type=boolean}
 IsOnAscender {type=boolean}
 IsOnGrindRail {type=boolean}
 isPlayDiveHitToIdle {type=boolean}
 IsPlayLandAdditive {type=boolean}
 IsWingLeaping {type=boolean}
 isWingReplenish {type=boolean}
 IsWingsBoosting {type=boolean}
 IsWingsCrashing {type=boolean}
 IsWingsFlying {type=boolean}
 "Pawn Rotation Rate Z (Yaw)" {type=number}
 Pelvis {type=number}
 PitchAlpha {type=number}
 RandomInteger {type=integer}
 RollAlpha {type=number}
 SlowDownOrLandTarget {type=number}
 SlownDownOrLandAlpha {type=number}
 SlowSpeedAlpha {type=number}
 Tag {type=GameplayTag, default {tagName=GameplayCue.Athena.SunRose.Wings.Replenish}}
 ToBoosting {type=boolean}
 YawAlpha {type=number}
 YawBlend {type=number}
]
