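# Physics asset for the SunRose wings gadget mesh in the flying gameplay plugin, defining simple collision bodies for the handle and both shoulders.
# A very small capsule sits on the handle, with larger capsules offset in opposite directions on the left and right shoulder bones.
# Each shoulder is constrained to the handle with both swing axes set free, so their stored swing limit angles do not restrict motion.
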PhysicsAsset /SunRoseFlyingGameplay/Items/Wings/Meshes/Wings_SunRose_Gadget_Physics1
bodies {
 handle [{shapeType=Capsule, length=0.505, radius=0.505, shapeName=handle_capsule, rotation=<rot 0 -90 0>}]
 shoulder_l [{shapeType=Capsule, length=5, radius=5, shapeName=shoulder_l_capsule, center=<0 0 -5>, rotation=<rot 0 -90 0>}]
 shoulder_r [{shapeType=Capsule, length=5, radius=5, shapeName=shoulder_r_capsule, center=<0 0 5>, rotation=<rot 0 -90 0>}]
}
constraints [
 {
  bone1Name=shoulder_l
  bone2Name=handle
  swing1LimitDegrees=20
  swing1Motion=Free
  swing2LimitDegrees=20
  swing2Motion=Free
  twistLimitDegrees=0
 }
 {
  bone1Name=shoulder_r
  bone2Name=handle
  swing1LimitDegrees=20
  swing1Motion=Free
  swing2LimitDegrees=20
  swing2Motion=Free
  twistLimitDegrees=0
 }
]
