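# Animation Blueprint for the SunRose Wings gadget, posing the wings to match the player's ground movement and flight.
# A ground state machine plus an additive machine form a cached pose, run through a linked procedural AnimBP with an optional LOD2 handle-scale fix, or used directly when procedural is deactivated.
# A conduit-routed air state machine covers leap, glide, boost, dive and dodge with additive roll, switched in by the flying flag, then passed through a slot and Dead Blending.
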AnimBlueprint /SunRoseFlyingGameplay/Items/Wings/Blueprints/SunRose_Wings_AnimBP
target_skeleton=/SunRoseFlyingGameplay/Items/Wings/Meshes/Wings_SunRose_Gadget_Skeleton
graphs [
 AnimGraph {nodes [
   n1 = SunRose_Wings_Procedural_AnimBP!(InPose:Pose Link Structure=Use cached pose 'WingsGroundPose'()) -> (Pose:Pose Link Structure) {
    is=title AnimGraphNode_LinkedAnimGraph
    bPropagateNotifiesToLinkedInstances=true
    bReceiveNotifiesFromLinkedInstances=true
    instanceClass=&/SunRose_Wings_Procedural_AnimBP.SunRose_Wings_Procedural_AnimBP_C
   }
   n2 = "Transform (Modify) Bone - Bone: handle"!(
    ComponentPose=Local To Component(LocalPose=@n1)
    bAlphaBoolEnabled:bool=Get bLOD2Fix() -> (bLOD2Fix:bool)
    Translation="0,0,0"
    Rotation="0,0,0"
    Scale="1.2,1.2,1.2"
   ) -> (Pose) {alphaInputType=Bool, scaleMode=BMM_Additive, alphaBoolBlend {blendOption=Linear}, boneToModify {boneName=handle}, scale=<1.2 1.2 1.2>}
   n3 = Blendspace Player 'Olympus_Wings_Glide_Roll_BS'!(X=Get PlayerGlideRollCombined() -> (PlayerGlideRollCombined:float)) -> (Pose) {blendSpace=/SunRoseFlyingGameplay/Items/Wings/Animations/Wings/Olympus_Wings_Glide_Roll_BS}
   AnimGraphNode_StateMachine_1 = WingsAirStatemachine() {is=title AnimGraphNode_StateMachine, bAllowConduitEntryStates=true}
   n4 = Get PlayerRollAlpha() -> (PlayerRollAlpha:float)
   n5 = Blend Poses by bool(
    BlendPose_0=Modify Curve(
     SourcePose=Apply Additive(Base=@AnimGraphNode_StateMachine_1, Additive=@n3, Alpha=@n4) {alphaBoolBlend {blendOption=Linear}}
     CurveValues_0=1
     Alpha=1
    ) {curveNames [Hurdle_Land], curveValues [0]}
    BlendPose_1=Blend Poses by bool(
     BlendPose_0=Use cached pose 'WingsGroundPose'()
     BlendPose_1=Component To Local(ComponentPose=@n2)
     bActiveValue=Get bIsProceduralDeactivated() -> (bIsProceduralDeactivated:bool)
    ) {blendType=Linear, childUpateMode=ResetChildOnActivate, transitionType=Inertialization}
    BlendTime_1=Get PlayerGroundBlendTime() -> (PlayerGroundBlendTime:float)
    bActiveValue=Get bIsFlyingStart() -> (bIsFlyingStart:bool)
   ) {blendType=Linear, childUpateMode=ResetChildOnActivate, transitionType=Inertialization}
   Output Pose(Result=Dead Blending!(Source:Pose Link Structure=Slot 'WingsAir'(Source=@n5) {slotName=WingsAir}) -> (Pose:Pose Link Structure) {is=title AnimGraphNode_DeadBlending, defaultBlendDuration=0.1, defaultBlendMode=Linear}) {layerGroup=None}
   AnimGraphNode_StateMachine_0 = WingsGroundStatemachine() {is=title AnimGraphNode_StateMachine, bAllowConduitEntryStates=true}
   AnimGraphNode_StateMachine_2 = WingsGroundAdditive() {is=title AnimGraphNode_StateMachine}
   Save cached pose 'WingsGroundPose'(Pose=Slot 'DefaultSlot'(Source=Apply Additive(Base=@AnimGraphNode_StateMachine_0, Additive=@AnimGraphNode_StateMachine_2) {alphaBoolBlend {blendOption=Linear}})) {cacheName=WingsGroundPose}
  ]}
 WingsGroundStatemachine {
  entry=Conduit
  path=AnimGraph.AnimGraphNode_StateMachine_0.WingsGroundStatemachine
  ids=id
  nodes [
   FromTacticalJump() {is=title AnimStateAliasNode, also {aliases [TacticalJumpLoop, TacticalJumpStart]}}
   ToSliding() {is=title AnimStateAliasNode, also {aliases [JumpLand, Moving, TacticalLand, TacticalSprint]}}
   ToZipline() {
    is=title AnimStateAliasNode
    also {aliases [FallLoop, FallStart, JumpLand, JumpStart, TacticalJumpLoop, TacticalJumpStart, TacticalLand, TacticalSprint]}
   }
   AnimStateConduitNode_0 = Conduit()
   AnimStateNode_0 = Swim() {is=title AnimStateNode}
   AnimStateNode_1 = Idle()
   AnimStateNode_10 = TacticalLand() {is=title AnimStateNode}
   AnimStateNode_11 = SlidingPose() {
    is=title AnimStateNode
    stateEntered {guid=#1}
    stateFullyBlended {guid=#2}
    stateLeft {guid=#3}
   }
   AnimStateNode_12 = CrouchIdle() {is=title AnimStateNode}
   AnimStateNode_13 = Crouch Moving() {is=title AnimStateNode}
   AnimStateNode_14 = SlideStart() {
    is=title AnimStateNode
    stateEntered {guid=#4}
    stateFullyBlended {guid=#5}
    stateLeft {guid=#6}
   }
   AnimStateNode_15 = Zipline() {is=title AnimStateNode}
   AnimStateNode_16 = DiveHitToIdle() {is=title AnimStateNode}
   AnimStateNode_17 = GrindRail() {is=title AnimStateNode}
   AnimStateNode_2 = JumpStart() {is=title AnimStateNode}
   AnimStateNode_3 = FallLoop() {is=title AnimStateNode}
   AnimStateNode_4 = JumpLand() {is=title AnimStateNode}
   AnimStateNode_5 = Moving() {is=title AnimStateNode}
   AnimStateNode_6 = FallStart() {is=title AnimStateNode}
   AnimStateNode_7 = TacticalSprint() {is=title AnimStateNode}
   AnimStateNode_8 = TacticalJumpStart() {is=title AnimStateNode}
   AnimStateNode_9 = TacticalJumpLoop() {is=title AnimStateNode}
   AnimStateTransitionNode_0 = AnimStateTransitionNode()
   AnimStateTransitionNode_1 = AnimStateTransitionNode()
   AnimStateTransitionNode_10 = AnimStateTransitionNode() {logicType=TLT_Inertialization}
   AnimStateTransitionNode_11 = AnimStateTransitionNode() {automaticRuleTriggerTime=0, bAutomaticRuleBasedOnSequencePlayerInState=true, crossfadeDuration=0.8, logicType=TLT_Inertialization}
   AnimStateTransitionNode_12 = AnimStateTransitionNode() {crossfadeDuration=0.1, logicType=TLT_Inertialization}
   AnimStateTransitionNode_13 = AnimStateTransitionNode() {automaticRuleTriggerTime=0, logicType=TLT_Inertialization}
   AnimStateTransitionNode_14 = AnimStateTransitionNode() {automaticRuleTriggerTime=0, bAutomaticRuleBasedOnSequencePlayerInState=true, logicType=TLT_Inertialization}
   AnimStateTransitionNode_15 = AnimStateTransitionNode()
   AnimStateTransitionNode_16 = AnimStateTransitionNode()
   AnimStateTransitionNode_17 = AnimStateTransitionNode() {logicType=TLT_Inertialization, priorityOrder=8}
   AnimStateTransitionNode_18 = AnimStateTransitionNode()
   AnimStateTransitionNode_19 = AnimStateTransitionNode() {logicType=TLT_Inertialization}
   AnimStateTransitionNode_2 = AnimStateTransitionNode() {crossfadeDuration=0.4, logicType=TLT_Inertialization, priorityOrder=100}
   AnimStateTransitionNode_20 = AnimStateTransitionNode() {automaticRuleTriggerTime=0, bAutomaticRuleBasedOnSequencePlayerInState=true, logicType=TLT_Inertialization}
   AnimStateTransitionNode_21 = AnimStateTransitionNode()
   AnimStateTransitionNode_22 = AnimStateTransitionNode() {priorityOrder=2}
   AnimStateTransitionNode_23 = AnimStateTransitionNode() {crossfadeDuration=0.4, logicType=TLT_Inertialization, priorityOrder=20}
   AnimStateTransitionNode_24 = AnimStateTransitionNode() {crossfadeDuration=0.1, logicType=TLT_Inertialization}
   AnimStateTransitionNode_25 = AnimStateTransitionNode() {crossfadeDuration=0.25, logicType=TLT_Inertialization, priorityOrder=2}
   AnimStateTransitionNode_26 = AnimStateTransitionNode() {crossfadeDuration=0.4, logicType=TLT_Inertialization, priorityOrder=4}
   AnimStateTransitionNode_27 = AnimStateTransitionNode() {crossfadeDuration=0.4, logicType=TLT_Inertialization, priorityOrder=5}
   AnimStateTransitionNode_28 = AnimStateTransitionNode() {crossfadeDuration=0.4, logicType=TLT_Inertialization}
   AnimStateTransitionNode_29 = AnimStateTransitionNode() {crossfadeDuration=0.4, logicType=TLT_Inertialization}
   AnimStateTransitionNode_3 = AnimStateTransitionNode() {automaticRuleTriggerTime=0, bAutomaticRuleBasedOnSequencePlayerInState=true, crossfadeDuration=0.3, logicType=TLT_Inertialization}
   AnimStateTransitionNode_30 = AnimStateTransitionNode()
   AnimStateTransitionNode_31 = AnimStateTransitionNode()
   AnimStateTransitionNode_32 = AnimStateTransitionNode() {automaticRuleTriggerTime=0, bAutomaticRuleBasedOnSequencePlayerInState=true}
   AnimStateTransitionNode_33 = AnimStateTransitionNode() {logicType=TLT_Inertialization, priorityOrder=10}
   AnimStateTransitionNode_34 = AnimStateTransitionNode() {
    bAutomaticRuleBasedOnSequencePlayerInState=true
    logicType=TLT_Inertialization
    transitionEnd {guid=#7}
    transitionInterrupt {guid=#8}
    transitionStart {guid=#9}
   }
   AnimStateTransitionNode_36 = AnimStateTransitionNode() {crossfadeDuration=0.4, logicType=TLT_Inertialization, priorityOrder=10}
   AnimStateTransitionNode_37 = AnimStateTransitionNode() {logicType=TLT_Inertialization}
   AnimStateTransitionNode_38 = AnimStateTransitionNode()
   AnimStateTransitionNode_39 = AnimStateTransitionNode()
   AnimStateTransitionNode_4 = AnimStateTransitionNode() {logicType=TLT_Inertialization, priorityOrder=9}
   AnimStateTransitionNode_40 = AnimStateTransitionNode() {logicType=TLT_Inertialization, priorityOrder=2}
   AnimStateTransitionNode_41 = AnimStateTransitionNode()
   AnimStateTransitionNode_42 = AnimStateTransitionNode() {logicType=TLT_Inertialization}
   AnimStateTransitionNode_43 = AnimStateTransitionNode()
   AnimStateTransitionNode_5 = AnimStateTransitionNode() {logicType=TLT_Inertialization, priorityOrder=3}
   AnimStateTransitionNode_6 = AnimStateTransitionNode() {logicType=TLT_Inertialization}
   AnimStateTransitionNode_7 = AnimStateTransitionNode() {logicType=TLT_Inertialization, priorityOrder=6}
   AnimStateTransitionNode_8 = AnimStateTransitionNode() {logicType=TLT_Inertialization, priorityOrder=7}
   AnimStateTransitionNode_9 = AnimStateTransitionNode() {automaticRuleTriggerTime=0, bAutomaticRuleBasedOnSequencePlayerInState=true}
  ]
  links [
   {
    dst=Crouch Moving
    rule=AnimGraph.AnimGraphNode_StateMachine_0.WingsGroundStatemachine.AnimStateTransitionNode_27.Transition
    src=Conduit
    via=AnimStateTransitionNode_27
   }
   {
    dst=CrouchIdle
    rule=AnimGraph.AnimGraphNode_StateMachine_0.WingsGroundStatemachine.AnimStateTransitionNode_26.Transition
    src=Conduit
    via=AnimStateTransitionNode_26
   }
   {
    dst=DiveHitToIdle
    rule=AnimGraph.AnimGraphNode_StateMachine_0.WingsGroundStatemachine.AnimStateTransitionNode_17.Transition
    src=Conduit
    via=AnimStateTransitionNode_17
   }
   {
    dst=GrindRail
    rule=AnimGraph.AnimGraphNode_StateMachine_0.WingsGroundStatemachine.AnimStateTransitionNode_36.Transition
    src=Conduit
    via=AnimStateTransitionNode_36
   }
   {
    dst=Idle
    rule=AnimGraph.AnimGraphNode_StateMachine_0.WingsGroundStatemachine.AnimStateTransitionNode_2.Transition
    src=Conduit
    via=AnimStateTransitionNode_2
   }
   {
    dst=JumpStart
    rule=AnimGraph.AnimGraphNode_StateMachine_0.WingsGroundStatemachine.AnimStateTransitionNode_8.Transition
    src=Conduit
    via=AnimStateTransitionNode_8
   }
   {
    dst=Moving
    rule=AnimGraph.AnimGraphNode_StateMachine_0.WingsGroundStatemachine.AnimStateTransitionNode_7.Transition
    src=Conduit
    via=AnimStateTransitionNode_7
   }
   {
    dst=SlideStart
    rule=AnimGraph.AnimGraphNode_StateMachine_0.WingsGroundStatemachine.AnimStateTransitionNode_37.Transition
    src=Conduit
    via=AnimStateTransitionNode_37
   }
   {
    dst=Swim
    rule=AnimGraph.AnimGraphNode_StateMachine_0.WingsGroundStatemachine.AnimStateTransitionNode_5.Transition
    src=Conduit
    via=AnimStateTransitionNode_5
   }
   {
    dst=TacticalSprint
    rule=AnimGraph.AnimGraphNode_StateMachine_0.WingsGroundStatemachine.AnimStateTransitionNode_4.Transition
    src=Conduit
    via=AnimStateTransitionNode_4
   }
   {
    dst=Zipline
    rule=AnimGraph.AnimGraphNode_StateMachine_0.WingsGroundStatemachine.AnimStateTransitionNode_40.Transition
    src=Conduit
    via=AnimStateTransitionNode_40
   }
   {
    dst=Conduit
    rule=AnimGraph.AnimGraphNode_StateMachine_0.WingsGroundStatemachine.AnimStateTransitionNode_31.Transition
    src=Crouch Moving
    via=AnimStateTransitionNode_31
   }
   {
    dst=CrouchIdle
    rule=AnimGraph.AnimGraphNode_StateMachine_0.WingsGroundStatemachine.AnimStateTransitionNode_29.Transition
    src=Crouch Moving
    via=AnimStateTransitionNode_29
   }
   {
    dst=Conduit
    rule=AnimGraph.AnimGraphNode_StateMachine_0.WingsGroundStatemachine.AnimStateTransitionNode_30.Transition
    src=CrouchIdle
    via=AnimStateTransitionNode_30
   }
   {
    dst=Crouch Moving
    rule=AnimGraph.AnimGraphNode_StateMachine_0.WingsGroundStatemachine.AnimStateTransitionNode_28.Transition
    src=CrouchIdle
    via=AnimStateTransitionNode_28
   }
   {
    dst=Conduit
    rule=AnimGraph.AnimGraphNode_StateMachine_0.WingsGroundStatemachine.AnimStateTransitionNode_32.Transition
    src=DiveHitToIdle
    via=AnimStateTransitionNode_32
   }
   {
    dst=JumpLand
    rule=AnimGraph.AnimGraphNode_StateMachine_0.WingsGroundStatemachine.AnimStateTransitionNode_12.Transition
    src=FallLoop
    via=AnimStateTransitionNode_12
   }
   {
    dst=FallLoop
    rule=AnimGraph.AnimGraphNode_StateMachine_0.WingsGroundStatemachine.AnimStateTransitionNode_11.Transition
    src=FallStart
    via=AnimStateTransitionNode_11
   }
   {
    dst=JumpLand
    rule=AnimGraph.AnimGraphNode_StateMachine_0.WingsGroundStatemachine.AnimStateTransitionNode_24.Transition
    src=FallStart
    via=AnimStateTransitionNode_24
   }
   {
    dst=FallLoop
    rule=AnimGraph.AnimGraphNode_StateMachine_0.WingsGroundStatemachine.AnimStateTransitionNode_23.Transition
    src=FromTacticalJump
    via=AnimStateTransitionNode_23
   }
   {
    dst=Conduit
    rule=AnimGraph.AnimGraphNode_StateMachine_0.WingsGroundStatemachine.AnimStateTransitionNode_43.Transition
    src=GrindRail
    via=AnimStateTransitionNode_43
   }
   {
    dst=Conduit
    rule=AnimGraph.AnimGraphNode_StateMachine_0.WingsGroundStatemachine.AnimStateTransitionNode_1.Transition
    src=Idle
    via=AnimStateTransitionNode_1
   }
   {
    dst=Conduit
    rule=AnimGraph.AnimGraphNode_StateMachine_0.WingsGroundStatemachine.AnimStateTransitionNode_9.Transition
    src=JumpLand
    via=AnimStateTransitionNode_9
   }
   {
    dst=JumpStart
    rule=AnimGraph.AnimGraphNode_StateMachine_0.WingsGroundStatemachine.AnimStateTransitionNode_6.AnimationTransitionGraph_1
    src=JumpLand
    via=AnimStateTransitionNode_6
   }
   {
    dst=FallStart
    rule=AnimGraph.AnimGraphNode_StateMachine_0.WingsGroundStatemachine.AnimStateTransitionNode_3.AnimationTransitionGraph_1
    src=JumpStart
    via=AnimStateTransitionNode_3
   }
   {
    dst=Conduit
    rule=AnimGraph.AnimGraphNode_StateMachine_0.WingsGroundStatemachine.AnimStateTransitionNode_0.Transition
    src=Moving
    via=AnimStateTransitionNode_0
   }
   {
    dst=FallStart
    rule=AnimGraph.AnimGraphNode_StateMachine_0.WingsGroundStatemachine.AnimStateTransitionNode_25.Transition
    src=Moving
    via=AnimStateTransitionNode_25
   }
   {
    dst=JumpStart
    rule=AnimGraph.AnimGraphNode_StateMachine_0.WingsGroundStatemachine.AnimStateTransitionNode_10.Transition
    src=Moving
    via=AnimStateTransitionNode_10
   }
   {
    dst=Conduit
    rule=AnimGraph.AnimGraphNode_StateMachine_0.WingsGroundStatemachine.AnimStateTransitionNode_38.Transition
    src=SlideStart
    via=AnimStateTransitionNode_38
   }
   {
    dst=SlidingPose
    rule=AnimGraph.AnimGraphNode_StateMachine_0.WingsGroundStatemachine.AnimStateTransitionNode_34.Transition
    src=SlideStart
    via=AnimStateTransitionNode_34
   }
   {
    dst=Conduit
    rule=AnimGraph.AnimGraphNode_StateMachine_0.WingsGroundStatemachine.AnimStateTransitionNode_39.Transition
    src=SlidingPose
    via=AnimStateTransitionNode_39
   }
   {
    dst=Conduit
    rule=AnimGraph.AnimGraphNode_StateMachine_0.WingsGroundStatemachine.AnimStateTransitionNode_16.Transition
    src=Swim
    via=AnimStateTransitionNode_16
   }
   {
    dst=Conduit
    rule=AnimGraph.AnimGraphNode_StateMachine_0.WingsGroundStatemachine.AnimStateTransitionNode_15.Transition
    src=TacticalJumpLoop
    via=AnimStateTransitionNode_15
   }
   {
    dst=TacticalLand
    rule=AnimGraph.AnimGraphNode_StateMachine_0.WingsGroundStatemachine.AnimStateTransitionNode_19.Transition
    src=TacticalJumpLoop
    via=AnimStateTransitionNode_19
   }
   {
    dst=TacticalJumpLoop
    rule=AnimGraph.AnimGraphNode_StateMachine_0.WingsGroundStatemachine.AnimStateTransitionNode_14.Transition
    src=TacticalJumpStart
    via=AnimStateTransitionNode_14
   }
   {
    dst=Conduit
    rule=AnimGraph.AnimGraphNode_StateMachine_0.WingsGroundStatemachine.AnimStateTransitionNode_21.Transition
    src=TacticalLand
    via=AnimStateTransitionNode_21
   }
   {
    dst=TacticalJumpStart
    rule=AnimGraph.AnimGraphNode_StateMachine_0.WingsGroundStatemachine.AnimStateTransitionNode_22.Transition
    src=TacticalLand
    via=AnimStateTransitionNode_22
   }
   {
    dst=TacticalSprint
    rule=AnimGraph.AnimGraphNode_StateMachine_0.WingsGroundStatemachine.AnimStateTransitionNode_20.Transition
    src=TacticalLand
    via=AnimStateTransitionNode_20
   }
   {
    dst=Conduit
    rule=AnimGraph.AnimGraphNode_StateMachine_0.WingsGroundStatemachine.AnimStateTransitionNode_18.Transition
    src=TacticalSprint
    via=AnimStateTransitionNode_18
   }
   {
    dst=TacticalJumpStart
    rule=AnimGraph.AnimGraphNode_StateMachine_0.WingsGroundStatemachine.AnimStateTransitionNode_13.Transition
    src=TacticalSprint
    via=AnimStateTransitionNode_13
   }
   {
    dst=SlideStart
    rule=AnimGraph.AnimGraphNode_StateMachine_0.WingsGroundStatemachine.AnimStateTransitionNode_33.Transition
    src=ToSliding
    via=AnimStateTransitionNode_33
   }
   {
    dst=Zipline
    rule=AnimGraph.AnimGraphNode_StateMachine_0.WingsGroundStatemachine.AnimStateTransitionNode_42.Transition
    src=ToZipline
    via=AnimStateTransitionNode_42
   }
   {
    dst=Conduit
    rule=AnimGraph.AnimGraphNode_StateMachine_0.WingsGroundStatemachine.AnimStateTransitionNode_41.Transition
    src=Zipline
    via=AnimStateTransitionNode_41
   }
  ]
 }
 Conduit {path=AnimGraph.AnimGraphNode_StateMachine_0.WingsGroundStatemachine.AnimStateConduitNode_0.Conduit, nodes [Result(bCanEnterTransition=true)]}
 Swim {
  path=AnimGraph.AnimGraphNode_StateMachine_0.WingsGroundStatemachine.AnimStateNode_0.Swim
  nodes [Output Animation Pose(Result=Sequence Player 'Olympus_Wings_Swimming_Pose'() {bLoopAnimation=false, sequence=/SunRoseFlyingGameplay/Items/Wings/Animations/Wings/Olympus_Wings_Swimming_Pose})]
 }
 Idle {
  path=AnimGraph.AnimGraphNode_StateMachine_0.WingsGroundStatemachine.AnimStateNode_1.Idle
  nodes [Output Animation Pose(Result=Sequence Player 'Olympus_Wings_Ground_Idle'() {sequence=/SunRoseFlyingGameplay/Items/Wings/Animations/Wings/Olympus_Wings_Ground_Idle})]
 }
 TacticalLand {
  path=AnimGraph.AnimGraphNode_StateMachine_0.WingsGroundStatemachine.AnimStateNode_10.TacticalLand
  nodes [Output Animation Pose(Result=Sequence Player 'Olympus_Wings_TacticalSprint_Jump_Land'() {bLoopAnimation=false, sequence=/SunRoseFlyingGameplay/Items/Wings/Animations/Wings/Olympus_Wings_TacticalSprint_Jump_Land})]
 }
 SlidingPose {
  path=AnimGraph.AnimGraphNode_StateMachine_0.WingsGroundStatemachine.AnimStateNode_11.SlidingPose
  nodes [Output Animation Pose(Result=Sequence Player 'Olympus_Wings_KneeSliding'() {bLoopAnimation=false, sequence=/SunRoseFlyingGameplay/Items/Wings/Animations/Wings/Olympus_Wings_KneeSliding})]
 }
 CrouchIdle {
  path=AnimGraph.AnimGraphNode_StateMachine_0.WingsGroundStatemachine.AnimStateNode_12.CrouchIdle
  nodes [Output Animation Pose(Result=Sequence Player 'Olympus_Wings_Crouch'() {bLoopAnimation=false, sequence=/SunRoseFlyingGameplay/Items/Wings/Animations/Wings/Olympus_Wings_Crouch})]
 }
 Crouch Moving {
  path=AnimGraph.AnimGraphNode_StateMachine_0.WingsGroundStatemachine.AnimStateNode_13.Crouch Moving
  nodes [Output Animation Pose(Result=Sequence Player 'Olympus_Wings_Crouch_Moving'() {bLoopAnimation=false, sequence=/SunRoseFlyingGameplay/Items/Wings/Animations/Wings/Olympus_Wings_Crouch_Moving})]
 }
 SlideStart {
  path=AnimGraph.AnimGraphNode_StateMachine_0.WingsGroundStatemachine.AnimStateNode_14.SlideStart
  nodes [Output Animation Pose(Result=Sequence Player 'Olympus_Wings_KneeSlidingInitiate'() {bLoopAnimation=false, sequence=/SunRoseFlyingGameplay/Items/Wings/Animations/Wings/Olympus_Wings_KneeSlidingInitiate})]
 }
 Zipline {
  path=AnimGraph.AnimGraphNode_StateMachine_0.WingsGroundStatemachine.AnimStateNode_15.Zipline
  nodes [Output Animation Pose(Result=Sequence Player 'Olympus_Wings_Zipline_Anim'() {sequence=/SunRoseFlyingGameplay/Items/Wings/Animations/Wings/Olympus_Wings_Zipline_Anim})]
 }
 DiveHitToIdle {
  path=AnimGraph.AnimGraphNode_StateMachine_0.WingsGroundStatemachine.AnimStateNode_16.DiveHitToIdle
  nodes [Output Animation Pose(Result=Sequence Player 'Olympus_Wings_Dive_Hit_To_Idle'() {bLoopAnimation=false, sequence=/SunRoseFlyingGameplay/Items/Wings/Animations/Wings/Olympus_Wings_Dive_Hit_To_Idle})]
 }
 GrindRail {
  path=AnimGraph.AnimGraphNode_StateMachine_0.WingsGroundStatemachine.AnimStateNode_17.GrindRail
  nodes [Output Animation Pose(Result=Blendspace Player 'Olympus_Wings_GrindRail_BS'!(X=Get PlayerSpeed2D() -> (PlayerSpeed2D:float)) -> (Pose) {blendSpace=/SunRoseFlyingGameplay/Items/Wings/Animations/Wings/Olympus_Wings_GrindRail_BS})]
 }
 JumpStart {
  path=AnimGraph.AnimGraphNode_StateMachine_0.WingsGroundStatemachine.AnimStateNode_2.JumpStart
  nodes [Output Animation Pose(Result=Sequence Player 'Olympus_Wings_Ground_Jump_Start'() {bLoopAnimation=false, sequence=/SunRoseFlyingGameplay/Items/Wings/Animations/Wings/Olympus_Wings_Ground_Jump_Start})]
 }
 FallLoop {
  path=AnimGraph.AnimGraphNode_StateMachine_0.WingsGroundStatemachine.AnimStateNode_3.FallLoop
  nodes [Output Animation Pose(Result=Sequence Player 'Olympus_Wings_Ground_Jump_Falling_Loop'() {sequence=/SunRoseFlyingGameplay/Items/Wings/Animations/Wings/Olympus_Wings_Ground_Jump_Falling_Loop})]
 }
 JumpLand {
  path=AnimGraph.AnimGraphNode_StateMachine_0.WingsGroundStatemachine.AnimStateNode_4.JumpLand
  nodes [
   Sequence Player 'Olympus_Wings_Ground_Jump_Falling_Land'() {bLoopAnimation=false, sequence=/SunRoseFlyingGameplay/Items/Wings/Animations/Wings/Olympus_Wings_Ground_Jump_Falling_Land}
   Output Animation Pose(Result=Sequence Player 'Olympus_Wings_Ground_Jump_Land'() {bLoopAnimation=false, sequence=/SunRoseFlyingGameplay/Items/Wings/Animations/Wings/Olympus_Wings_Ground_Jump_Land})
  ]
 }
 Moving {
  path=AnimGraph.AnimGraphNode_StateMachine_0.WingsGroundStatemachine.AnimStateNode_5.Moving
  nodes [
   AnimGraphNode_StateMachine_0 = Moving_Wings() {is=title AnimGraphNode_StateMachine}
   Output Animation Pose(Result=@AnimGraphNode_StateMachine_0)
  ]
 }
 Moving_Wings {
  entry=Intro_Moving
  path=AnimGraph.AnimGraphNode_StateMachine_0.WingsGroundStatemachine.AnimStateNode_5.Moving.AnimGraphNode_StateMachine_0.Moving_Wings
  ids=id
  nodes [
   AnimStateNode_0 = Intro_Moving() {is=title AnimStateNode}
   AnimStateNode_1 = Moving_Pose() {is=title AnimStateNode}
   AnimStateTransitionNode_0 = AnimStateTransitionNode() {automaticRuleTriggerTime=0, bAutomaticRuleBasedOnSequencePlayerInState=true, logicType=TLT_Inertialization}
  ]
  links [{
    dst=Moving_Pose
    rule=AnimGraph.AnimGraphNode_StateMachine_0.WingsGroundStatemachine.AnimStateNode_5.Moving.AnimGraphNode_StateMachine_0.Moving_Wings.AnimStateTransitionNode_0.Transition
    src=Intro_Moving
    via=AnimStateTransitionNode_0
   }]
 }
 Intro_Moving {
  path=AnimGraph.AnimGraphNode_StateMachine_0.WingsGroundStatemachine.AnimStateNode_5.Moving.AnimGraphNode_StateMachine_0.Moving_Wings.AnimStateNode_0.Intro_Moving
  nodes [Output Animation Pose(Result=Sequence Player 'Olympus_Wings_Ground_Loco_Start'() {bLoopAnimation=false, sequence=/SunRoseFlyingGameplay/Items/Wings/Animations/Wings/Olympus_Wings_Ground_Loco_Start})]
 }
 Moving_Pose {
  path=AnimGraph.AnimGraphNode_StateMachine_0.WingsGroundStatemachine.AnimStateNode_5.Moving.AnimGraphNode_StateMachine_0.Moving_Wings.AnimStateNode_1.Moving_Pose
  nodes [Output Animation Pose(Result=Sequence Player 'Olympus_Wings_Ground_Loco_Loop'() {sequence=/SunRoseFlyingGameplay/Items/Wings/Animations/Wings/Olympus_Wings_Ground_Loco_Loop})]
 }
 Transition {
  path=AnimGraph.AnimGraphNode_StateMachine_0.WingsGroundStatemachine.AnimStateNode_5.Moving.AnimGraphNode_StateMachine_0.Moving_Wings.AnimStateTransitionNode_0.Transition
  nodes [Result()]
 }
 FallStart {
  path=AnimGraph.AnimGraphNode_StateMachine_0.WingsGroundStatemachine.AnimStateNode_6.FallStart
  nodes [Output Animation Pose(Result=Sequence Player 'Olympus_Wings_Ground_Jump_Loop'() {bLoopAnimation=false, sequence=/SunRoseFlyingGameplay/Items/Wings/Animations/Wings/Olympus_Wings_Ground_Jump_Loop})]
 }
 TacticalSprint {
  path=AnimGraph.AnimGraphNode_StateMachine_0.WingsGroundStatemachine.AnimStateNode_7.TacticalSprint
  nodes [Output Animation Pose(Result=Sequence Player 'Olympus_Wings_TacticalSprint_Pose'() {bLoopAnimation=false, sequence=/SunRoseFlyingGameplay/Items/Wings/Animations/Wings/Olympus_Wings_TacticalSprint_Pose})]
 }
 TacticalJumpStart {
  path=AnimGraph.AnimGraphNode_StateMachine_0.WingsGroundStatemachine.AnimStateNode_8.TacticalJumpStart
  nodes [Output Animation Pose(Result=Sequence Player 'Olympus_Wings_TacticalSprint_Jump_Start'() {bLoopAnimation=false, sequence=/SunRoseFlyingGameplay/Items/Wings/Animations/Wings/Olympus_Wings_TacticalSprint_Jump_Start})]
 }
 TacticalJumpLoop {
  path=AnimGraph.AnimGraphNode_StateMachine_0.WingsGroundStatemachine.AnimStateNode_9.TacticalJumpLoop
  nodes [Output Animation Pose(Result=Sequence Player 'Olympus_Wings_TacticalSprint_Jump_Loop'() {sequence=/SunRoseFlyingGameplay/Items/Wings/Animations/Wings/Olympus_Wings_TacticalSprint_Jump_Loop})]
 }
 Transition {
  path=AnimGraph.AnimGraphNode_StateMachine_0.WingsGroundStatemachine.AnimStateTransitionNode_0.Transition
  nodes [Result(bCanEnterTransition=Get bIsNotMoving() -> (bIsNotMoving:bool))]
 }
 Transition {
  path=AnimGraph.AnimGraphNode_StateMachine_0.WingsGroundStatemachine.AnimStateTransitionNode_1.Transition
  nodes [Result(bCanEnterTransition=true)]
 }
 Transition {
  path=AnimGraph.AnimGraphNode_StateMachine_0.WingsGroundStatemachine.AnimStateTransitionNode_10.Transition
  nodes [Result(bCanEnterTransition=Get bPlayerIsJumping() -> (bPlayerIsJumping:bool))]
 }
 Transition {path=AnimGraph.AnimGraphNode_StateMachine_0.WingsGroundStatemachine.AnimStateTransitionNode_11.Transition, nodes [Result()]}
 Transition {
  path=AnimGraph.AnimGraphNode_StateMachine_0.WingsGroundStatemachine.AnimStateTransitionNode_12.Transition
  nodes [Result(bCanEnterTransition=Get bIsFallLoopToJumpLand() -> (bIsFallLoopToJumpLand:bool))]
 }
 Transition {
  path=AnimGraph.AnimGraphNode_StateMachine_0.WingsGroundStatemachine.AnimStateTransitionNode_13.Transition
  nodes [Result(bCanEnterTransition=Get bPlayerIsTacticalJump() -> (bPlayerIsTacticalJump:bool))]
 }
 Transition {path=AnimGraph.AnimGraphNode_StateMachine_0.WingsGroundStatemachine.AnimStateTransitionNode_14.Transition, nodes [Result()]}
 Transition {
  path=AnimGraph.AnimGraphNode_StateMachine_0.WingsGroundStatemachine.AnimStateTransitionNode_15.Transition
  nodes [Result(bCanEnterTransition=Get bIsTacticalJumpLoopEnd() -> (bIsTacticalJumpLoopEnd:bool))]
 }
 Transition {
  path=AnimGraph.AnimGraphNode_StateMachine_0.WingsGroundStatemachine.AnimStateTransitionNode_16.Transition
  nodes [Result(bCanEnterTransition=NOT Boolean(A=Get bPlayerIsSurfaceSwimming() -> (bPlayerIsSurfaceSwimming:bool)))]
 }
 Transition {
  path=AnimGraph.AnimGraphNode_StateMachine_0.WingsGroundStatemachine.AnimStateTransitionNode_17.Transition
  nodes [Result(bCanEnterTransition=Get bPlayerIsPlayDiveHitToIdle() -> (bPlayerIsPlayDiveHitToIdle:bool))]
 }
 Transition {
  path=AnimGraph.AnimGraphNode_StateMachine_0.WingsGroundStatemachine.AnimStateTransitionNode_18.Transition
  nodes [Result(bCanEnterTransition=NOT Boolean(A=Get bPlayerIsTacticalSprint() -> (bPlayerIsTacticalSprint:bool)))]
 }
 Transition {
  path=AnimGraph.AnimGraphNode_StateMachine_0.WingsGroundStatemachine.AnimStateTransitionNode_19.Transition
  nodes [Result(bCanEnterTransition=Get bTacticalJumpLoopToTacticalLand() -> (bTacticalJumpLoopToTacticalLand:bool))]
 }
 Transition {
  path=AnimGraph.AnimGraphNode_StateMachine_0.WingsGroundStatemachine.AnimStateTransitionNode_2.Transition
  nodes [Result(bCanEnterTransition=true)]
 }
 Transition {path=AnimGraph.AnimGraphNode_StateMachine_0.WingsGroundStatemachine.AnimStateTransitionNode_20.Transition, nodes [Result()]}
 Transition {
  path=AnimGraph.AnimGraphNode_StateMachine_0.WingsGroundStatemachine.AnimStateTransitionNode_21.Transition
  nodes [Result(bCanEnterTransition=NOT Boolean(A=Get bPlayerIsTacticalSprint()))]
 }
 Transition {
  path=AnimGraph.AnimGraphNode_StateMachine_0.WingsGroundStatemachine.AnimStateTransitionNode_22.Transition
  nodes [Result(bCanEnterTransition=Get bPlayerIsTacticalJump())]
 }
 Transition {
  path=AnimGraph.AnimGraphNode_StateMachine_0.WingsGroundStatemachine.AnimStateTransitionNode_23.Transition
  nodes [Result(bCanEnterTransition=NOT Boolean(A=Get bPlayerIsTacticalSprint()))]
 }
 Transition {
  path=AnimGraph.AnimGraphNode_StateMachine_0.WingsGroundStatemachine.AnimStateTransitionNode_24.Transition
  nodes [Result(bCanEnterTransition=Get bIsFallLoopToJumpLand())]
 }
 Transition {
  path=AnimGraph.AnimGraphNode_StateMachine_0.WingsGroundStatemachine.AnimStateTransitionNode_25.Transition
  nodes [Result(bCanEnterTransition=Get bPlayerIsFalling() -> (bPlayerIsFalling:bool))]
 }
 Transition {
  path=AnimGraph.AnimGraphNode_StateMachine_0.WingsGroundStatemachine.AnimStateTransitionNode_26.Transition
  nodes [Result(bCanEnterTransition=Get bPlayerIsCrouching() -> (bPlayerIsCrouching:bool))]
 }
 Transition {
  path=AnimGraph.AnimGraphNode_StateMachine_0.WingsGroundStatemachine.AnimStateTransitionNode_27.Transition
  nodes [Result(bCanEnterTransition=Get bPlayerIsCrouchMoving() -> (bPlayerIsCrouchMoving:bool))]
 }
 Transition {
  path=AnimGraph.AnimGraphNode_StateMachine_0.WingsGroundStatemachine.AnimStateTransitionNode_28.Transition
  nodes [Result(bCanEnterTransition=Get bIsCrouchIdleToCrouchMoving() -> (bIsCrouchIdleToCrouchMoving:bool))]
 }
 Transition {
  path=AnimGraph.AnimGraphNode_StateMachine_0.WingsGroundStatemachine.AnimStateTransitionNode_29.Transition
  nodes [Result(bCanEnterTransition=Get bIsCrouchMovingToCrouchIdle() -> (bIsCrouchMovingToCrouchIdle:bool))]
 }
 AnimationTransitionGraph_1 {path=AnimGraph.AnimGraphNode_StateMachine_0.WingsGroundStatemachine.AnimStateTransitionNode_3.AnimationTransitionGraph_1, nodes [Result()]}
 Transition {
  path=AnimGraph.AnimGraphNode_StateMachine_0.WingsGroundStatemachine.AnimStateTransitionNode_30.Transition
  nodes [Result(bCanEnterTransition=Get bIsNotCrouchIdle() -> (bIsNotCrouchIdle:bool))]
 }
 Transition {
  path=AnimGraph.AnimGraphNode_StateMachine_0.WingsGroundStatemachine.AnimStateTransitionNode_31.Transition
  nodes [Result(bCanEnterTransition=Get bIsNotCrouchMoving() -> (bIsNotCrouchMoving:bool))]
 }
 Transition {path=AnimGraph.AnimGraphNode_StateMachine_0.WingsGroundStatemachine.AnimStateTransitionNode_32.Transition, nodes [Result()]}
 Transition {
  path=AnimGraph.AnimGraphNode_StateMachine_0.WingsGroundStatemachine.AnimStateTransitionNode_33.Transition
  nodes [Result(bCanEnterTransition=Get bIsToSliding() -> (bIsToSliding:bool))]
 }
 Transition {path=AnimGraph.AnimGraphNode_StateMachine_0.WingsGroundStatemachine.AnimStateTransitionNode_34.Transition, nodes [Result()]}
 Transition {
  path=AnimGraph.AnimGraphNode_StateMachine_0.WingsGroundStatemachine.AnimStateTransitionNode_36.Transition
  nodes [Result(bCanEnterTransition=Get bPlayerIsOnGrindRail() -> (bPlayerIsOnGrindRail:bool))]
 }
 Transition {
  path=AnimGraph.AnimGraphNode_StateMachine_0.WingsGroundStatemachine.AnimStateTransitionNode_37.Transition
  nodes [Result(bCanEnterTransition=Get bPlayerIsSliding() -> (bPlayerIsSliding:bool))]
 }
 Transition {
  path=AnimGraph.AnimGraphNode_StateMachine_0.WingsGroundStatemachine.AnimStateTransitionNode_38.Transition
  nodes [Result(bCanEnterTransition=NOT Boolean(A=Get bPlayerIsSliding()))]
 }
 Transition {
  path=AnimGraph.AnimGraphNode_StateMachine_0.WingsGroundStatemachine.AnimStateTransitionNode_39.Transition
  nodes [Result(bCanEnterTransition=NOT Boolean(A=Get bPlayerIsSliding()))]
 }
 Transition {
  path=AnimGraph.AnimGraphNode_StateMachine_0.WingsGroundStatemachine.AnimStateTransitionNode_4.Transition
  nodes [Result(bCanEnterTransition=Get bPlayerIsTacticalSprint())]
 }
 Transition {
  path=AnimGraph.AnimGraphNode_StateMachine_0.WingsGroundStatemachine.AnimStateTransitionNode_40.Transition
  nodes [Result(bCanEnterTransition=Get bIsOnZipline() -> (bIsOnZipline:bool))]
 }
 Transition {
  path=AnimGraph.AnimGraphNode_StateMachine_0.WingsGroundStatemachine.AnimStateTransitionNode_41.Transition
  nodes [Result(bCanEnterTransition=Get bIsNotOnZipline() -> (bIsNotOnZipline:bool))]
 }
 Transition {
  path=AnimGraph.AnimGraphNode_StateMachine_0.WingsGroundStatemachine.AnimStateTransitionNode_42.Transition
  nodes [Result(bCanEnterTransition=Get bIsOnZipline())]
 }
 Transition {
  path=AnimGraph.AnimGraphNode_StateMachine_0.WingsGroundStatemachine.AnimStateTransitionNode_43.Transition
  nodes [Result(bCanEnterTransition=NOT Boolean(A=Get bPlayerIsOnGrindRail()))]
 }
 Transition {
  path=AnimGraph.AnimGraphNode_StateMachine_0.WingsGroundStatemachine.AnimStateTransitionNode_5.Transition
  nodes [Result(bCanEnterTransition=Get bPlayerIsSurfaceSwimming())]
 }
 AnimationTransitionGraph_1 {
  path=AnimGraph.AnimGraphNode_StateMachine_0.WingsGroundStatemachine.AnimStateTransitionNode_6.AnimationTransitionGraph_1
  nodes [Result(bCanEnterTransition=Get bPlayerIsJumping())]
 }
 Transition {
  path=AnimGraph.AnimGraphNode_StateMachine_0.WingsGroundStatemachine.AnimStateTransitionNode_7.Transition
  nodes [Result(bCanEnterTransition=Get bIsConduitToMoving() -> (bIsConduitToMoving:bool))]
 }
 Transition {
  path=AnimGraph.AnimGraphNode_StateMachine_0.WingsGroundStatemachine.AnimStateTransitionNode_8.Transition
  nodes [Result(bCanEnterTransition=Get bToJumpStart() -> (bToJumpStart:bool))]
 }
 Transition {path=AnimGraph.AnimGraphNode_StateMachine_0.WingsGroundStatemachine.AnimStateTransitionNode_9.Transition, nodes [Result()]}
 WingsAirStatemachine {
  entry=Conduit
  path=AnimGraph.AnimGraphNode_StateMachine_1.WingsAirStatemachine
  ids=id
  nodes [
   ToDodge() {is=title AnimStateAliasNode, also {aliases [BoostingLoop, BoostingStart, DodgeLeftEnd, DodgeRightEnd]}}
   ToBoost() {is=title AnimStateAliasNode, also {aliases [DodgeLeftEnd, DodgeRightEnd]}}
   ToDive() {is=title AnimStateAliasNode, also {aliases [BoostEnd, BoostingLoop, DodgeLeftEnd, DodgeRightEnd]}}
   AnimStateConduitNode_0 = Conduit()
   AnimStateNode_0 = Leap() {is=title AnimStateNode}
   AnimStateNode_1 = Glide() {is=title AnimStateNode}
   AnimStateNode_10 = DodgeLeftEnd() {is=title AnimStateNode}
   AnimStateNode_2 = DivinStart() {is=title AnimStateNode}
   AnimStateNode_3 = DodgeRightStart() {is=title AnimStateNode}
   AnimStateNode_4 = DodgeLeftStart() {is=title AnimStateNode}
   AnimStateNode_5 = BoostingStart() {is=title AnimStateNode}
   AnimStateNode_6 = BoostingLoop() {is=title AnimStateNode}
   AnimStateNode_7 = DiveLoop() {is=title AnimStateNode}
   AnimStateNode_8 = BoostEnd() {is=title AnimStateNode}
   AnimStateNode_9 = DodgeRightEnd() {is=title AnimStateNode}
   AnimStateTransitionNode_0 = AnimStateTransitionNode() {crossfadeDuration=0.1, logicType=TLT_Inertialization, priorityOrder=3}
   AnimStateTransitionNode_1 = AnimStateTransitionNode() {crossfadeDuration=0, priorityOrder=5}
   AnimStateTransitionNode_10 = AnimStateTransitionNode() {crossfadeDuration=0.1, logicType=TLT_Inertialization, priorityOrder=2}
   AnimStateTransitionNode_11 = AnimStateTransitionNode()
   AnimStateTransitionNode_12 = AnimStateTransitionNode()
   AnimStateTransitionNode_13 = AnimStateTransitionNode() {crossfadeDuration=0, logicType=TLT_Inertialization}
   AnimStateTransitionNode_14 = AnimStateTransitionNode() {crossfadeDuration=0, logicType=TLT_Inertialization, priorityOrder=2}
   AnimStateTransitionNode_16 = AnimStateTransitionNode() {automaticRuleTriggerTime=0, bAutomaticRuleBasedOnSequencePlayerInState=true, crossfadeDuration=0.1, logicType=TLT_Inertialization}
   AnimStateTransitionNode_17 = AnimStateTransitionNode()
   AnimStateTransitionNode_18 = AnimStateTransitionNode()
   AnimStateTransitionNode_19 = AnimStateTransitionNode() {crossfadeDuration=0.4, logicType=TLT_Inertialization}
   AnimStateTransitionNode_2 = AnimStateTransitionNode() {crossfadeDuration=0, priorityOrder=4}
   AnimStateTransitionNode_20 = AnimStateTransitionNode() {bAutomaticRuleBasedOnSequencePlayerInState=true}
   AnimStateTransitionNode_21 = AnimStateTransitionNode()
   AnimStateTransitionNode_24 = AnimStateTransitionNode() {bAutomaticRuleBasedOnSequencePlayerInState=true, crossfadeDuration=0.1}
   AnimStateTransitionNode_25 = AnimStateTransitionNode() {bAutomaticRuleBasedOnSequencePlayerInState=true, crossfadeDuration=0.1}
   AnimStateTransitionNode_26 = AnimStateTransitionNode() {bAutomaticRuleBasedOnSequencePlayerInState=true}
   AnimStateTransitionNode_27 = AnimStateTransitionNode() {bAutomaticRuleBasedOnSequencePlayerInState=true}
   AnimStateTransitionNode_3 = AnimStateTransitionNode() {crossfadeDuration=0.1, logicType=TLT_Inertialization}
   AnimStateTransitionNode_4 = AnimStateTransitionNode() {automaticRuleTriggerTime=0, bAutomaticRuleBasedOnSequencePlayerInState=true}
   AnimStateTransitionNode_5 = AnimStateTransitionNode() {crossfadeDuration=0.4, logicType=TLT_Inertialization, priorityOrder=100}
   AnimStateTransitionNode_6 = AnimStateTransitionNode() {crossfadeDuration=0.1, logicType=TLT_Inertialization}
   AnimStateTransitionNode_7 = AnimStateTransitionNode() {logicType=TLT_Inertialization}
   AnimStateTransitionNode_8 = AnimStateTransitionNode()
   AnimStateTransitionNode_9 = AnimStateTransitionNode()
  ]
  links [
   {
    dst=Conduit
    rule=AnimGraph.AnimGraphNode_StateMachine_1.WingsAirStatemachine.AnimStateTransitionNode_20.Transition
    src=BoostEnd
    via=AnimStateTransitionNode_20
   }
   {
    dst=Conduit
    rule=AnimGraph.AnimGraphNode_StateMachine_1.WingsAirStatemachine.AnimStateTransitionNode_21.Transition
    src=BoostEnd
    via=AnimStateTransitionNode_21
   }
   {
    dst=BoostEnd
    rule=AnimGraph.AnimGraphNode_StateMachine_1.WingsAirStatemachine.AnimStateTransitionNode_19.Transition
    src=BoostingLoop
    via=AnimStateTransitionNode_19
   }
   {
    dst=BoostingLoop
    rule=AnimGraph.AnimGraphNode_StateMachine_1.WingsAirStatemachine.AnimStateTransitionNode_6.Transition
    src=BoostingStart
    via=AnimStateTransitionNode_6
   }
   {
    dst=Conduit
    rule=AnimGraph.AnimGraphNode_StateMachine_1.WingsAirStatemachine.AnimStateTransitionNode_12.Transition
    src=BoostingStart
    via=AnimStateTransitionNode_12
   }
   {
    dst=BoostingStart
    rule=AnimGraph.AnimGraphNode_StateMachine_1.WingsAirStatemachine.AnimStateTransitionNode_0.Transition
    src=Conduit
    via=AnimStateTransitionNode_0
   }
   {
    dst=DivinStart
    rule=AnimGraph.AnimGraphNode_StateMachine_1.WingsAirStatemachine.AnimStateTransitionNode_10.Transition
    src=Conduit
    via=AnimStateTransitionNode_10
   }
   {
    dst=DodgeLeftStart
    rule=AnimGraph.AnimGraphNode_StateMachine_1.WingsAirStatemachine.AnimStateTransitionNode_1.Transition
    src=Conduit
    via=AnimStateTransitionNode_1
   }
   {
    dst=DodgeRightStart
    rule=AnimGraph.AnimGraphNode_StateMachine_1.WingsAirStatemachine.AnimStateTransitionNode_2.Transition
    src=Conduit
    via=AnimStateTransitionNode_2
   }
   {
    dst=Glide
    rule=AnimGraph.AnimGraphNode_StateMachine_1.WingsAirStatemachine.AnimStateTransitionNode_5.Transition
    src=Conduit
    via=AnimStateTransitionNode_5
   }
   {
    dst=Leap
    rule=AnimGraph.AnimGraphNode_StateMachine_1.WingsAirStatemachine.AnimStateTransitionNode_7.Transition
    src=Conduit
    via=AnimStateTransitionNode_7
   }
   {
    dst=Conduit
    rule=AnimGraph.AnimGraphNode_StateMachine_1.WingsAirStatemachine.AnimStateTransitionNode_17.Transition
    src=DiveLoop
    via=AnimStateTransitionNode_17
   }
   {
    dst=Conduit
    rule=AnimGraph.AnimGraphNode_StateMachine_1.WingsAirStatemachine.AnimStateTransitionNode_11.Transition
    src=DivinStart
    via=AnimStateTransitionNode_11
   }
   {
    dst=DiveLoop
    rule=AnimGraph.AnimGraphNode_StateMachine_1.WingsAirStatemachine.AnimStateTransitionNode_16.Transition
    src=DivinStart
    via=AnimStateTransitionNode_16
   }
   {
    dst=Conduit
    rule=AnimGraph.AnimGraphNode_StateMachine_1.WingsAirStatemachine.AnimStateTransitionNode_26.Transition
    src=DodgeLeftEnd
    via=AnimStateTransitionNode_26
   }
   {
    dst=DodgeLeftEnd
    rule=AnimGraph.AnimGraphNode_StateMachine_1.WingsAirStatemachine.AnimStateTransitionNode_25.Transition
    src=DodgeLeftStart
    via=AnimStateTransitionNode_25
   }
   {
    dst=Conduit
    rule=AnimGraph.AnimGraphNode_StateMachine_1.WingsAirStatemachine.AnimStateTransitionNode_27.Transition
    src=DodgeRightEnd
    via=AnimStateTransitionNode_27
   }
   {
    dst=DodgeRightEnd
    rule=AnimGraph.AnimGraphNode_StateMachine_1.WingsAirStatemachine.AnimStateTransitionNode_24.Transition
    src=DodgeRightStart
    via=AnimStateTransitionNode_24
   }
   {
    dst=Conduit
    rule=AnimGraph.AnimGraphNode_StateMachine_1.WingsAirStatemachine.AnimStateTransitionNode_9.Transition
    src=Glide
    via=AnimStateTransitionNode_9
   }
   {
    dst=Conduit
    rule=AnimGraph.AnimGraphNode_StateMachine_1.WingsAirStatemachine.AnimStateTransitionNode_18.Transition
    src=Leap
    via=AnimStateTransitionNode_18
   }
   {
    dst=Conduit
    rule=AnimGraph.AnimGraphNode_StateMachine_1.WingsAirStatemachine.AnimStateTransitionNode_4.Transition
    src=Leap
    via=AnimStateTransitionNode_4
   }
   {
    dst=BoostingStart
    rule=AnimGraph.AnimGraphNode_StateMachine_1.WingsAirStatemachine.AnimStateTransitionNode_8.Transition
    src=ToBoost
    via=AnimStateTransitionNode_8
   }
   {
    dst=DivinStart
    rule=AnimGraph.AnimGraphNode_StateMachine_1.WingsAirStatemachine.AnimStateTransitionNode_3.Transition
    src=ToDive
    via=AnimStateTransitionNode_3
   }
   {
    dst=DodgeLeftStart
    rule=AnimGraph.AnimGraphNode_StateMachine_1.WingsAirStatemachine.AnimStateTransitionNode_14.Transition
    src=ToDodge
    via=AnimStateTransitionNode_14
   }
   {
    dst=DodgeRightStart
    rule=AnimGraph.AnimGraphNode_StateMachine_1.WingsAirStatemachine.AnimStateTransitionNode_13.Transition
    src=ToDodge
    via=AnimStateTransitionNode_13
   }
  ]
 }
 Conduit {path=AnimGraph.AnimGraphNode_StateMachine_1.WingsAirStatemachine.AnimStateConduitNode_0.Conduit, nodes [Result(bCanEnterTransition=true)]}
 Leap {
  path=AnimGraph.AnimGraphNode_StateMachine_1.WingsAirStatemachine.AnimStateNode_0.Leap
  nodes [Output Animation Pose(Result=Sequence Player 'Olympus_Wings_Takeoff'() {bLoopAnimation=false, sequence=/SunRoseFlyingGameplay/Items/Wings/Animations/Wings/Olympus_Wings_Takeoff})]
 }
 Glide {
  path=AnimGraph.AnimGraphNode_StateMachine_1.WingsAirStatemachine.AnimStateNode_1.Glide
  nodes [
   AnimGraphNode_StateMachine_0 = GlideStateMachine() {is=title AnimGraphNode_StateMachine}
   n1 = Blend Poses by bool(
    BlendPose_0=Sequence Player 'Olympus_Wings_Flap_Large'() {sequence=/SunRoseFlyingGameplay/Items/Wings/Animations/Wings/Olympus_Wings_Flap_Large}
    BlendPose_1=@AnimGraphNode_StateMachine_0
    BlendTime_0=0.5
    BlendTime_1=0.4
    bActiveValue=Get bPlayerIsAllowWingFlap() -> (bPlayerIsAllowWingFlap:bool)
   ) {blendType=Sinusoidal, transitionType=Inertialization}
   n2 = Sequence Player 'Olympus_Wings_NoFuel'() {sequence=/SunRoseFlyingGameplay/Items/Wings/Animations/Wings/Olympus_Wings_NoFuel}
   n3 = Get bIsOutOfGas() -> (bIsOutOfGas:bool)
   n4 = Apply Additive(
    Base=Two Way Blend(
     A=Blend Poses by bool(BlendPose_0=@n2, BlendPose_1=@n1, BlendTime_0=0.4, BlendTime_1=0.5, bActiveValue=@n3) {blendType=Sinusoidal, transitionType=Inertialization}
     B=Sequence Player 'Olympus_Wings_Glide_Pitch_Down'() {bLoopAnimation=false, sequence=/SunRoseFlyingGameplay/Items/Wings/Animations/Wings/Olympus_Wings_Glide_Pitch_Down}
     Alpha=Get PlayerGlidePitchAlpha() -> (PlayerGlidePitchAlpha:float)
    ) {blendNode {alphaBoolBlend {blendOption=Linear}}}
    Additive="Sequence Player (additive) 'Olympus_Wings_Feather_Noise'"() {sequence=/SunRoseFlyingGameplay/Items/Wings/Animations/Wings/Olympus_Wings_Feather_Noise}
   ) {alphaBoolBlend {blendOption=Linear}}
   n5 = "Sequence Player (additive) 'Olympus_Wings_Landing_Pose'"() {bLoopAnimation=false, sequence=/SunRoseFlyingGameplay/Items/Wings/Animations/Wings/Olympus_Wings_Landing_Pose}
   n6 = Get WingSlowDownOrLandAlpha() -> (WingSlowDownOrLandAlpha:float)
   Output Animation Pose(Result=Blend Poses by bool(
     BlendPose_0=Sequence Player 'Olympus_Wings_Ascend_Loop'() {groupName=Flapping, method=SyncGroup, sequence=/SunRoseFlyingGameplay/Items/Wings/Animations/Wings/Olympus_Wings_Ascend_Loop}
     BlendPose_1=Apply Additive(Base=@n4, Additive=@n5, Alpha=@n6) {alphaBoolBlend {blendOption=Linear}}
     BlendTime_0=0.3
     BlendTime_1=0.6
     bActiveValue=Get bPlayerIsAscending() -> (bPlayerIsAscending:bool)
    ) {blendType=Linear, transitionType=Inertialization})
  ]
 }
 GlideStateMachine {
  entry=GlideState
  path=AnimGraph.AnimGraphNode_StateMachine_1.WingsAirStatemachine.AnimStateNode_1.Glide.AnimGraphNode_StateMachine_0.GlideStateMachine
  ids=id
  nodes [
   AnimStateNode_0 = GlideState() {is=title AnimStateNode}
   AnimStateNode_1 = RandomFlap() {is=title AnimStateNode}
   AnimStateTransitionNode_0 = AnimStateTransitionNode() {crossfadeDuration=0.25, logicType=TLT_Inertialization}
   AnimStateTransitionNode_1 = AnimStateTransitionNode() {bAutomaticRuleBasedOnSequencePlayerInState=true, crossfadeDuration=0.3, logicType=TLT_Inertialization}
  ]
  links [
   {
    dst=RandomFlap
    rule=AnimGraph.AnimGraphNode_StateMachine_1.WingsAirStatemachine.AnimStateNode_1.Glide.AnimGraphNode_StateMachine_0.GlideStateMachine.AnimStateTransitionNode_0.Transition
    src=GlideState
    via=AnimStateTransitionNode_0
   }
   {
    dst=GlideState
    rule=AnimGraph.AnimGraphNode_StateMachine_1.WingsAirStatemachine.AnimStateNode_1.Glide.AnimGraphNode_StateMachine_0.GlideStateMachine.AnimStateTransitionNode_1.Transition
    src=RandomFlap
    via=AnimStateTransitionNode_1
   }
  ]
 }
 GlideState {
  path=AnimGraph.AnimGraphNode_StateMachine_1.WingsAirStatemachine.AnimStateNode_1.Glide.AnimGraphNode_StateMachine_0.GlideStateMachine.AnimStateNode_0.GlideState
  nodes [Output Animation Pose(Result=Blendspace Player 'Olympus_Wings_Glide_Turbulence_BS'!(X=Get PlayerFuel() -> (PlayerFuel:float)) -> (Pose) {blendSpace=/SunRoseFlyingGameplay/Items/Wings/Animations/Wings/Olympus_Wings_Glide_Turbulence_BS})]
 }
 RandomFlap {
  path=AnimGraph.AnimGraphNode_StateMachine_1.WingsAirStatemachine.AnimStateNode_1.Glide.AnimGraphNode_StateMachine_0.GlideStateMachine.AnimStateNode_1.RandomFlap
  nodes [Output Animation Pose(Result=Sequence Player 'Olympus_Wings_Flap_Small'() {bLoopAnimation=false, sequence=/SunRoseFlyingGameplay/Items/Wings/Animations/Wings/Olympus_Wings_Flap_Small})]
 }
 Transition {
  path=AnimGraph.AnimGraphNode_StateMachine_1.WingsAirStatemachine.AnimStateNode_1.Glide.AnimGraphNode_StateMachine_0.GlideStateMachine.AnimStateTransitionNode_0.Transition
  nodes [Result(bCanEnterTransition=Get bIsRandomFlap() -> (bIsRandomFlap:bool))]
 }
 Transition {
  path=AnimGraph.AnimGraphNode_StateMachine_1.WingsAirStatemachine.AnimStateNode_1.Glide.AnimGraphNode_StateMachine_0.GlideStateMachine.AnimStateTransitionNode_1.Transition
  nodes [Result()]
 }
 DodgeLeftEnd {
  path=AnimGraph.AnimGraphNode_StateMachine_1.WingsAirStatemachine.AnimStateNode_10.DodgeLeftEnd
  nodes [Output Animation Pose(Result=Sequence Player 'Olympus_Wings_Dodge_Left_Recover'() {bLoopAnimation=false, sequence=/SunRoseFlyingGameplay/Items/Wings/Animations/Wings/Olympus_Wings_Dodge_Left_Recover})]
 }
 DivinStart {
  path=AnimGraph.AnimGraphNode_StateMachine_1.WingsAirStatemachine.AnimStateNode_2.DivinStart
  nodes [Output Animation Pose(Result=Blend Poses by bool(
     BlendPose_0=Sequence Player 'Olympus_Wings_Dive_Start_HitReact'() {bLoopAnimation=false, sequence=/SunRoseFlyingGameplay/Items/Wings/Animations/Wings/Olympus_Wings_Dive_Start_HitReact}
     BlendPose_1=Sequence Player 'Olympus_Wings_Dive'() {bLoopAnimation=false, sequence=/SunRoseFlyingGameplay/Items/Wings/Animations/Wings/Olympus_Wings_Dive}
     bActiveValue=Get bPlayerIsForceDive() -> (bPlayerIsForceDive:bool)
    ) {blendType=Linear, transitionType=Inertialization})]
 }
 DodgeRightStart {
  path=AnimGraph.AnimGraphNode_StateMachine_1.WingsAirStatemachine.AnimStateNode_3.DodgeRightStart
  nodes [Output Animation Pose(Result=Modify Curve(
     SourcePose=Sequence Player 'Olympus_Wings_Dodge_Right'() {bLoopAnimation=false, sequence=/SunRoseFlyingGameplay/Items/Wings/Animations/Wings/Olympus_Wings_Dodge_Right}
     CurveValues_0=1
     Alpha=1
    ) {curveNames [Stop], curveValues [0]})]
 }
 DodgeLeftStart {
  path=AnimGraph.AnimGraphNode_StateMachine_1.WingsAirStatemachine.AnimStateNode_4.DodgeLeftStart
  nodes [Output Animation Pose(Result=Modify Curve(
     SourcePose=Sequence Player 'Olympus_Wings_Dodge_Left'() {bLoopAnimation=false, sequence=/SunRoseFlyingGameplay/Items/Wings/Animations/Wings/Olympus_Wings_Dodge_Left}
     CurveValues_0=1
     Alpha=1
    ) {curveNames [Stop], curveValues [0]})]
 }
 BoostingStart {
  path=AnimGraph.AnimGraphNode_StateMachine_1.WingsAirStatemachine.AnimStateNode_5.BoostingStart
  nodes [Output Animation Pose(Result=Modify Curve(
     SourcePose=Sequence Player 'Olympus_Wings_Boost_Start'() {bLoopAnimation=false, sequence=/SunRoseFlyingGameplay/Items/Wings/Animations/Wings/Olympus_Wings_Boost_Start}
     CurveValues_0=1
     Alpha=1
    ) {curveNames [Stop], curveValues [0]})]
 }
 BoostingLoop {
  path=AnimGraph.AnimGraphNode_StateMachine_1.WingsAirStatemachine.AnimStateNode_6.BoostingLoop
  nodes [
   Apply Additive(Additive=Blendspace Player 'Olympus_Wings_Boost_Roll_BS'!(X=Get PlayerGlideRoll() -> (PlayerGlideRoll:float)) -> (Pose) {blendSpace=/SunRoseFlyingGameplay/Items/Wings/Animations/Wings/Olympus_Wings_Boost_Roll_BS}) {alphaBoolBlend {blendOption=Linear}}
   Output Animation Pose(Result=Modify Curve(
     SourcePose=Sequence Player 'Olympus_Wings_Boost_Loop'() {sequence=/SunRoseFlyingGameplay/Items/Wings/Animations/Wings/Olympus_Wings_Boost_Loop}
     CurveValues_0=1
     Alpha=1
    ) {curveNames [Stop], curveValues [0]})
  ]
 }
 DiveLoop {
  path=AnimGraph.AnimGraphNode_StateMachine_1.WingsAirStatemachine.AnimStateNode_7.DiveLoop
  nodes [Output Animation Pose(Result=Sequence Player 'Olympus_Wings_Dive_Loop'() {bLoopAnimation=false, sequence=/SunRoseFlyingGameplay/Items/Wings/Animations/Wings/Olympus_Wings_Dive_Loop})]
 }
 BoostEnd {
  path=AnimGraph.AnimGraphNode_StateMachine_1.WingsAirStatemachine.AnimStateNode_8.BoostEnd
  nodes [Output Animation Pose(Result=Sequence Player 'Olympus_Wings_Boost_End'() {bLoopAnimation=false, sequence=/SunRoseFlyingGameplay/Items/Wings/Animations/Wings/Olympus_Wings_Boost_End})]
 }
 DodgeRightEnd {
  path=AnimGraph.AnimGraphNode_StateMachine_1.WingsAirStatemachine.AnimStateNode_9.DodgeRightEnd
  nodes [Output Animation Pose(Result=Sequence Player 'Olympus_Wings_Dodge_Right_Recover'() {bLoopAnimation=false, sequence=/SunRoseFlyingGameplay/Items/Wings/Animations/Wings/Olympus_Wings_Dodge_Right_Recover})]
 }
 Transition {
  path=AnimGraph.AnimGraphNode_StateMachine_1.WingsAirStatemachine.AnimStateTransitionNode_0.Transition
  nodes [Result(bCanEnterTransition=Get bPlayerIsWingsBoosting() -> (bPlayerIsWingsBoosting:bool))]
 }
 Transition {
  path=AnimGraph.AnimGraphNode_StateMachine_1.WingsAirStatemachine.AnimStateTransitionNode_1.Transition
  nodes [Result(bCanEnterTransition=Get bPlayerIsDodgeLeft() -> (bPlayerIsDodgeLeft:bool))]
 }
 Transition {
  path=AnimGraph.AnimGraphNode_StateMachine_1.WingsAirStatemachine.AnimStateTransitionNode_10.Transition
  nodes [Result(bCanEnterTransition=Get bPlayerIsWingsCrashing() -> (bPlayerIsWingsCrashing:bool))]
 }
 Transition {
  path=AnimGraph.AnimGraphNode_StateMachine_1.WingsAirStatemachine.AnimStateTransitionNode_11.Transition
  nodes [Result(bCanEnterTransition=true)]
 }
 Transition {
  path=AnimGraph.AnimGraphNode_StateMachine_1.WingsAirStatemachine.AnimStateTransitionNode_12.Transition
  nodes [Result(bCanEnterTransition=NOT Boolean(A=Get bPlayerIsWingsBoosting()))]
 }
 Transition {
  path=AnimGraph.AnimGraphNode_StateMachine_1.WingsAirStatemachine.AnimStateTransitionNode_13.Transition
  nodes [Result(bCanEnterTransition=Get bPlayerIsDodgeRight() -> (bPlayerIsDodgeRight:bool))]
 }
 Transition {
  path=AnimGraph.AnimGraphNode_StateMachine_1.WingsAirStatemachine.AnimStateTransitionNode_14.Transition
  nodes [Result(bCanEnterTransition=Get bPlayerIsDodgeLeft())]
 }
 Transition {path=AnimGraph.AnimGraphNode_StateMachine_1.WingsAirStatemachine.AnimStateTransitionNode_16.Transition, nodes [Result()]}
 Transition {
  path=AnimGraph.AnimGraphNode_StateMachine_1.WingsAirStatemachine.AnimStateTransitionNode_17.Transition
  nodes [Result(bCanEnterTransition=NOT Boolean(A=Get bPlayerIsWingsCrashing()))]
 }
 Transition {
  path=AnimGraph.AnimGraphNode_StateMachine_1.WingsAirStatemachine.AnimStateTransitionNode_18.Transition
  nodes [Result(bCanEnterTransition=Get bPlayerIsLeapInterruptible() -> (bPlayerIsLeapInterruptible:bool))]
 }
 Transition {
  path=AnimGraph.AnimGraphNode_StateMachine_1.WingsAirStatemachine.AnimStateTransitionNode_19.Transition
  nodes [Result(bCanEnterTransition=NOT Boolean(A=Get bPlayerIsWingsBoosting()))]
 }
 Transition {
  path=AnimGraph.AnimGraphNode_StateMachine_1.WingsAirStatemachine.AnimStateTransitionNode_2.Transition
  nodes [Result(bCanEnterTransition=Get bPlayerIsDodgeRight())]
 }
 Transition {path=AnimGraph.AnimGraphNode_StateMachine_1.WingsAirStatemachine.AnimStateTransitionNode_20.Transition, nodes [Result()]}
 Transition {
  path=AnimGraph.AnimGraphNode_StateMachine_1.WingsAirStatemachine.AnimStateTransitionNode_21.Transition
  nodes [Result(bCanEnterTransition=Get bIsToBoostEnd() -> (bIsToBoostEnd:bool))]
 }
 Transition {path=AnimGraph.AnimGraphNode_StateMachine_1.WingsAirStatemachine.AnimStateTransitionNode_24.Transition, nodes [Result()]}
 Transition {path=AnimGraph.AnimGraphNode_StateMachine_1.WingsAirStatemachine.AnimStateTransitionNode_25.Transition, nodes [Result()]}
 Transition {path=AnimGraph.AnimGraphNode_StateMachine_1.WingsAirStatemachine.AnimStateTransitionNode_26.Transition, nodes [Result()]}
 Transition {path=AnimGraph.AnimGraphNode_StateMachine_1.WingsAirStatemachine.AnimStateTransitionNode_27.Transition, nodes [Result()]}
 Transition {
  path=AnimGraph.AnimGraphNode_StateMachine_1.WingsAirStatemachine.AnimStateTransitionNode_3.Transition
  nodes [Result(bCanEnterTransition=Get bPlayerIsWingsCrashing())]
 }
 Transition {path=AnimGraph.AnimGraphNode_StateMachine_1.WingsAirStatemachine.AnimStateTransitionNode_4.Transition, nodes [Result()]}
 Transition {
  path=AnimGraph.AnimGraphNode_StateMachine_1.WingsAirStatemachine.AnimStateTransitionNode_5.Transition
  nodes [Result(bCanEnterTransition=true)]
 }
 Transition {
  path=AnimGraph.AnimGraphNode_StateMachine_1.WingsAirStatemachine.AnimStateTransitionNode_6.Transition
  nodes [Result(bCanEnterTransition=Get bPlayerIsBoostStartEnd() -> (bPlayerIsBoostStartEnd:bool))]
 }
 Transition {
  path=AnimGraph.AnimGraphNode_StateMachine_1.WingsAirStatemachine.AnimStateTransitionNode_7.Transition
  nodes [Result(bCanEnterTransition=Get bPlayerIsWingLeaping() -> (bPlayerIsWingLeaping:bool))]
 }
 Transition {
  path=AnimGraph.AnimGraphNode_StateMachine_1.WingsAirStatemachine.AnimStateTransitionNode_8.Transition
  nodes [Result(bCanEnterTransition=Get bIsToBoostStart() -> (bIsToBoostStart:bool))]
 }
 Transition {
  path=AnimGraph.AnimGraphNode_StateMachine_1.WingsAirStatemachine.AnimStateTransitionNode_9.Transition
  nodes [Result(bCanEnterTransition=true)]
 }
 WingsGroundAdditive {
  entry=AdditiveIdentity
  path=AnimGraph.AnimGraphNode_StateMachine_2.WingsGroundAdditive
  ids=id
  nodes [
   AnimStateNode_0 = AdditiveIdentity() {is=title AnimStateNode}
   AnimStateNode_1 = LandAdditive() {is=title AnimStateNode}
   AnimStateNode_2 = WingsReplenish() {is=title AnimStateNode}
   AnimStateTransitionNode_0 = AnimStateTransitionNode() {logicType=TLT_Inertialization}
   AnimStateTransitionNode_1 = AnimStateTransitionNode() {automaticRuleTriggerTime=0, bAutomaticRuleBasedOnSequencePlayerInState=true, logicType=TLT_Inertialization}
   AnimStateTransitionNode_2 = AnimStateTransitionNode() {crossfadeDuration=0.1, logicType=TLT_Inertialization, priorityOrder=2}
   AnimStateTransitionNode_3 = AnimStateTransitionNode() {automaticRuleTriggerTime=0, bAutomaticRuleBasedOnSequencePlayerInState=true}
  ]
  links [
   {
    dst=LandAdditive
    rule=AnimGraph.AnimGraphNode_StateMachine_2.WingsGroundAdditive.AnimStateTransitionNode_0.Transition
    src=AdditiveIdentity
    via=AnimStateTransitionNode_0
   }
   {
    dst=WingsReplenish
    rule=AnimGraph.AnimGraphNode_StateMachine_2.WingsGroundAdditive.AnimStateTransitionNode_2.Transition
    src=AdditiveIdentity
    via=AnimStateTransitionNode_2
   }
   {
    dst=AdditiveIdentity
    rule=AnimGraph.AnimGraphNode_StateMachine_2.WingsGroundAdditive.AnimStateTransitionNode_1.Transition
    src=LandAdditive
    via=AnimStateTransitionNode_1
   }
   {
    dst=AdditiveIdentity
    rule=AnimGraph.AnimGraphNode_StateMachine_2.WingsGroundAdditive.AnimStateTransitionNode_3.Transition
    src=WingsReplenish
    via=AnimStateTransitionNode_3
   }
  ]
 }
 AdditiveIdentity {
  path=AnimGraph.AnimGraphNode_StateMachine_2.WingsGroundAdditive.AnimStateNode_0.AdditiveIdentity
  nodes [Output Animation Pose(Result=Additive Identity Pose())]
 }
 LandAdditive {
  path=AnimGraph.AnimGraphNode_StateMachine_2.WingsGroundAdditive.AnimStateNode_1.LandAdditive
  nodes [Output Animation Pose(Result="Sequence Player (additive) 'Olympus_Wings_LandingAdditive'"() {bLoopAnimation=false, sequence=/SunRoseFlyingGameplay/Items/Wings/Animations/Wings/Olympus_Wings_LandingAdditive})]
 }
 WingsReplenish {
  path=AnimGraph.AnimGraphNode_StateMachine_2.WingsGroundAdditive.AnimStateNode_2.WingsReplenish
  nodes [Output Animation Pose(Result="Sequence Player (additive) 'Olympus_Wings_Cooldown_Refresh'"() {bLoopAnimation=false, sequence=/SunRoseFlyingGameplay/Items/Wings/Animations/Wings/Olympus_Wings_Cooldown_Refresh})]
 }
 Transition {
  path=AnimGraph.AnimGraphNode_StateMachine_2.WingsGroundAdditive.AnimStateTransitionNode_0.Transition
  nodes [Result(bCanEnterTransition=Get bPlayerIsPlayLandAdditive() -> (bPlayerIsPlayLandAdditive:bool))]
 }
 Transition {path=AnimGraph.AnimGraphNode_StateMachine_2.WingsGroundAdditive.AnimStateTransitionNode_1.Transition, nodes [Result()]}
 Transition {
  path=AnimGraph.AnimGraphNode_StateMachine_2.WingsGroundAdditive.AnimStateTransitionNode_2.Transition
  nodes [Result(bCanEnterTransition=Get bPlayerIsWingReplenish() -> (bPlayerIsWingReplenish:bool))]
 }
 Transition {path=AnimGraph.AnimGraphNode_StateMachine_2.WingsGroundAdditive.AnimStateTransitionNode_3.Transition, nodes [Result()]}
 EventGraph {links [], nodes []}
]
variables []
