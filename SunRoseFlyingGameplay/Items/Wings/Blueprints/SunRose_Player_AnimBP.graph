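# Player Anim Blueprint for the SunRose wings item, overriding the full-body item layer for flight and landing.
# A conduit-hub state machine handles leap, glide, boost, dodge and dive, with pitch/roll and yaw additives layered on.
# bIsFlyingStart switches between flight and a ground machine that passes through the cached input pose or plays landings, smoothed by Dead Blending.
# Other layers are empty stubs, and the driving variables are defined outside this asset.
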AnimBlueprint /SunRoseFlyingGameplay/Items/Wings/Blueprints/SunRose_Player_AnimBP
target_skeleton=&Fortnite_M_Avg_Player_Skeleton
graphs [
 AnimGraph {nodes [Output Pose() {layerGroup=None}]}
 EventGraph {links [], nodes []}
 ItemAimOffset {nodes [
   Input Pose!() -> (Pose, AimOffsetAlpha:double, Yaw:double, Pitch:double) {name=InPose_AimOffset}
   Input Pose!() -> (Pose) {name=InPose_PreAimOffsets}
   Output Pose() {layerGroup=Item Override}
  ]}
 ItemFallAdditive {nodes [
   Input Pose() {name=InPoseItemFallAdditive}
   Output Pose() {layerGroup=Item Override}
  ]}
 ItemFallLandAdditive {nodes [
   Input Pose() {name=InPoseItemFallLandAdditive}
   Output Pose() {layerGroup=Item Override}
  ]}
 ItemFallLevel2Override {nodes [
   Input Pose() {name=InPose_FallLevel2}
   Output Pose() {layerGroup=Item Override}
  ]}
 ItemFinalPoseOverride {nodes [
   Input Pose() {name=InFinalPose}
   Output Pose() {layerGroup=Item Override}
  ]}
 ItemFullBodyAimOffsets {nodes [
   Input Pose!() -> (Pose, FullBody_AimOffsetAlpha:double, Aim_Yaw:double, Aim_Pitch:double) {name=InPoseFullBodyAimOffsets}
   Input Pose!() -> (Pose) {name=InPosePreFullBodyAimOffsets}
   Output Pose() {layerGroup=Item Override}
  ]}
 ItemFullBodyOverride {nodes [
   n1 = Get GlidePitchInterp() -> (GlidePitchInterp:float)
   n2 = Get GlideRollInterp() -> (GlideRollInterp:float)
   n3 = Blend Poses by bool(
    BlendPose_0=Blendspace Player 'Olympus_Player_Glide_Boost_Pitch_Roll_BS'(X=@n2, Y=@n1) {blendSpace=/SunRoseFlyingGameplay/Items/Wings/Animations/Player/Olympus_Player_Glide_Boost_Pitch_Roll_BS}
    BlendPose_1=Blendspace Player 'Olympus_Player_Glide_Pitch_Roll_BS'(X=@n2, Y=@n1) {blendSpace=/SunRoseFlyingGameplay/Items/Wings/Animations/Player/Olympus_Player_Glide_Pitch_Roll_BS}
    BlendTime_0=0.25
    BlendTime_1=0.2
    bActiveValue=Get bIsWingsBoosting() -> (bIsWingsBoosting:bool)
   ) {blendType=Linear}
   AnimGraphNode_StateMachine_0 = FlyingStateMachine() {is=title AnimGraphNode_StateMachine, bAllowConduitEntryStates=true}
   n4 = Get PitchAlpha() -> (PitchAlpha:float)
   n5 = Blendspace Player 'Olympus_Player_Glide_Yaw_BS'!(X=Get GlideRollCombined() -> (GlideRollCombined:float)) -> (Pose) {blendSpace=/SunRoseFlyingGameplay/Items/Wings/Animations/Player/Olympus_Player_Glide_Yaw_BS}
   n6 = "Sequence Player (additive) 'Olympus_Player_Glide_Center'"() {bLoopAnimation=false, sequence=/SunRoseFlyingGameplay/Items/Wings/Animations/Player/Olympus_Player_Glide_Center}
   n7 = Get YawBlend() -> (YawBlend:float)
   n8 = Modify Curve!(
    SourcePose=Apply Additive(
     Base=Apply Additive(Base=@AnimGraphNode_StateMachine_0, Additive=@n3, Alpha=@n4) {alphaBoolBlend {blendOption=Linear}}
     Additive=Two Way Blend(A=@n6, B=@n5, Alpha=@n7) {blendNode {alphaBoolBlend {blendOption=Linear}}}
     Alpha=Get YawAlpha() -> (YawAlpha:float)
    ) {alphaBoolBlend {blendOption=Linear}}
    CurveValues_0=1
    CurveValues_1:float=1
    CurveValues_2:float=1
    Alpha=1
   ) -> (Pose) {curveNames [BothHandFK,BothLegFK,Hurdle_Land], curveValues [0,0,0]}
   AnimGraphNode_StateMachine_1 = GroundStateMachine() {is=title AnimGraphNode_StateMachine, bAllowConduitEntryStates=true}
   n9 = Get GroundBlendTime() -> (GroundBlendTime:float)
   n10 = Get bIsFlyingStart() -> (bIsFlyingStart:bool)
   Output Pose(Result=Dead Blending!(Source:Pose Link Structure=Blend Poses by bool(BlendPose_0=@n8, BlendPose_1=@AnimGraphNode_StateMachine_1, BlendTime_1=@n9, bActiveValue=@n10) {blendType=Sinusoidal, childUpateMode=ResetChildOnActivate}) -> (Pose:Pose Link Structure) {is=title AnimGraphNode_DeadBlending, defaultBlendDuration=0.1, defaultBlendMode=Linear}) {layerGroup=Item Override}
   Save cached pose 'SavedPose9831'(Pose=Input Pose() {name=InPoseFullBody}) {cacheName=SavedPose9831}
   Slot 'OverrideFullBody'() {slotName=OverrideFullBody}
  ]}
 FlyingStateMachine {
  entry=Conduit
  path=ItemFullBodyOverride.AnimGraphNode_StateMachine_0.FlyingStateMachine
  ids=id
  nodes [
   ToDodge() {is=title AnimStateAliasNode, also {aliases [BoostLoop, BoostStart, DodgeLeftEnd, DodgeRightEnd]}}
   ToBoosting() {is=title AnimStateAliasNode, also {aliases [DodgeLeftEnd, DodgeRightEnd]}}
   ToDive() {is=title AnimStateAliasNode, also {aliases [BoostEnd, DodgeLeftEnd, DodgeRightEnd, Gliding]}}
   AnimStateConduitNode_0 = Conduit()
   AnimStateNode_0 = Leap() {
    is=title AnimStateNode
    stateEntered {guid=#1}
    stateFullyBlended {guid=#2}
    stateLeft {guid=#3}
   }
   AnimStateNode_1 = DodgeRightStart() {is=title AnimStateNode}
   AnimStateNode_10 = DodgeLeftEnd() {is=title AnimStateNode}
   AnimStateNode_2 = DiveLoop() {is=title AnimStateNode}
   AnimStateNode_3 = Dive Start() {
    is=title AnimStateNode
    stateEntered {guid=#4}
    stateFullyBlended {guid=#5}
    stateLeft {guid=#6}
   }
   AnimStateNode_4 = DodgeLeftStart() {is=title AnimStateNode}
   AnimStateNode_5 = BoostStart() {is=title AnimStateNode}
   AnimStateNode_6 = BoostLoop() {
    is=title AnimStateNode
    stateEntered {guid=#7}
    stateFullyBlended {guid=#8}
    stateLeft {guid=#9}
   }
   AnimStateNode_7 = Gliding() {is=title AnimStateNode}
   AnimStateNode_8 = BoostEnd() {is=title AnimStateNode}
   AnimStateNode_9 = DodgeRightEnd() {is=title AnimStateNode}
   AnimStateTransitionNode_0 = AnimStateTransitionNode() {automaticRuleTriggerTime=0, bAutomaticRuleBasedOnSequencePlayerInState=true}
   AnimStateTransitionNode_1 = AnimStateTransitionNode() {crossfadeDuration=0, priorityOrder=4}
   AnimStateTransitionNode_10 = AnimStateTransitionNode() {crossfadeDuration=0}
   AnimStateTransitionNode_11 = AnimStateTransitionNode() {crossfadeDuration=0, priorityOrder=2}
   AnimStateTransitionNode_12 = AnimStateTransitionNode() {crossfadeDuration=0.4, logicType=TLT_Inertialization, priorityOrder=100}
   AnimStateTransitionNode_13 = AnimStateTransitionNode()
   AnimStateTransitionNode_14 = AnimStateTransitionNode() {crossfadeDuration=0.1, logicType=TLT_Inertialization}
   AnimStateTransitionNode_15 = AnimStateTransitionNode()
   AnimStateTransitionNode_18 = AnimStateTransitionNode() {automaticRuleTriggerTime=0, bAutomaticRuleBasedOnSequencePlayerInState=true, crossfadeDuration=0.1, logicType=TLT_Inertialization}
   AnimStateTransitionNode_19 = AnimStateTransitionNode()
   AnimStateTransitionNode_2 = AnimStateTransitionNode() {crossfadeDuration=0.1, logicType=TLT_Inertialization}
   AnimStateTransitionNode_20 = AnimStateTransitionNode()
   AnimStateTransitionNode_21 = AnimStateTransitionNode() {crossfadeDuration=0.4, logicType=TLT_Inertialization}
   AnimStateTransitionNode_22 = AnimStateTransitionNode() {bAutomaticRuleBasedOnSequencePlayerInState=true}
   AnimStateTransitionNode_23 = AnimStateTransitionNode()
   AnimStateTransitionNode_24 = AnimStateTransitionNode() {bAutomaticRuleBasedOnSequencePlayerInState=true, crossfadeDuration=0.1}
   AnimStateTransitionNode_25 = AnimStateTransitionNode() {logicType=TLT_Inertialization}
   AnimStateTransitionNode_27 = AnimStateTransitionNode() {bAutomaticRuleBasedOnSequencePlayerInState=true}
   AnimStateTransitionNode_28 = AnimStateTransitionNode() {bAutomaticRuleBasedOnSequencePlayerInState=true}
   AnimStateTransitionNode_3 = AnimStateTransitionNode()
   AnimStateTransitionNode_4 = AnimStateTransitionNode() {crossfadeDuration=0, priorityOrder=5}
   AnimStateTransitionNode_6 = AnimStateTransitionNode() {crossfadeDuration=0.1, logicType=TLT_Inertialization, priorityOrder=2}
   AnimStateTransitionNode_7 = AnimStateTransitionNode() {bAutomaticRuleBasedOnSequencePlayerInState=true, crossfadeDuration=0.1}
   AnimStateTransitionNode_8 = AnimStateTransitionNode()
   AnimStateTransitionNode_9 = AnimStateTransitionNode() {crossfadeDuration=0.1, logicType=TLT_Inertialization, priorityOrder=3}
  ]
  links [
   {
    dst=Conduit
    rule=ItemFullBodyOverride.AnimGraphNode_StateMachine_0.FlyingStateMachine.AnimStateTransitionNode_22.Transition
    src=BoostEnd
    via=AnimStateTransitionNode_22
   }
   {
    dst=Conduit
    rule=ItemFullBodyOverride.AnimGraphNode_StateMachine_0.FlyingStateMachine.AnimStateTransitionNode_23.Transition
    src=BoostEnd
    via=AnimStateTransitionNode_23
   }
   {
    dst=BoostEnd
    rule=ItemFullBodyOverride.AnimGraphNode_StateMachine_0.FlyingStateMachine.AnimStateTransitionNode_21.Transition
    src=BoostLoop
    via=AnimStateTransitionNode_21
   }
   {
    dst=BoostLoop
    rule=ItemFullBodyOverride.AnimGraphNode_StateMachine_0.FlyingStateMachine.AnimStateTransitionNode_14.Transition
    src=BoostStart
    via=AnimStateTransitionNode_14
   }
   {
    dst=Conduit
    rule=ItemFullBodyOverride.AnimGraphNode_StateMachine_0.FlyingStateMachine.AnimStateTransitionNode_15.Transition
    src=BoostStart
    via=AnimStateTransitionNode_15
   }
   {
    dst=BoostStart
    rule=ItemFullBodyOverride.AnimGraphNode_StateMachine_0.FlyingStateMachine.AnimStateTransitionNode_9.Transition
    src=Conduit
    via=AnimStateTransitionNode_9
   }
   {
    dst=Dive Start
    rule=ItemFullBodyOverride.AnimGraphNode_StateMachine_0.FlyingStateMachine.AnimStateTransitionNode_6.Transition
    src=Conduit
    via=AnimStateTransitionNode_6
   }
   {
    dst=DodgeLeftStart
    rule=ItemFullBodyOverride.AnimGraphNode_StateMachine_0.FlyingStateMachine.AnimStateTransitionNode_4.Transition
    src=Conduit
    via=AnimStateTransitionNode_4
   }
   {
    dst=DodgeRightStart
    rule=ItemFullBodyOverride.AnimGraphNode_StateMachine_0.FlyingStateMachine.AnimStateTransitionNode_1.Transition
    src=Conduit
    via=AnimStateTransitionNode_1
   }
   {
    dst=Gliding
    rule=ItemFullBodyOverride.AnimGraphNode_StateMachine_0.FlyingStateMachine.AnimStateTransitionNode_12.Transition
    src=Conduit
    via=AnimStateTransitionNode_12
   }
   {
    dst=Leap
    rule=ItemFullBodyOverride.AnimGraphNode_StateMachine_0.FlyingStateMachine.AnimStateTransitionNode_25.Transition
    src=Conduit
    via=AnimStateTransitionNode_25
   }
   {
    dst=Conduit
    rule=ItemFullBodyOverride.AnimGraphNode_StateMachine_0.FlyingStateMachine.AnimStateTransitionNode_8.Transition
    src=Dive Start
    via=AnimStateTransitionNode_8
   }
   {
    dst=DiveLoop
    rule=ItemFullBodyOverride.AnimGraphNode_StateMachine_0.FlyingStateMachine.AnimStateTransitionNode_18.Transition
    src=Dive Start
    via=AnimStateTransitionNode_18
   }
   {
    dst=Conduit
    rule=ItemFullBodyOverride.AnimGraphNode_StateMachine_0.FlyingStateMachine.AnimStateTransitionNode_19.Transition
    src=DiveLoop
    via=AnimStateTransitionNode_19
   }
   {
    dst=Conduit
    rule=ItemFullBodyOverride.AnimGraphNode_StateMachine_0.FlyingStateMachine.AnimStateTransitionNode_27.Transition
    src=DodgeLeftEnd
    via=AnimStateTransitionNode_27
   }
   {
    dst=DodgeLeftEnd
    rule=ItemFullBodyOverride.AnimGraphNode_StateMachine_0.FlyingStateMachine.AnimStateTransitionNode_24.Transition
    src=DodgeLeftStart
    via=AnimStateTransitionNode_24
   }
   {
    dst=Conduit
    rule=ItemFullBodyOverride.AnimGraphNode_StateMachine_0.FlyingStateMachine.AnimStateTransitionNode_28.Transition
    src=DodgeRightEnd
    via=AnimStateTransitionNode_28
   }
   {
    dst=DodgeRightEnd
    rule=ItemFullBodyOverride.AnimGraphNode_StateMachine_0.FlyingStateMachine.AnimStateTransitionNode_7.Transition
    src=DodgeRightStart
    via=AnimStateTransitionNode_7
   }
   {
    dst=Conduit
    rule=ItemFullBodyOverride.AnimGraphNode_StateMachine_0.FlyingStateMachine.AnimStateTransitionNode_13.Transition
    src=Gliding
    via=AnimStateTransitionNode_13
   }
   {
    dst=Conduit
    rule=ItemFullBodyOverride.AnimGraphNode_StateMachine_0.FlyingStateMachine.AnimStateTransitionNode_0.Transition
    src=Leap
    via=AnimStateTransitionNode_0
   }
   {
    dst=Conduit
    rule=ItemFullBodyOverride.AnimGraphNode_StateMachine_0.FlyingStateMachine.AnimStateTransitionNode_20.Transition
    src=Leap
    via=AnimStateTransitionNode_20
   }
   {
    dst=BoostStart
    rule=ItemFullBodyOverride.AnimGraphNode_StateMachine_0.FlyingStateMachine.AnimStateTransitionNode_3.Transition
    src=ToBoosting
    via=AnimStateTransitionNode_3
   }
   {
    dst=Dive Start
    rule=ItemFullBodyOverride.AnimGraphNode_StateMachine_0.FlyingStateMachine.AnimStateTransitionNode_2.Transition
    src=ToDive
    via=AnimStateTransitionNode_2
   }
   {
    dst=DodgeLeftStart
    rule=ItemFullBodyOverride.AnimGraphNode_StateMachine_0.FlyingStateMachine.AnimStateTransitionNode_11.Transition
    src=ToDodge
    via=AnimStateTransitionNode_11
   }
   {
    dst=DodgeRightStart
    rule=ItemFullBodyOverride.AnimGraphNode_StateMachine_0.FlyingStateMachine.AnimStateTransitionNode_10.Transition
    src=ToDodge
    via=AnimStateTransitionNode_10
   }
  ]
 }
 Conduit {
  path=ItemFullBodyOverride.AnimGraphNode_StateMachine_0.FlyingStateMachine.AnimStateConduitNode_0.Conduit
  nodes [Result(bCanEnterTransition=true)]
 }
 Leap {
  path=ItemFullBodyOverride.AnimGraphNode_StateMachine_0.FlyingStateMachine.AnimStateNode_0.Leap
  nodes [Output Animation Pose(Result=Sequence Player 'Olympus_Player_Takeoff'() {bLoopAnimation=false, sequence=/SunRoseFlyingGameplay/Items/Wings/Animations/Player/Olympus_Player_Takeoff})]
 }
 DodgeRightStart {
  path=ItemFullBodyOverride.AnimGraphNode_StateMachine_0.FlyingStateMachine.AnimStateNode_1.DodgeRightStart
  nodes [Output Animation Pose(Result=Sequence Player 'Olympus_Player_Dodge_Right'() {bLoopAnimation=false, sequence=/SunRoseFlyingGameplay/Items/Wings/Animations/Player/Olympus_Player_Dodge_Right})]
 }
 DodgeLeftEnd {
  path=ItemFullBodyOverride.AnimGraphNode_StateMachine_0.FlyingStateMachine.AnimStateNode_10.DodgeLeftEnd
  nodes [Output Animation Pose(Result=Sequence Player 'Olympus_Player_Dodge_Left_Recover'() {bLoopAnimation=false, sequence=/SunRoseFlyingGameplay/Items/Wings/Animations/Player/Olympus_Player_Dodge_Left_Recover})]
 }
 DiveLoop {
  path=ItemFullBodyOverride.AnimGraphNode_StateMachine_0.FlyingStateMachine.AnimStateNode_2.DiveLoop
  nodes [
   Modify Curve!(SourcePose, CurveValues_0=1, CurveValues_1=1, Alpha=1) -> (Pose) {curveNames [Roll,Yaw], curveValues [0,0]}
   Output Animation Pose(Result=Sequence Player 'Olympus_Player_Dive_Loop'() {sequence=/SunRoseFlyingGameplay/Items/Wings/Animations/Player/Olympus_Player_Dive_Loop})
  ]
 }
 Dive Start {
  path=ItemFullBodyOverride.AnimGraphNode_StateMachine_0.FlyingStateMachine.AnimStateNode_3.Dive Start
  nodes [
   Modify Curve(CurveValues_0=1, CurveValues_1=1, Alpha=1) {curveNames [Roll,Yaw], curveValues [0,0]}
   Output Animation Pose(Result=Blend Poses by bool(
     BlendPose_0=Sequence Player 'Olympus_Player_Dive_Start_HitReact'() {bLoopAnimation=false, sequence=/SunRoseFlyingGameplay/Items/Wings/Animations/Player/Olympus_Player_Dive_Start_HitReact}
     BlendPose_1=Sequence Player 'Olympus_Player_Dive'() {bLoopAnimation=false, sequence=/SunRoseFlyingGameplay/Items/Wings/Animations/Player/Olympus_Player_Dive}
     bActiveValue=Get bIsForceDive() -> (bIsForceDive:bool)
    ) {blendType=Linear, transitionType=Inertialization})
  ]
 }
 DodgeLeftStart {
  path=ItemFullBodyOverride.AnimGraphNode_StateMachine_0.FlyingStateMachine.AnimStateNode_4.DodgeLeftStart
  nodes [Output Animation Pose(Result=Sequence Player 'Olympus_Player_Dodge_Left'() {bLoopAnimation=false, sequence=/SunRoseFlyingGameplay/Items/Wings/Animations/Player/Olympus_Player_Dodge_Left})]
 }
 BoostStart {
  path=ItemFullBodyOverride.AnimGraphNode_StateMachine_0.FlyingStateMachine.AnimStateNode_5.BoostStart
  nodes [Output Animation Pose(Result=Sequence Player 'Olympus_Player_Boost_start'() {bLoopAnimation=false, sequence=/SunRoseFlyingGameplay/Items/Wings/Animations/Player/Olympus_Player_Boost_start})]
 }
 BoostLoop {
  path=ItemFullBodyOverride.AnimGraphNode_StateMachine_0.FlyingStateMachine.AnimStateNode_6.BoostLoop
  nodes [Output Animation Pose(Result=Sequence Player 'Olympus_Player_Boost_Loop'() {sequence=/SunRoseFlyingGameplay/Items/Wings/Animations/Player/Olympus_Player_Boost_Loop})]
 }
 Gliding {
  path=ItemFullBodyOverride.AnimGraphNode_StateMachine_0.FlyingStateMachine.AnimStateNode_7.Gliding
  nodes [
   Apply Additive(
    Additive="Sequence Player (additive) 'Olympus_Player_Glide_Slow'"() {bLoopAnimation=false, sequence=/SunRoseFlyingGameplay/Items/Wings/Animations/Player/Olympus_Player_Glide_Slow}
    Alpha=Get SlowSpeedAlpha() -> (SlowSpeedAlpha:float)
   ) {alphaBoolBlend {blendOption=Linear}}
   n1 = Blendspace Player 'Olympus_Player_Glide_Turbulence_BS'!(X=Get Fuel() -> (Fuel:float)) -> (Pose) {blendSpace=/SunRoseFlyingGameplay/Items/Wings/Animations/Player/Olympus_Player_Glide_Turbulence_BS}
   n2 = Component To Local(ComponentPose="Transform (Modify) Bone - Bone: Root"!(
     ComponentPose=Local To Component(LocalPose=@n1)
     Alpha=1
     Translation_X:double=0
     Translation_Y:double=0
     Translation_Z:double=Get Pelvis() -> (Pelvis:float)
     Rotation="0,0,0"
     Scale="1,1,1"
    ) -> (Pose) {translationMode=BMM_Additive, alphaBoolBlend {blendOption=Linear}, boneToModify {boneName=Root}})
   n3 = Sequence Player 'Olympus_Player_Flap_Large'() {sequence=/SunRoseFlyingGameplay/Items/Wings/Animations/Player/Olympus_Player_Flap_Large}
   n4 = Get bIsAllowWingFlap() -> (bIsAllowWingFlap:bool)
   n5 = Blend Poses by bool(
    BlendPose_0=Sequence Player 'Olympus_Player_NoFuel'() {sequence=/SunRoseFlyingGameplay/Items/Wings/Animations/Player/Olympus_Player_NoFuel}
    BlendPose_1=Two Way Blend(
     A=Blend Poses by bool(BlendPose_0=@n3, BlendPose_1=@n2, BlendTime_0=0.5, BlendTime_1=0.8, bActiveValue=@n4) {blendType=Sinusoidal, transitionType=Inertialization}
     B=Sequence Player 'Olympus_Player_Glide_Idle_High_15'() {bLoopAnimation=false, sequence=/SunRoseFlyingGameplay/Items/Wings/Animations/Player/Olympus_Player_Glide_Idle_High_15}
     Alpha=Get GlidePitchAlpha() -> (GlidePitchAlpha:float)
    ) {blendNode {alphaBoolBlend {blendOption=Linear}}}
    BlendTime_0=0.4
    BlendTime_1=0.5
    bActiveValue=Get bIsHighWithNoFuel() -> (bIsHighWithNoFuel:bool)
   ) {blendType=Sinusoidal, transitionType=Inertialization}
   n6 = Sequence Player 'Olympus_Player_Ascend_Loop'() {sequence=/SunRoseFlyingGameplay/Items/Wings/Animations/Player/Olympus_Player_Ascend_Loop}
   n7 = Get bIsAscending() -> (bIsAscending:bool)
   n8 = Get GlideRoll() -> (GlideRoll:float)
   Output Animation Pose(Result=Apply Additive(
     Base=Blend Poses by bool(BlendPose_0=@n6, BlendPose_1=@n5, BlendTime_0=0.3, BlendTime_1=0.5, bActiveValue=@n7) {blendType=Linear, transitionType=Inertialization}
     Additive=Blendspace Player 'Olympus_Player_Land_BS'!(X=@n8) -> (Pose) {blendSpace=/SunRoseFlyingGameplay/Items/Wings/Animations/Player/Olympus_Player_Land_BS}
     Alpha=Get SlowDownOrLandAlpha() -> (SlowDownOrLandAlpha:float)
    ) {alphaBoolBlend {blendOption=Linear}})
  ]
 }
 BoostEnd {
  path=ItemFullBodyOverride.AnimGraphNode_StateMachine_0.FlyingStateMachine.AnimStateNode_8.BoostEnd
  nodes [Output Animation Pose(Result=Sequence Player 'Olympus_Player_Boost_End'() {bLoopAnimation=false, sequence=/SunRoseFlyingGameplay/Items/Wings/Animations/Player/Olympus_Player_Boost_End})]
 }
 DodgeRightEnd {
  path=ItemFullBodyOverride.AnimGraphNode_StateMachine_0.FlyingStateMachine.AnimStateNode_9.DodgeRightEnd
  nodes [Output Animation Pose(Result=Sequence Player 'Olympus_Player_Dodge_Right_Recover'() {bLoopAnimation=false, sequence=/SunRoseFlyingGameplay/Items/Wings/Animations/Player/Olympus_Player_Dodge_Right_Recover})]
 }
 Transition {path=ItemFullBodyOverride.AnimGraphNode_StateMachine_0.FlyingStateMachine.AnimStateTransitionNode_0.Transition, nodes [Result()]}
 Transition {
  path=ItemFullBodyOverride.AnimGraphNode_StateMachine_0.FlyingStateMachine.AnimStateTransitionNode_1.Transition
  nodes [Result(bCanEnterTransition=Get bIsDodgeRight() -> (bIsDodgeRight:bool))]
 }
 Transition {
  path=ItemFullBodyOverride.AnimGraphNode_StateMachine_0.FlyingStateMachine.AnimStateTransitionNode_10.Transition
  nodes [Result(bCanEnterTransition=Get bIsDodgeRight())]
 }
 Transition {
  path=ItemFullBodyOverride.AnimGraphNode_StateMachine_0.FlyingStateMachine.AnimStateTransitionNode_11.Transition
  nodes [Result(bCanEnterTransition=Get bIsDodgeLeft() -> (bIsDodgeLeft:bool))]
 }
 Transition {
  path=ItemFullBodyOverride.AnimGraphNode_StateMachine_0.FlyingStateMachine.AnimStateTransitionNode_12.Transition
  nodes [Result(bCanEnterTransition=true)]
 }
 Transition {
  path=ItemFullBodyOverride.AnimGraphNode_StateMachine_0.FlyingStateMachine.AnimStateTransitionNode_13.Transition
  nodes [Result(bCanEnterTransition=true)]
 }
 Transition {
  path=ItemFullBodyOverride.AnimGraphNode_StateMachine_0.FlyingStateMachine.AnimStateTransitionNode_14.Transition
  nodes [Result(bCanEnterTransition=Get bIsBoostStartEnd() -> (bIsBoostStartEnd:bool))]
 }
 Transition {
  path=ItemFullBodyOverride.AnimGraphNode_StateMachine_0.FlyingStateMachine.AnimStateTransitionNode_15.Transition
  nodes [Result(bCanEnterTransition=NOT Boolean(A=Get bIsWingsBoosting()))]
 }
 Transition {path=ItemFullBodyOverride.AnimGraphNode_StateMachine_0.FlyingStateMachine.AnimStateTransitionNode_18.Transition, nodes [Result()]}
 Transition {
  path=ItemFullBodyOverride.AnimGraphNode_StateMachine_0.FlyingStateMachine.AnimStateTransitionNode_19.Transition
  nodes [Result(bCanEnterTransition=NOT Boolean(A=Get bIsWingsCrashing() -> (bIsWingsCrashing:bool)))]
 }
 Transition {
  path=ItemFullBodyOverride.AnimGraphNode_StateMachine_0.FlyingStateMachine.AnimStateTransitionNode_2.Transition
  nodes [Result(bCanEnterTransition=Get bIsWingsCrashing())]
 }
 Transition {
  path=ItemFullBodyOverride.AnimGraphNode_StateMachine_0.FlyingStateMachine.AnimStateTransitionNode_20.Transition
  nodes [Result(bCanEnterTransition=Get bIsLeapInterruptible() -> (bIsLeapInterruptible:bool))]
 }
 Transition {
  path=ItemFullBodyOverride.AnimGraphNode_StateMachine_0.FlyingStateMachine.AnimStateTransitionNode_21.Transition
  nodes [Result(bCanEnterTransition=NOT Boolean(A=Get bIsWingsBoosting()))]
 }
 Transition {path=ItemFullBodyOverride.AnimGraphNode_StateMachine_0.FlyingStateMachine.AnimStateTransitionNode_22.Transition, nodes [Result()]}
 Transition {
  path=ItemFullBodyOverride.AnimGraphNode_StateMachine_0.FlyingStateMachine.AnimStateTransitionNode_23.Transition
  nodes [Result(bCanEnterTransition=Get bIsExtBoostEarly() -> (bIsExtBoostEarly:bool))]
 }
 Transition {path=ItemFullBodyOverride.AnimGraphNode_StateMachine_0.FlyingStateMachine.AnimStateTransitionNode_24.Transition, nodes [Result()]}
 Transition {
  path=ItemFullBodyOverride.AnimGraphNode_StateMachine_0.FlyingStateMachine.AnimStateTransitionNode_25.Transition
  nodes [Result(bCanEnterTransition=Get bIsWingLeaping() -> (bIsWingLeaping:bool))]
 }
 Transition {path=ItemFullBodyOverride.AnimGraphNode_StateMachine_0.FlyingStateMachine.AnimStateTransitionNode_27.Transition, nodes [Result()]}
 Transition {path=ItemFullBodyOverride.AnimGraphNode_StateMachine_0.FlyingStateMachine.AnimStateTransitionNode_28.Transition, nodes [Result()]}
 Transition {
  path=ItemFullBodyOverride.AnimGraphNode_StateMachine_0.FlyingStateMachine.AnimStateTransitionNode_3.Transition
  nodes [Result(bCanEnterTransition=Get bToBoosting() -> (bToBoosting:bool))]
 }
 Transition {
  path=ItemFullBodyOverride.AnimGraphNode_StateMachine_0.FlyingStateMachine.AnimStateTransitionNode_4.Transition
  nodes [Result(bCanEnterTransition=Get bIsDodgeLeft())]
 }
 Transition {
  path=ItemFullBodyOverride.AnimGraphNode_StateMachine_0.FlyingStateMachine.AnimStateTransitionNode_6.Transition
  nodes [Result(bCanEnterTransition=Get bIsDiveStart() -> (bIsDiveStart:bool))]
 }
 Transition {path=ItemFullBodyOverride.AnimGraphNode_StateMachine_0.FlyingStateMachine.AnimStateTransitionNode_7.Transition, nodes [Result()]}
 Transition {
  path=ItemFullBodyOverride.AnimGraphNode_StateMachine_0.FlyingStateMachine.AnimStateTransitionNode_8.Transition
  nodes [Result(bCanEnterTransition=NOT Boolean(A=Get bIsWingsCrashing()))]
 }
 Transition {
  path=ItemFullBodyOverride.AnimGraphNode_StateMachine_0.FlyingStateMachine.AnimStateTransitionNode_9.Transition
  nodes [Result(bCanEnterTransition=Get bIsWingsBoosting())]
 }
 GroundStateMachine {
  entry=Conduit
  path=ItemFullBodyOverride.AnimGraphNode_StateMachine_1.GroundStateMachine
  ids=id
  nodes [
   AnimStateConduitNode_0 = Conduit()
   AnimStateNode_0 = LandAdditve() {is=title AnimStateNode}
   AnimStateNode_1 = Ground Navigation() {is=title AnimStateNode}
   AnimStateNode_2 = DiveHitToIdle() {is=title AnimStateNode}
   AnimStateTransitionNode_0 = AnimStateTransitionNode() {crossfadeDuration=0.3, logicType=TLT_Inertialization}
   AnimStateTransitionNode_1 = AnimStateTransitionNode() {logicType=TLT_Inertialization, priorityOrder=100}
   AnimStateTransitionNode_2 = AnimStateTransitionNode() {automaticRuleTriggerTime=0, bAutomaticRuleBasedOnSequencePlayerInState=true, logicType=TLT_Inertialization}
   AnimStateTransitionNode_3 = AnimStateTransitionNode() {automaticRuleTriggerTime=0, logicType=TLT_Inertialization, priorityOrder=2}
   AnimStateTransitionNode_4 = AnimStateTransitionNode() {automaticRuleTriggerTime=0, bAutomaticRuleBasedOnSequencePlayerInState=true, crossfadeDuration=0.25, logicType=TLT_Inertialization}
   AnimStateTransitionNode_5 = AnimStateTransitionNode() {crossfadeDuration=0.25, logicType=TLT_Inertialization, priorityOrder=2}
  ]
  links [
   {
    dst=DiveHitToIdle
    rule=ItemFullBodyOverride.AnimGraphNode_StateMachine_1.GroundStateMachine.AnimStateTransitionNode_3.Transition
    src=Conduit
    via=AnimStateTransitionNode_3
   }
   {
    dst=Ground Navigation
    rule=ItemFullBodyOverride.AnimGraphNode_StateMachine_1.GroundStateMachine.AnimStateTransitionNode_1.Transition
    src=Conduit
    via=AnimStateTransitionNode_1
   }
   {
    dst=LandAdditve
    rule=ItemFullBodyOverride.AnimGraphNode_StateMachine_1.GroundStateMachine.AnimStateTransitionNode_0.Transition
    src=Conduit
    via=AnimStateTransitionNode_0
   }
   {
    dst=Ground Navigation
    rule=ItemFullBodyOverride.AnimGraphNode_StateMachine_1.GroundStateMachine.AnimStateTransitionNode_4.Transition
    src=DiveHitToIdle
    via=AnimStateTransitionNode_4
   }
   {
    dst=Ground Navigation
    rule=ItemFullBodyOverride.AnimGraphNode_StateMachine_1.GroundStateMachine.AnimStateTransitionNode_5.Transition
    src=DiveHitToIdle
    via=AnimStateTransitionNode_5
   }
   {
    dst=Ground Navigation
    rule=ItemFullBodyOverride.AnimGraphNode_StateMachine_1.GroundStateMachine.AnimStateTransitionNode_2.Transition
    src=LandAdditve
    via=AnimStateTransitionNode_2
   }
  ]
 }
 Conduit {
  path=ItemFullBodyOverride.AnimGraphNode_StateMachine_1.GroundStateMachine.AnimStateConduitNode_0.Conduit
  nodes [Result(bCanEnterTransition=true)]
 }
 LandAdditve {
  path=ItemFullBodyOverride.AnimGraphNode_StateMachine_1.GroundStateMachine.AnimStateNode_0.LandAdditve
  nodes [
   Layered blend per bone() {
    blendMode=BlendMask
    curveBlendOption=UseMaxValue
    blendMasks [&Fortnite_M_Avg_Player_Skeleton:FirstPersonAdditives]
    layerSetup []
   }
   Output Animation Pose(Result=Apply Additive(
     Base=Use cached pose 'SavedPose9831'()
     Additive="Sequence Player (additive) 'Olympus_Player_LandingAdditive'"() {bLoopAnimation=false, sequence=/SunRoseFlyingGameplay/Items/Wings/Animations/Player/Olympus_Player_LandingAdditive}
    ) {alphaBoolBlend {blendOption=Linear}})
  ]
 }
 Ground Navigation {
  path=ItemFullBodyOverride.AnimGraphNode_StateMachine_1.GroundStateMachine.AnimStateNode_1.Ground Navigation
  nodes [Output Animation Pose(Result=Use cached pose 'SavedPose9831'())]
 }
 DiveHitToIdle {
  path=ItemFullBodyOverride.AnimGraphNode_StateMachine_1.GroundStateMachine.AnimStateNode_2.DiveHitToIdle
  nodes [Output Animation Pose(Result=Sequence Player 'Olympus_Player_Dive_Hit_To_Idle'() {bLoopAnimation=false, sequence=/SunRoseFlyingGameplay/Items/Wings/Animations/Player/Olympus_Player_Dive_Hit_To_Idle})]
 }
 Transition {
  path=ItemFullBodyOverride.AnimGraphNode_StateMachine_1.GroundStateMachine.AnimStateTransitionNode_0.Transition
  nodes [Result(bCanEnterTransition=Get bIsPlayLandAdditive() -> (bIsPlayLandAdditive:bool))]
 }
 Transition {
  path=ItemFullBodyOverride.AnimGraphNode_StateMachine_1.GroundStateMachine.AnimStateTransitionNode_1.Transition
  nodes [Result(bCanEnterTransition=true)]
 }
 Transition {path=ItemFullBodyOverride.AnimGraphNode_StateMachine_1.GroundStateMachine.AnimStateTransitionNode_2.Transition, nodes [Result()]}
 Transition {
  path=ItemFullBodyOverride.AnimGraphNode_StateMachine_1.GroundStateMachine.AnimStateTransitionNode_3.Transition
  nodes [Result(bCanEnterTransition=Get bIsPlayDiveHitToIdle() -> (bIsPlayDiveHitToIdle:bool))]
 }
 Transition {path=ItemFullBodyOverride.AnimGraphNode_StateMachine_1.GroundStateMachine.AnimStateTransitionNode_4.Transition, nodes [Result()]}
 Transition {
  path=ItemFullBodyOverride.AnimGraphNode_StateMachine_1.GroundStateMachine.AnimStateTransitionNode_5.Transition
  nodes [Result(bCanEnterTransition=Get bIsAccelerating2D())]
 }
 ItemIdleOverride {nodes [
   Input Pose() {name=InPose_Idle}
   Output Pose() {layerGroup=Item Override}
  ]}
 ItemJetPackJumpAdditive {nodes [Output Pose() {layerGroup=Item Override}]}
 ItemJetPackStartAdditive {nodes [Output Pose() {layerGroup=Item Override}]}
 ItemJumpLoopAdditive {nodes [
   Input Pose() {name=InPoseItemJumpLoopAdditive}
   Output Pose() {layerGroup=Item Override}
  ]}
 ItemJumpOverride {nodes [
   Input Pose() {name=InPoseJump}
   Output Pose() {layerGroup=Item Override}
  ]}
 ItemJumpUpAdditive {nodes [
   Input Pose() {name=InPoseItemJumpUpAdditive}
   Output Pose() {layerGroup=Item Override}
  ]}
 ItemPreAimOffsets {nodes [
   Input Pose() {name=InPosePreAimOffsets}
   Output Pose() {layerGroup=Item Override}
  ]}
 ItemPreIK {nodes [
   Input Pose() {name=InPose_PreIK}
   Output Pose() {layerGroup=Item Override}
  ]}
 ItemRightHandIK {nodes [
   Input Pose() {name=InPosePreRightHandIK}
   Input Pose() {name=InPosePostRightHandIK}
   Output Pose() {layerGroup=Item Override}
  ]}
 ItemSkeletalControl {nodes [
   Input Pose()
   Output Pose() {layerGroup=Item Override}
  ]}
 ItemSwimJumpFallLoop {nodes [Output Pose() {layerGroup=Item Override}]}
 ItemSwimJumpSurfaceEnd {nodes [Output Pose() {layerGroup=Item Override}]}
 ItemSwimJumpSurfaceLoop {nodes [Output Pose() {layerGroup=Item Override}]}
 ItemUpperBody {nodes [
   Input Pose!() -> (Pose, InputParam:Fort Anim Input Adjusted Aim Structure) {name=InPoseUpperBody}
   Output Pose() {layerGroup=Item Override}
  ]}
 ItemZipLineStartAdditive {nodes [Output Pose() {layerGroup=Item Override}]}
]
variables []
