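# Animation Blueprint for the SunRose wings item, animating the wings to match the player's ground and flight movement.
# Conduit-hubbed state machines cover ground locomotion, with an additive layer, and air moves like glide, boost, dive and dodge.
# The cached ground pose passes through a linked procedural AnimBP unless deactivated; the air pose gets additive glide roll.
# IsFlyingStart picks air or ground using inertialization, before the WingsAir slot and Dead Blending.
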AnimBlueprint /SunRoseFlyingGameplay/Items/Wings/Blueprints/SunRose_Wings_AnimBP_PreNativized
target_skeleton=/SunRoseFlyingGameplay/Items/Wings/Meshes/Wings_SunRose_Gadget_Skeleton
graphs [
 AnimGraph {nodes [
   n1 = SunRose_Wings_Procedural_AnimBP!(InPose:Pose Link Structure=Use cached pose 'WingsGroundPose'()) -> (Pose:Pose Link Structure) {
    is=title AnimGraphNode_LinkedAnimGraph
    bPropagateNotifiesToLinkedInstances=true
    bReceiveNotifiesFromLinkedInstances=true
    instanceClass=&/SunRose_Wings_Procedural_AnimBP.SunRose_Wings_Procedural_AnimBP_C
   }
   n2 = Blendspace Player 'Olympus_Wings_Glide_Roll_BS'!(X=Get PlayerGlide Roll Combined() -> (PlayerGlide Roll Combined:double)) -> (Pose) {blendSpace=/SunRoseFlyingGameplay/Items/Wings/Animations/Wings/Olympus_Wings_Glide_Roll_BS}
   AnimGraphNode_StateMachine_1 = WingsAirStatemachine() {is=title AnimGraphNode_StateMachine, bAllowConduitEntryStates=true}
   n3 = Get PlayerRoll Alpha() -> (PlayerRoll Alpha:float)
   n4 = Blend Poses by bool(
    BlendPose_0=Modify Curve(
     SourcePose=Apply Additive(Base=@AnimGraphNode_StateMachine_1, Additive=@n2, Alpha=@n3) {alphaBoolBlend {blendOption=Linear}}
     CurveValues_0=1
     Alpha=1
    ) {curveNames [Hurdle_Land], curveValues [0]}
    BlendPose_1=Blend Poses by bool(
     BlendPose_0=Use cached pose 'WingsGroundPose'()
     BlendPose_1=@n1
     bActiveValue=Get IsProceduralDeactivated() -> (IsProceduralDeactivated:bool)
    ) {blendType=Linear, childUpateMode=ResetChildOnActivate, transitionType=Inertialization}
    BlendTime_1=Get PlayerGround Blend Time() -> (PlayerGround Blend Time:double)
    bActiveValue=Get IsFlyingStart() -> (IsFlyingStart:bool)
   ) {blendType=Linear, childUpateMode=ResetChildOnActivate, transitionType=Inertialization}
   Output Pose(Result=Dead Blending!(Source:Pose Link Structure=Slot 'WingsAir'(Source=@n4) {slotName=WingsAir}) -> (Pose:Pose Link Structure) {is=title AnimGraphNode_DeadBlending, defaultBlendDuration=0.1, defaultBlendMode=Linear}) {layerGroup=None}
   AnimGraphNode_StateMachine_0 = WingsGroundStatemachine() {is=title AnimGraphNode_StateMachine, bAllowConduitEntryStates=true}
   AnimGraphNode_StateMachine_2 = WingsGroundAdditive() {is=title AnimGraphNode_StateMachine}
   Save cached pose 'WingsGroundPose'(Pose=Slot 'DefaultSlot'(Source=Apply Additive(Base=@AnimGraphNode_StateMachine_0, Additive=@AnimGraphNode_StateMachine_2) {alphaBoolBlend {blendOption=Linear}})) {cacheName=WingsGroundPose}
  ]}
 WingsGroundStatemachine {
  entry=Conduit
  path=AnimGraph.AnimGraphNode_StateMachine_0.WingsGroundStatemachine
  ids=id
  nodes [
   FromTacticalJump() {is=title AnimStateAliasNode, also {aliases [TacticalJumpLoop, TacticalJumpStart]}}
   ToSliding() {is=title AnimStateAliasNode, also {aliases [JumpLand, Moving, TacticalLand, TacticalSprint]}}
   ToZipline() {
    is=title AnimStateAliasNode
    also {aliases [FallLoop, FallStart, JumpLand, JumpStart, TacticalJumpLoop, TacticalJumpStart, TacticalLand, TacticalSprint]}
   }
   AnimStateConduitNode_0 = Conduit()
   AnimStateNode_0 = Swim() {is=title AnimStateNode}
   AnimStateNode_1 = Idle()
   AnimStateNode_10 = TacticalLand() {is=title AnimStateNode}
   AnimStateNode_11 = SlidingPose() {
    is=title AnimStateNode
    stateEntered {guid=#1}
    stateFullyBlended {guid=#2}
    stateLeft {guid=#3}
   }
   AnimStateNode_12 = CrouchIdle() {is=title AnimStateNode}
   AnimStateNode_13 = Crouch Moving() {is=title AnimStateNode}
   AnimStateNode_14 = SlideStart() {
    is=title AnimStateNode
    stateEntered {guid=#4}
    stateFullyBlended {guid=#5}
    stateLeft {guid=#6}
   }
   AnimStateNode_15 = Zipline() {is=title AnimStateNode}
   AnimStateNode_16 = DiveHitToIdle() {is=title AnimStateNode}
   AnimStateNode_17 = GrindRail() {is=title AnimStateNode}
   AnimStateNode_2 = JumpStart() {is=title AnimStateNode}
   AnimStateNode_3 = FallLoop() {is=title AnimStateNode}
   AnimStateNode_4 = JumpLand() {is=title AnimStateNode}
   AnimStateNode_5 = Moving() {is=title AnimStateNode}
   AnimStateNode_6 = FallStart() {is=title AnimStateNode}
   AnimStateNode_7 = TacticalSprint() {is=title AnimStateNode}
   AnimStateNode_8 = TacticalJumpStart() {is=title AnimStateNode}
   AnimStateNode_9 = TacticalJumpLoop() {is=title AnimStateNode}
   AnimStateTransitionNode_0 = AnimStateTransitionNode()
   AnimStateTransitionNode_1 = AnimStateTransitionNode()
   AnimStateTransitionNode_10 = AnimStateTransitionNode() {logicType=TLT_Inertialization}
   AnimStateTransitionNode_11 = AnimStateTransitionNode() {automaticRuleTriggerTime=0, bAutomaticRuleBasedOnSequencePlayerInState=true, crossfadeDuration=0.8, logicType=TLT_Inertialization}
   AnimStateTransitionNode_12 = AnimStateTransitionNode() {crossfadeDuration=0.1, logicType=TLT_Inertialization}
   AnimStateTransitionNode_13 = AnimStateTransitionNode() {automaticRuleTriggerTime=0, logicType=TLT_Inertialization}
   AnimStateTransitionNode_14 = AnimStateTransitionNode() {automaticRuleTriggerTime=0, bAutomaticRuleBasedOnSequencePlayerInState=true, logicType=TLT_Inertialization}
   AnimStateTransitionNode_15 = AnimStateTransitionNode()
   AnimStateTransitionNode_16 = AnimStateTransitionNode()
   AnimStateTransitionNode_17 = AnimStateTransitionNode() {logicType=TLT_Inertialization, priorityOrder=8}
   AnimStateTransitionNode_18 = AnimStateTransitionNode()
   AnimStateTransitionNode_19 = AnimStateTransitionNode() {logicType=TLT_Inertialization}
   AnimStateTransitionNode_2 = AnimStateTransitionNode() {crossfadeDuration=0.4, logicType=TLT_Inertialization, priorityOrder=100}
   AnimStateTransitionNode_20 = AnimStateTransitionNode() {automaticRuleTriggerTime=0, bAutomaticRuleBasedOnSequencePlayerInState=true, logicType=TLT_Inertialization}
   AnimStateTransitionNode_21 = AnimStateTransitionNode()
   AnimStateTransitionNode_22 = AnimStateTransitionNode() {priorityOrder=2}
   AnimStateTransitionNode_23 = AnimStateTransitionNode() {crossfadeDuration=0.4, logicType=TLT_Inertialization, priorityOrder=20}
   AnimStateTransitionNode_24 = AnimStateTransitionNode() {crossfadeDuration=0.1, logicType=TLT_Inertialization}
   AnimStateTransitionNode_25 = AnimStateTransitionNode() {crossfadeDuration=0.25, logicType=TLT_Inertialization, priorityOrder=2}
   AnimStateTransitionNode_26 = AnimStateTransitionNode() {crossfadeDuration=0.4, logicType=TLT_Inertialization, priorityOrder=4}
   AnimStateTransitionNode_27 = AnimStateTransitionNode() {crossfadeDuration=0.4, logicType=TLT_Inertialization, priorityOrder=5}
   AnimStateTransitionNode_28 = AnimStateTransitionNode() {crossfadeDuration=0.4, logicType=TLT_Inertialization}
   AnimStateTransitionNode_29 = AnimStateTransitionNode() {crossfadeDuration=0.4, logicType=TLT_Inertialization}
   AnimStateTransitionNode_3 = AnimStateTransitionNode() {automaticRuleTriggerTime=0, bAutomaticRuleBasedOnSequencePlayerInState=true, crossfadeDuration=0.3, logicType=TLT_Inertialization}
   AnimStateTransitionNode_30 = AnimStateTransitionNode()
   AnimStateTransitionNode_31 = AnimStateTransitionNode()
   AnimStateTransitionNode_32 = AnimStateTransitionNode() {automaticRuleTriggerTime=0, bAutomaticRuleBasedOnSequencePlayerInState=true}
   AnimStateTransitionNode_33 = AnimStateTransitionNode() {logicType=TLT_Inertialization, priorityOrder=10}
   AnimStateTransitionNode_34 = AnimStateTransitionNode() {
    bAutomaticRuleBasedOnSequencePlayerInState=true
    logicType=TLT_Inertialization
    transitionEnd {guid=#7}
    transitionInterrupt {guid=#8}
    transitionStart {guid=#9}
   }
   AnimStateTransitionNode_36 = AnimStateTransitionNode() {crossfadeDuration=0.4, logicType=TLT_Inertialization, priorityOrder=10}
   AnimStateTransitionNode_37 = AnimStateTransitionNode() {logicType=TLT_Inertialization}
   AnimStateTransitionNode_38 = AnimStateTransitionNode()
   AnimStateTransitionNode_39 = AnimStateTransitionNode()
   AnimStateTransitionNode_4 = AnimStateTransitionNode() {logicType=TLT_Inertialization, priorityOrder=9}
   AnimStateTransitionNode_40 = AnimStateTransitionNode() {logicType=TLT_Inertialization, priorityOrder=2}
   AnimStateTransitionNode_41 = AnimStateTransitionNode()
   AnimStateTransitionNode_42 = AnimStateTransitionNode() {logicType=TLT_Inertialization}
   AnimStateTransitionNode_43 = AnimStateTransitionNode()
   AnimStateTransitionNode_5 = AnimStateTransitionNode() {logicType=TLT_Inertialization, priorityOrder=3}
   AnimStateTransitionNode_6 = AnimStateTransitionNode() {logicType=TLT_Inertialization}
   AnimStateTransitionNode_7 = AnimStateTransitionNode() {logicType=TLT_Inertialization, priorityOrder=6}
   AnimStateTransitionNode_8 = AnimStateTransitionNode() {logicType=TLT_Inertialization, priorityOrder=7}
   AnimStateTransitionNode_9 = AnimStateTransitionNode() {automaticRuleTriggerTime=0, bAutomaticRuleBasedOnSequencePlayerInState=true}
  ]
  links [
   {
    dst=Crouch Moving
    rule=AnimGraph.AnimGraphNode_StateMachine_0.WingsGroundStatemachine.AnimStateTransitionNode_27.Transition
    src=Conduit
    via=AnimStateTransitionNode_27
   }
   {
    dst=CrouchIdle
    rule=AnimGraph.AnimGraphNode_StateMachine_0.WingsGroundStatemachine.AnimStateTransitionNode_26.Transition
    src=Conduit
    via=AnimStateTransitionNode_26
   }
   {
    dst=DiveHitToIdle
    rule=AnimGraph.AnimGraphNode_StateMachine_0.WingsGroundStatemachine.AnimStateTransitionNode_17.Transition
    src=Conduit
    via=AnimStateTransitionNode_17
   }
   {
    dst=GrindRail
    rule=AnimGraph.AnimGraphNode_StateMachine_0.WingsGroundStatemachine.AnimStateTransitionNode_36.Transition
    src=Conduit
    via=AnimStateTransitionNode_36
   }
   {
    dst=Idle
    rule=AnimGraph.AnimGraphNode_StateMachine_0.WingsGroundStatemachine.AnimStateTransitionNode_2.Transition
    src=Conduit
    via=AnimStateTransitionNode_2
   }
   {
    dst=JumpStart
    rule=AnimGraph.AnimGraphNode_StateMachine_0.WingsGroundStatemachine.AnimStateTransitionNode_8.Transition
    src=Conduit
    via=AnimStateTransitionNode_8
   }
   {
    dst=Moving
    rule=AnimGraph.AnimGraphNode_StateMachine_0.WingsGroundStatemachine.AnimStateTransitionNode_7.Transition
    src=Conduit
    via=AnimStateTransitionNode_7
   }
   {
    dst=SlideStart
    rule=AnimGraph.AnimGraphNode_StateMachine_0.WingsGroundStatemachine.AnimStateTransitionNode_37.Transition
    src=Conduit
    via=AnimStateTransitionNode_37
   }
   {
    dst=Swim
    rule=AnimGraph.AnimGraphNode_StateMachine_0.WingsGroundStatemachine.AnimStateTransitionNode_5.Transition
    src=Conduit
    via=AnimStateTransitionNode_5
   }
   {
    dst=TacticalSprint
    rule=AnimGraph.AnimGraphNode_StateMachine_0.WingsGroundStatemachine.AnimStateTransitionNode_4.Transition
    src=Conduit
    via=AnimStateTransitionNode_4
   }
   {
    dst=Zipline
    rule=AnimGraph.AnimGraphNode_StateMachine_0.WingsGroundStatemachine.AnimStateTransitionNode_40.Transition
    src=Conduit
    via=AnimStateTransitionNode_40
   }
   {
    dst=Conduit
    rule=AnimGraph.AnimGraphNode_StateMachine_0.WingsGroundStatemachine.AnimStateTransitionNode_31.Transition
    src=Crouch Moving
    via=AnimStateTransitionNode_31
   }
   {
    dst=CrouchIdle
    rule=AnimGraph.AnimGraphNode_StateMachine_0.WingsGroundStatemachine.AnimStateTransitionNode_29.Transition
    src=Crouch Moving
    via=AnimStateTransitionNode_29
   }
   {
    dst=Conduit
    rule=AnimGraph.AnimGraphNode_StateMachine_0.WingsGroundStatemachine.AnimStateTransitionNode_30.Transition
    src=CrouchIdle
    via=AnimStateTransitionNode_30
   }
   {
    dst=Crouch Moving
    rule=AnimGraph.AnimGraphNode_StateMachine_0.WingsGroundStatemachine.AnimStateTransitionNode_28.Transition
    src=CrouchIdle
    via=AnimStateTransitionNode_28
   }
   {
    dst=Conduit
    rule=AnimGraph.AnimGraphNode_StateMachine_0.WingsGroundStatemachine.AnimStateTransitionNode_32.Transition
    src=DiveHitToIdle
    via=AnimStateTransitionNode_32
   }
   {
    dst=JumpLand
    rule=AnimGraph.AnimGraphNode_StateMachine_0.WingsGroundStatemachine.AnimStateTransitionNode_12.Transition
    src=FallLoop
    via=AnimStateTransitionNode_12
   }
   {
    dst=FallLoop
    rule=AnimGraph.AnimGraphNode_StateMachine_0.WingsGroundStatemachine.AnimStateTransitionNode_11.Transition
    src=FallStart
    via=AnimStateTransitionNode_11
   }
   {
    dst=JumpLand
    rule=AnimGraph.AnimGraphNode_StateMachine_0.WingsGroundStatemachine.AnimStateTransitionNode_24.Transition
    src=FallStart
    via=AnimStateTransitionNode_24
   }
   {
    dst=FallLoop
    rule=AnimGraph.AnimGraphNode_StateMachine_0.WingsGroundStatemachine.AnimStateTransitionNode_23.Transition
    src=FromTacticalJump
    via=AnimStateTransitionNode_23
   }
   {
    dst=Conduit
    rule=AnimGraph.AnimGraphNode_StateMachine_0.WingsGroundStatemachine.AnimStateTransitionNode_43.Transition
    src=GrindRail
    via=AnimStateTransitionNode_43
   }
   {
    dst=Conduit
    rule=AnimGraph.AnimGraphNode_StateMachine_0.WingsGroundStatemachine.AnimStateTransitionNode_1.Transition
    src=Idle
    via=AnimStateTransitionNode_1
   }
   {
    dst=Conduit
    rule=AnimGraph.AnimGraphNode_StateMachine_0.WingsGroundStatemachine.AnimStateTransitionNode_9.Transition
    src=JumpLand
    via=AnimStateTransitionNode_9
   }
   {
    dst=JumpStart
    rule=AnimGraph.AnimGraphNode_StateMachine_0.WingsGroundStatemachine.AnimStateTransitionNode_6.AnimationTransitionGraph_1
    src=JumpLand
    via=AnimStateTransitionNode_6
   }
   {
    dst=FallStart
    rule=AnimGraph.AnimGraphNode_StateMachine_0.WingsGroundStatemachine.AnimStateTransitionNode_3.AnimationTransitionGraph_1
    src=JumpStart
    via=AnimStateTransitionNode_3
   }
   {
    dst=Conduit
    rule=AnimGraph.AnimGraphNode_StateMachine_0.WingsGroundStatemachine.AnimStateTransitionNode_0.Transition
    src=Moving
    via=AnimStateTransitionNode_0
   }
   {
    dst=FallStart
    rule=AnimGraph.AnimGraphNode_StateMachine_0.WingsGroundStatemachine.AnimStateTransitionNode_25.Transition
    src=Moving
    via=AnimStateTransitionNode_25
   }
   {
    dst=JumpStart
    rule=AnimGraph.AnimGraphNode_StateMachine_0.WingsGroundStatemachine.AnimStateTransitionNode_10.Transition
    src=Moving
    via=AnimStateTransitionNode_10
   }
   {
    dst=Conduit
    rule=AnimGraph.AnimGraphNode_StateMachine_0.WingsGroundStatemachine.AnimStateTransitionNode_38.Transition
    src=SlideStart
    via=AnimStateTransitionNode_38
   }
   {
    dst=SlidingPose
    rule=AnimGraph.AnimGraphNode_StateMachine_0.WingsGroundStatemachine.AnimStateTransitionNode_34.Transition
    src=SlideStart
    via=AnimStateTransitionNode_34
   }
   {
    dst=Conduit
    rule=AnimGraph.AnimGraphNode_StateMachine_0.WingsGroundStatemachine.AnimStateTransitionNode_39.Transition
    src=SlidingPose
    via=AnimStateTransitionNode_39
   }
   {
    dst=Conduit
    rule=AnimGraph.AnimGraphNode_StateMachine_0.WingsGroundStatemachine.AnimStateTransitionNode_16.Transition
    src=Swim
    via=AnimStateTransitionNode_16
   }
   {
    dst=Conduit
    rule=AnimGraph.AnimGraphNode_StateMachine_0.WingsGroundStatemachine.AnimStateTransitionNode_15.Transition
    src=TacticalJumpLoop
    via=AnimStateTransitionNode_15
   }
   {
    dst=TacticalLand
    rule=AnimGraph.AnimGraphNode_StateMachine_0.WingsGroundStatemachine.AnimStateTransitionNode_19.Transition
    src=TacticalJumpLoop
    via=AnimStateTransitionNode_19
   }
   {
    dst=TacticalJumpLoop
    rule=AnimGraph.AnimGraphNode_StateMachine_0.WingsGroundStatemachine.AnimStateTransitionNode_14.Transition
    src=TacticalJumpStart
    via=AnimStateTransitionNode_14
   }
   {
    dst=Conduit
    rule=AnimGraph.AnimGraphNode_StateMachine_0.WingsGroundStatemachine.AnimStateTransitionNode_21.Transition
    src=TacticalLand
    via=AnimStateTransitionNode_21
   }
   {
    dst=TacticalJumpStart
    rule=AnimGraph.AnimGraphNode_StateMachine_0.WingsGroundStatemachine.AnimStateTransitionNode_22.Transition
    src=TacticalLand
    via=AnimStateTransitionNode_22
   }
   {
    dst=TacticalSprint
    rule=AnimGraph.AnimGraphNode_StateMachine_0.WingsGroundStatemachine.AnimStateTransitionNode_20.Transition
    src=TacticalLand
    via=AnimStateTransitionNode_20
   }
   {
    dst=Conduit
    rule=AnimGraph.AnimGraphNode_StateMachine_0.WingsGroundStatemachine.AnimStateTransitionNode_18.Transition
    src=TacticalSprint
    via=AnimStateTransitionNode_18
   }
   {
    dst=TacticalJumpStart
    rule=AnimGraph.AnimGraphNode_StateMachine_0.WingsGroundStatemachine.AnimStateTransitionNode_13.Transition
    src=TacticalSprint
    via=AnimStateTransitionNode_13
   }
   {
    dst=SlideStart
    rule=AnimGraph.AnimGraphNode_StateMachine_0.WingsGroundStatemachine.AnimStateTransitionNode_33.Transition
    src=ToSliding
    via=AnimStateTransitionNode_33
   }
   {
    dst=Zipline
    rule=AnimGraph.AnimGraphNode_StateMachine_0.WingsGroundStatemachine.AnimStateTransitionNode_42.Transition
    src=ToZipline
    via=AnimStateTransitionNode_42
   }
   {
    dst=Conduit
    rule=AnimGraph.AnimGraphNode_StateMachine_0.WingsGroundStatemachine.AnimStateTransitionNode_41.Transition
    src=Zipline
    via=AnimStateTransitionNode_41
   }
  ]
 }
 Conduit {path=AnimGraph.AnimGraphNode_StateMachine_0.WingsGroundStatemachine.AnimStateConduitNode_0.Conduit, nodes [Result(bCanEnterTransition=true)]}
 Swim {
  path=AnimGraph.AnimGraphNode_StateMachine_0.WingsGroundStatemachine.AnimStateNode_0.Swim
  nodes [Output Animation Pose(Result=Sequence Player 'Olympus_Wings_Swimming_Pose'() {bLoopAnimation=false, sequence=/SunRoseFlyingGameplay/Items/Wings/Animations/Wings/Olympus_Wings_Swimming_Pose})]
 }
 Idle {
  path=AnimGraph.AnimGraphNode_StateMachine_0.WingsGroundStatemachine.AnimStateNode_1.Idle
  nodes [Output Animation Pose(Result=Sequence Player 'Olympus_Wings_Ground_Idle'() {sequence=/SunRoseFlyingGameplay/Items/Wings/Animations/Wings/Olympus_Wings_Ground_Idle})]
 }
 TacticalLand {
  path=AnimGraph.AnimGraphNode_StateMachine_0.WingsGroundStatemachine.AnimStateNode_10.TacticalLand
  nodes [Output Animation Pose(Result=Sequence Player 'Olympus_Wings_TacticalSprint_Jump_Land'() {bLoopAnimation=false, sequence=/SunRoseFlyingGameplay/Items/Wings/Animations/Wings/Olympus_Wings_TacticalSprint_Jump_Land})]
 }
 SlidingPose {
  path=AnimGraph.AnimGraphNode_StateMachine_0.WingsGroundStatemachine.AnimStateNode_11.SlidingPose
  nodes [Output Animation Pose(Result=Sequence Player 'Olympus_Wings_KneeSliding'() {bLoopAnimation=false, sequence=/SunRoseFlyingGameplay/Items/Wings/Animations/Wings/Olympus_Wings_KneeSliding})]
 }
 CrouchIdle {
  path=AnimGraph.AnimGraphNode_StateMachine_0.WingsGroundStatemachine.AnimStateNode_12.CrouchIdle
  nodes [Output Animation Pose(Result=Sequence Player 'Olympus_Wings_Crouch'() {bLoopAnimation=false, sequence=/SunRoseFlyingGameplay/Items/Wings/Animations/Wings/Olympus_Wings_Crouch})]
 }
 Crouch Moving {
  path=AnimGraph.AnimGraphNode_StateMachine_0.WingsGroundStatemachine.AnimStateNode_13.Crouch Moving
  nodes [Output Animation Pose(Result=Sequence Player 'Olympus_Wings_Crouch_Moving'() {bLoopAnimation=false, sequence=/SunRoseFlyingGameplay/Items/Wings/Animations/Wings/Olympus_Wings_Crouch_Moving})]
 }
 SlideStart {
  path=AnimGraph.AnimGraphNode_StateMachine_0.WingsGroundStatemachine.AnimStateNode_14.SlideStart
  nodes [Output Animation Pose(Result=Sequence Player 'Olympus_Wings_KneeSlidingInitiate'() {bLoopAnimation=false, sequence=/SunRoseFlyingGameplay/Items/Wings/Animations/Wings/Olympus_Wings_KneeSlidingInitiate})]
 }
 Zipline {
  path=AnimGraph.AnimGraphNode_StateMachine_0.WingsGroundStatemachine.AnimStateNode_15.Zipline
  nodes [Output Animation Pose(Result=Sequence Player 'Olympus_Wings_Zipline_Anim'() {sequence=/SunRoseFlyingGameplay/Items/Wings/Animations/Wings/Olympus_Wings_Zipline_Anim})]
 }
 DiveHitToIdle {
  path=AnimGraph.AnimGraphNode_StateMachine_0.WingsGroundStatemachine.AnimStateNode_16.DiveHitToIdle
  nodes [Output Animation Pose(Result=Sequence Player 'Olympus_Wings_Dive_Hit_To_Idle'() {bLoopAnimation=false, sequence=/SunRoseFlyingGameplay/Items/Wings/Animations/Wings/Olympus_Wings_Dive_Hit_To_Idle})]
 }
 GrindRail {
  path=AnimGraph.AnimGraphNode_StateMachine_0.WingsGroundStatemachine.AnimStateNode_17.GrindRail
  nodes [
   n1 = Get As Sun Rose Player Anim BP() -> (As Sun Rose Player Anim BP:Sun Rose Player Anim BP Object Reference)
   Output Animation Pose(Result=Blendspace Player 'Olympus_Wings_GrindRail_BS'!(X=Get Speed2D!(self:Fort Player Anim Instance Proxy Object Reference=@n1) -> (Speed2D:float)) -> (Pose) {blendSpace=/SunRoseFlyingGameplay/Items/Wings/Animations/Wings/Olympus_Wings_GrindRail_BS})
  ]
 }
 JumpStart {
  path=AnimGraph.AnimGraphNode_StateMachine_0.WingsGroundStatemachine.AnimStateNode_2.JumpStart
  nodes [Output Animation Pose(Result=Sequence Player 'Olympus_Wings_Ground_Jump_Start'() {bLoopAnimation=false, sequence=/SunRoseFlyingGameplay/Items/Wings/Animations/Wings/Olympus_Wings_Ground_Jump_Start})]
 }
 FallLoop {
  path=AnimGraph.AnimGraphNode_StateMachine_0.WingsGroundStatemachine.AnimStateNode_3.FallLoop
  nodes [Output Animation Pose(Result=Sequence Player 'Olympus_Wings_Ground_Jump_Falling_Loop'() {sequence=/SunRoseFlyingGameplay/Items/Wings/Animations/Wings/Olympus_Wings_Ground_Jump_Falling_Loop})]
 }
 JumpLand {
  path=AnimGraph.AnimGraphNode_StateMachine_0.WingsGroundStatemachine.AnimStateNode_4.JumpLand
  nodes [
   Sequence Player 'Olympus_Wings_Ground_Jump_Falling_Land'() {bLoopAnimation=false, sequence=/SunRoseFlyingGameplay/Items/Wings/Animations/Wings/Olympus_Wings_Ground_Jump_Falling_Land}
   Output Animation Pose(Result=Sequence Player 'Olympus_Wings_Ground_Jump_Land'() {bLoopAnimation=false, sequence=/SunRoseFlyingGameplay/Items/Wings/Animations/Wings/Olympus_Wings_Ground_Jump_Land})
  ]
 }
 Moving {
  path=AnimGraph.AnimGraphNode_StateMachine_0.WingsGroundStatemachine.AnimStateNode_5.Moving
  nodes [
   AnimGraphNode_StateMachine_0 = Moving_Wings() {is=title AnimGraphNode_StateMachine}
   Output Animation Pose(Result=@AnimGraphNode_StateMachine_0)
  ]
 }
 Moving_Wings {
  entry=Intro_Moving
  path=AnimGraph.AnimGraphNode_StateMachine_0.WingsGroundStatemachine.AnimStateNode_5.Moving.AnimGraphNode_StateMachine_0.Moving_Wings
  ids=id
  nodes [
   AnimStateNode_0 = Intro_Moving() {is=title AnimStateNode}
   AnimStateNode_1 = Moving_Pose() {is=title AnimStateNode}
   AnimStateTransitionNode_0 = AnimStateTransitionNode() {automaticRuleTriggerTime=0, bAutomaticRuleBasedOnSequencePlayerInState=true, logicType=TLT_Inertialization}
  ]
  links [{
    dst=Moving_Pose
    rule=AnimGraph.AnimGraphNode_StateMachine_0.WingsGroundStatemachine.AnimStateNode_5.Moving.AnimGraphNode_StateMachine_0.Moving_Wings.AnimStateTransitionNode_0.Transition
    src=Intro_Moving
    via=AnimStateTransitionNode_0
   }]
 }
 Intro_Moving {
  path=AnimGraph.AnimGraphNode_StateMachine_0.WingsGroundStatemachine.AnimStateNode_5.Moving.AnimGraphNode_StateMachine_0.Moving_Wings.AnimStateNode_0.Intro_Moving
  nodes [Output Animation Pose(Result=Sequence Player 'Olympus_Wings_Ground_Loco_Start'() {bLoopAnimation=false, sequence=/SunRoseFlyingGameplay/Items/Wings/Animations/Wings/Olympus_Wings_Ground_Loco_Start})]
 }
 Moving_Pose {
  path=AnimGraph.AnimGraphNode_StateMachine_0.WingsGroundStatemachine.AnimStateNode_5.Moving.AnimGraphNode_StateMachine_0.Moving_Wings.AnimStateNode_1.Moving_Pose
  nodes [Output Animation Pose(Result=Sequence Player 'Olympus_Wings_Ground_Loco_Loop'() {sequence=/SunRoseFlyingGameplay/Items/Wings/Animations/Wings/Olympus_Wings_Ground_Loco_Loop})]
 }
 Transition {
  path=AnimGraph.AnimGraphNode_StateMachine_0.WingsGroundStatemachine.AnimStateNode_5.Moving.AnimGraphNode_StateMachine_0.Moving_Wings.AnimStateTransitionNode_0.Transition
  nodes [Result()]
 }
 FallStart {
  path=AnimGraph.AnimGraphNode_StateMachine_0.WingsGroundStatemachine.AnimStateNode_6.FallStart
  nodes [Output Animation Pose(Result=Sequence Player 'Olympus_Wings_Ground_Jump_Loop'() {bLoopAnimation=false, sequence=/SunRoseFlyingGameplay/Items/Wings/Animations/Wings/Olympus_Wings_Ground_Jump_Loop})]
 }
 TacticalSprint {
  path=AnimGraph.AnimGraphNode_StateMachine_0.WingsGroundStatemachine.AnimStateNode_7.TacticalSprint
  nodes [Output Animation Pose(Result=Sequence Player 'Olympus_Wings_TacticalSprint_Pose'() {bLoopAnimation=false, sequence=/SunRoseFlyingGameplay/Items/Wings/Animations/Wings/Olympus_Wings_TacticalSprint_Pose})]
 }
 TacticalJumpStart {
  path=AnimGraph.AnimGraphNode_StateMachine_0.WingsGroundStatemachine.AnimStateNode_8.TacticalJumpStart
  nodes [Output Animation Pose(Result=Sequence Player 'Olympus_Wings_TacticalSprint_Jump_Start'() {bLoopAnimation=false, sequence=/SunRoseFlyingGameplay/Items/Wings/Animations/Wings/Olympus_Wings_TacticalSprint_Jump_Start})]
 }
 TacticalJumpLoop {
  path=AnimGraph.AnimGraphNode_StateMachine_0.WingsGroundStatemachine.AnimStateNode_9.TacticalJumpLoop
  nodes [Output Animation Pose(Result=Sequence Player 'Olympus_Wings_TacticalSprint_Jump_Loop'() {sequence=/SunRoseFlyingGameplay/Items/Wings/Animations/Wings/Olympus_Wings_TacticalSprint_Jump_Loop})]
 }
 Transition {
  path=AnimGraph.AnimGraphNode_StateMachine_0.WingsGroundStatemachine.AnimStateTransitionNode_0.Transition
  nodes [Result(bCanEnterTransition=Get IsNotMoving() -> (IsNotMoving:bool))]
 }
 Transition {
  path=AnimGraph.AnimGraphNode_StateMachine_0.WingsGroundStatemachine.AnimStateTransitionNode_1.Transition
  nodes [Result(bCanEnterTransition=true)]
 }
 Transition {
  path=AnimGraph.AnimGraphNode_StateMachine_0.WingsGroundStatemachine.AnimStateTransitionNode_10.Transition
  nodes [Result(bCanEnterTransition=Get PlayerIsJumping() -> (PlayerIsJumping:bool))]
 }
 Transition {path=AnimGraph.AnimGraphNode_StateMachine_0.WingsGroundStatemachine.AnimStateTransitionNode_11.Transition, nodes [Result()]}
 Transition {
  path=AnimGraph.AnimGraphNode_StateMachine_0.WingsGroundStatemachine.AnimStateTransitionNode_12.Transition
  nodes [Result(bCanEnterTransition=Get IsFallLoopToJumpLand() -> (IsFallLoopToJumpLand:bool))]
 }
 Transition {
  path=AnimGraph.AnimGraphNode_StateMachine_0.WingsGroundStatemachine.AnimStateTransitionNode_13.Transition
  nodes [Result(bCanEnterTransition=Get PlayerIs Tactical Jump() -> (PlayerIs Tactical Jump:bool))]
 }
 Transition {path=AnimGraph.AnimGraphNode_StateMachine_0.WingsGroundStatemachine.AnimStateTransitionNode_14.Transition, nodes [Result()]}
 Transition {
  path=AnimGraph.AnimGraphNode_StateMachine_0.WingsGroundStatemachine.AnimStateTransitionNode_15.Transition
  nodes [Result(bCanEnterTransition=Get IsTacticalJumpLoopEnd() -> (IsTacticalJumpLoopEnd:bool))]
 }
 Transition {
  path=AnimGraph.AnimGraphNode_StateMachine_0.WingsGroundStatemachine.AnimStateTransitionNode_16.Transition
  nodes [Result(bCanEnterTransition=NOT Boolean(A=Get PlayerIs Surface Swimming() -> (PlayerIs Surface Swimming:bool)))]
 }
 Transition {
  path=AnimGraph.AnimGraphNode_StateMachine_0.WingsGroundStatemachine.AnimStateTransitionNode_17.Transition
  nodes [Result(bCanEnterTransition=Get PlayerIs Play Dive Hit to Idle() -> (PlayerIs Play Dive Hit to Idle:bool))]
 }
 Transition {
  path=AnimGraph.AnimGraphNode_StateMachine_0.WingsGroundStatemachine.AnimStateTransitionNode_18.Transition
  nodes [Result(bCanEnterTransition=NOT Boolean(A=Get PlayerIs Tactical Sprint() -> (PlayerIs Tactical Sprint:bool)))]
 }
 Transition {
  path=AnimGraph.AnimGraphNode_StateMachine_0.WingsGroundStatemachine.AnimStateTransitionNode_19.Transition
  nodes [Result(bCanEnterTransition=Get TacticalJumpLoopToTacticalLand() -> (TacticalJumpLoopToTacticalLand:bool))]
 }
 Transition {
  path=AnimGraph.AnimGraphNode_StateMachine_0.WingsGroundStatemachine.AnimStateTransitionNode_2.Transition
  nodes [
   Result(bCanEnterTransition=true)
   AND Boolean(A=Get IsBoosting() -> (IsBoosting:bool), B=NOT Boolean(A=Get IsCrashing() -> (IsCrashing:bool)))
  ]
 }
 Transition {path=AnimGraph.AnimGraphNode_StateMachine_0.WingsGroundStatemachine.AnimStateTransitionNode_20.Transition, nodes [Result()]}
 Transition {
  path=AnimGraph.AnimGraphNode_StateMachine_0.WingsGroundStatemachine.AnimStateTransitionNode_21.Transition
  nodes [Result(bCanEnterTransition=NOT Boolean(A=Get PlayerIs Tactical Sprint()))]
 }
 Transition {
  path=AnimGraph.AnimGraphNode_StateMachine_0.WingsGroundStatemachine.AnimStateTransitionNode_22.Transition
  nodes [Result(bCanEnterTransition=Get PlayerIs Tactical Jump())]
 }
 Transition {
  path=AnimGraph.AnimGraphNode_StateMachine_0.WingsGroundStatemachine.AnimStateTransitionNode_23.Transition
  nodes [Result(bCanEnterTransition=NOT Boolean(A=Get PlayerIs Tactical Sprint()))]
 }
 Transition {
  path=AnimGraph.AnimGraphNode_StateMachine_0.WingsGroundStatemachine.AnimStateTransitionNode_24.Transition
  nodes [Result(bCanEnterTransition=Get IsFallLoopToJumpLand())]
 }
 Transition {
  path=AnimGraph.AnimGraphNode_StateMachine_0.WingsGroundStatemachine.AnimStateTransitionNode_25.Transition
  nodes [Result(bCanEnterTransition=Get PlayerIsFalling() -> (PlayerIsFalling:bool))]
 }
 Transition {
  path=AnimGraph.AnimGraphNode_StateMachine_0.WingsGroundStatemachine.AnimStateTransitionNode_26.Transition
  nodes [Result(bCanEnterTransition=Get PlayerIs Crouching() -> (PlayerIs Crouching:bool))]
 }
 Transition {
  path=AnimGraph.AnimGraphNode_StateMachine_0.WingsGroundStatemachine.AnimStateTransitionNode_27.Transition
  nodes [Result(bCanEnterTransition=Get PlayerIs Crouch Moving() -> (PlayerIs Crouch Moving:bool))]
 }
 Transition {
  path=AnimGraph.AnimGraphNode_StateMachine_0.WingsGroundStatemachine.AnimStateTransitionNode_28.Transition
  nodes [Result(bCanEnterTransition=Get IsCrouchIdleToCrouchMoving() -> (IsCrouchIdleToCrouchMoving:bool))]
 }
 Transition {
  path=AnimGraph.AnimGraphNode_StateMachine_0.WingsGroundStatemachine.AnimStateTransitionNode_29.Transition
  nodes [Result(bCanEnterTransition=Get IsCrouchMovingToCrouchIdle() -> (IsCrouchMovingToCrouchIdle:bool))]
 }
 AnimationTransitionGraph_1 {path=AnimGraph.AnimGraphNode_StateMachine_0.WingsGroundStatemachine.AnimStateTransitionNode_3.AnimationTransitionGraph_1, nodes [Result()]}
 Transition {
  path=AnimGraph.AnimGraphNode_StateMachine_0.WingsGroundStatemachine.AnimStateTransitionNode_30.Transition
  nodes [Result(bCanEnterTransition=Get IsNotCrouchIdle() -> (IsNotCrouchIdle:bool))]
 }
 Transition {
  path=AnimGraph.AnimGraphNode_StateMachine_0.WingsGroundStatemachine.AnimStateTransitionNode_31.Transition
  nodes [Result(bCanEnterTransition=Get IsNotCrouchMoving() -> (IsNotCrouchMoving:bool))]
 }
 Transition {path=AnimGraph.AnimGraphNode_StateMachine_0.WingsGroundStatemachine.AnimStateTransitionNode_32.Transition, nodes [Result()]}
 Transition {
  path=AnimGraph.AnimGraphNode_StateMachine_0.WingsGroundStatemachine.AnimStateTransitionNode_33.Transition
  nodes [Result(bCanEnterTransition=Get IsToSliding() -> (IsToSliding:bool))]
 }
 Transition {path=AnimGraph.AnimGraphNode_StateMachine_0.WingsGroundStatemachine.AnimStateTransitionNode_34.Transition, nodes [Result()]}
 Transition {
  path=AnimGraph.AnimGraphNode_StateMachine_0.WingsGroundStatemachine.AnimStateTransitionNode_36.Transition
  nodes [Result(bCanEnterTransition=Get PlayerIs on Grind Rail() -> (PlayerIs on Grind Rail:bool))]
 }
 Transition {
  path=AnimGraph.AnimGraphNode_StateMachine_0.WingsGroundStatemachine.AnimStateTransitionNode_37.Transition
  nodes [Result(bCanEnterTransition=Get Player Is Sliding() -> (Player Is Sliding:bool))]
 }
 Transition {
  path=AnimGraph.AnimGraphNode_StateMachine_0.WingsGroundStatemachine.AnimStateTransitionNode_38.Transition
  nodes [Result(bCanEnterTransition=NOT Boolean(A=Get Player Is Sliding()))]
 }
 Transition {
  path=AnimGraph.AnimGraphNode_StateMachine_0.WingsGroundStatemachine.AnimStateTransitionNode_39.Transition
  nodes [Result(bCanEnterTransition=NOT Boolean(A=Get Player Is Sliding()))]
 }
 Transition {
  path=AnimGraph.AnimGraphNode_StateMachine_0.WingsGroundStatemachine.AnimStateTransitionNode_4.Transition
  nodes [Result(bCanEnterTransition=Get PlayerIs Tactical Sprint())]
 }
 Transition {
  path=AnimGraph.AnimGraphNode_StateMachine_0.WingsGroundStatemachine.AnimStateTransitionNode_40.Transition
  nodes [Result(bCanEnterTransition=Get IsOnZipline() -> (IsOnZipline:bool))]
 }
 Transition {
  path=AnimGraph.AnimGraphNode_StateMachine_0.WingsGroundStatemachine.AnimStateTransitionNode_41.Transition
  nodes [Result(bCanEnterTransition=Get IsNotOnZipline() -> (IsNotOnZipline:bool))]
 }
 Transition {
  path=AnimGraph.AnimGraphNode_StateMachine_0.WingsGroundStatemachine.AnimStateTransitionNode_42.Transition
  nodes [Result(bCanEnterTransition=Get IsOnZipline())]
 }
 Transition {
  path=AnimGraph.AnimGraphNode_StateMachine_0.WingsGroundStatemachine.AnimStateTransitionNode_43.Transition
  nodes [Result(bCanEnterTransition=NOT Boolean(A=Get PlayerIs on Grind Rail()))]
 }
 Transition {
  path=AnimGraph.AnimGraphNode_StateMachine_0.WingsGroundStatemachine.AnimStateTransitionNode_5.Transition
  nodes [Result(bCanEnterTransition=Get PlayerIs Surface Swimming())]
 }
 AnimationTransitionGraph_1 {
  path=AnimGraph.AnimGraphNode_StateMachine_0.WingsGroundStatemachine.AnimStateTransitionNode_6.AnimationTransitionGraph_1
  nodes [Result(bCanEnterTransition=Get PlayerIsJumping())]
 }
 Transition {
  path=AnimGraph.AnimGraphNode_StateMachine_0.WingsGroundStatemachine.AnimStateTransitionNode_7.Transition
  nodes [
   n1 = Get PlayerIs on Ascender() -> (PlayerIs on Ascender:bool)
   n2 = Get IsFlying() -> (IsFlying:bool)
   n3 = Get PlayerIs Tactical Sprint()
   Result(bCanEnterTransition=AND Boolean!(
     A=Get PlayerIs Accelerating 2D() -> (PlayerIs Accelerating 2D:bool)
     B=NOT Boolean(A=@n2)
     C:bool=NOT Boolean(A=@n3)
     D:bool=NOT Boolean(A=@n1)
    ) -> (ReturnValue))
  ]
 }
 Transition {
  path=AnimGraph.AnimGraphNode_StateMachine_0.WingsGroundStatemachine.AnimStateTransitionNode_8.Transition
  nodes [Result(bCanEnterTransition=Get ToJumpStart() -> (ToJumpStart:bool))]
 }
 Transition {path=AnimGraph.AnimGraphNode_StateMachine_0.WingsGroundStatemachine.AnimStateTransitionNode_9.Transition, nodes [Result()]}
 WingsAirStatemachine {
  entry=Conduit
  path=AnimGraph.AnimGraphNode_StateMachine_1.WingsAirStatemachine
  ids=id
  nodes [
   ToDodge() {is=title AnimStateAliasNode, also {aliases [BoostingLoop, BoostingStart, DodgeLeftEnd, DodgeRightEnd]}}
   ToBoost() {is=title AnimStateAliasNode, also {aliases [DodgeLeftEnd, DodgeRightEnd]}}
   ToDive() {is=title AnimStateAliasNode, also {aliases [BoostEnd, BoostingLoop, DodgeLeftEnd, DodgeRightEnd]}}
   AnimStateConduitNode_0 = Conduit()
   AnimStateNode_0 = Leap() {is=title AnimStateNode}
   AnimStateNode_1 = Glide() {is=title AnimStateNode}
   AnimStateNode_10 = DodgeLeftEnd() {is=title AnimStateNode}
   AnimStateNode_2 = DivinStart() {is=title AnimStateNode}
   AnimStateNode_3 = DodgeRightStart() {is=title AnimStateNode}
   AnimStateNode_4 = DodgeLeftStart() {is=title AnimStateNode}
   AnimStateNode_5 = BoostingStart() {is=title AnimStateNode}
   AnimStateNode_6 = BoostingLoop() {is=title AnimStateNode}
   AnimStateNode_7 = DiveLoop() {is=title AnimStateNode}
   AnimStateNode_8 = BoostEnd() {is=title AnimStateNode}
   AnimStateNode_9 = DodgeRightEnd() {is=title AnimStateNode}
   AnimStateTransitionNode_0 = AnimStateTransitionNode() {crossfadeDuration=0.1, logicType=TLT_Inertialization, priorityOrder=3}
   AnimStateTransitionNode_1 = AnimStateTransitionNode() {crossfadeDuration=0, priorityOrder=5}
   AnimStateTransitionNode_10 = AnimStateTransitionNode() {crossfadeDuration=0.1, logicType=TLT_Inertialization, priorityOrder=2}
   AnimStateTransitionNode_11 = AnimStateTransitionNode()
   AnimStateTransitionNode_12 = AnimStateTransitionNode()
   AnimStateTransitionNode_13 = AnimStateTransitionNode() {crossfadeDuration=0, logicType=TLT_Inertialization}
   AnimStateTransitionNode_14 = AnimStateTransitionNode() {crossfadeDuration=0, logicType=TLT_Inertialization, priorityOrder=2}
   AnimStateTransitionNode_16 = AnimStateTransitionNode() {automaticRuleTriggerTime=0, bAutomaticRuleBasedOnSequencePlayerInState=true, crossfadeDuration=0.1, logicType=TLT_Inertialization}
   AnimStateTransitionNode_17 = AnimStateTransitionNode()
   AnimStateTransitionNode_18 = AnimStateTransitionNode()
   AnimStateTransitionNode_19 = AnimStateTransitionNode() {crossfadeDuration=0.4, logicType=TLT_Inertialization}
   AnimStateTransitionNode_2 = AnimStateTransitionNode() {crossfadeDuration=0, priorityOrder=4}
   AnimStateTransitionNode_20 = AnimStateTransitionNode() {bAutomaticRuleBasedOnSequencePlayerInState=true}
   AnimStateTransitionNode_21 = AnimStateTransitionNode()
   AnimStateTransitionNode_24 = AnimStateTransitionNode() {bAutomaticRuleBasedOnSequencePlayerInState=true, crossfadeDuration=0.1}
   AnimStateTransitionNode_25 = AnimStateTransitionNode() {bAutomaticRuleBasedOnSequencePlayerInState=true, crossfadeDuration=0.1}
   AnimStateTransitionNode_26 = AnimStateTransitionNode() {bAutomaticRuleBasedOnSequencePlayerInState=true}
   AnimStateTransitionNode_27 = AnimStateTransitionNode() {bAutomaticRuleBasedOnSequencePlayerInState=true}
   AnimStateTransitionNode_3 = AnimStateTransitionNode() {crossfadeDuration=0.1, logicType=TLT_Inertialization}
   AnimStateTransitionNode_4 = AnimStateTransitionNode() {automaticRuleTriggerTime=0, bAutomaticRuleBasedOnSequencePlayerInState=true}
   AnimStateTransitionNode_5 = AnimStateTransitionNode() {crossfadeDuration=0.4, logicType=TLT_Inertialization, priorityOrder=100}
   AnimStateTransitionNode_6 = AnimStateTransitionNode() {crossfadeDuration=0.1, logicType=TLT_Inertialization}
   AnimStateTransitionNode_7 = AnimStateTransitionNode() {logicType=TLT_Inertialization}
   AnimStateTransitionNode_8 = AnimStateTransitionNode()
   AnimStateTransitionNode_9 = AnimStateTransitionNode()
  ]
  links [
   {
    dst=Conduit
    rule=AnimGraph.AnimGraphNode_StateMachine_1.WingsAirStatemachine.AnimStateTransitionNode_20.Transition
    src=BoostEnd
    via=AnimStateTransitionNode_20
   }
   {
    dst=Conduit
    rule=AnimGraph.AnimGraphNode_StateMachine_1.WingsAirStatemachine.AnimStateTransitionNode_21.Transition
    src=BoostEnd
    via=AnimStateTransitionNode_21
   }
   {
    dst=BoostEnd
    rule=AnimGraph.AnimGraphNode_StateMachine_1.WingsAirStatemachine.AnimStateTransitionNode_19.Transition
    src=BoostingLoop
    via=AnimStateTransitionNode_19
   }
   {
    dst=BoostingLoop
    rule=AnimGraph.AnimGraphNode_StateMachine_1.WingsAirStatemachine.AnimStateTransitionNode_6.Transition
    src=BoostingStart
    via=AnimStateTransitionNode_6
   }
   {
    dst=Conduit
    rule=AnimGraph.AnimGraphNode_StateMachine_1.WingsAirStatemachine.AnimStateTransitionNode_12.Transition
    src=BoostingStart
    via=AnimStateTransitionNode_12
   }
   {
    dst=BoostingStart
    rule=AnimGraph.AnimGraphNode_StateMachine_1.WingsAirStatemachine.AnimStateTransitionNode_0.Transition
    src=Conduit
    via=AnimStateTransitionNode_0
   }
   {
    dst=DivinStart
    rule=AnimGraph.AnimGraphNode_StateMachine_1.WingsAirStatemachine.AnimStateTransitionNode_10.Transition
    src=Conduit
    via=AnimStateTransitionNode_10
   }
   {
    dst=DodgeLeftStart
    rule=AnimGraph.AnimGraphNode_StateMachine_1.WingsAirStatemachine.AnimStateTransitionNode_1.Transition
    src=Conduit
    via=AnimStateTransitionNode_1
   }
   {
    dst=DodgeRightStart
    rule=AnimGraph.AnimGraphNode_StateMachine_1.WingsAirStatemachine.AnimStateTransitionNode_2.Transition
    src=Conduit
    via=AnimStateTransitionNode_2
   }
   {
    dst=Glide
    rule=AnimGraph.AnimGraphNode_StateMachine_1.WingsAirStatemachine.AnimStateTransitionNode_5.Transition
    src=Conduit
    via=AnimStateTransitionNode_5
   }
   {
    dst=Leap
    rule=AnimGraph.AnimGraphNode_StateMachine_1.WingsAirStatemachine.AnimStateTransitionNode_7.Transition
    src=Conduit
    via=AnimStateTransitionNode_7
   }
   {
    dst=Conduit
    rule=AnimGraph.AnimGraphNode_StateMachine_1.WingsAirStatemachine.AnimStateTransitionNode_17.Transition
    src=DiveLoop
    via=AnimStateTransitionNode_17
   }
   {
    dst=Conduit
    rule=AnimGraph.AnimGraphNode_StateMachine_1.WingsAirStatemachine.AnimStateTransitionNode_11.Transition
    src=DivinStart
    via=AnimStateTransitionNode_11
   }
   {
    dst=DiveLoop
    rule=AnimGraph.AnimGraphNode_StateMachine_1.WingsAirStatemachine.AnimStateTransitionNode_16.Transition
    src=DivinStart
    via=AnimStateTransitionNode_16
   }
   {
    dst=Conduit
    rule=AnimGraph.AnimGraphNode_StateMachine_1.WingsAirStatemachine.AnimStateTransitionNode_26.Transition
    src=DodgeLeftEnd
    via=AnimStateTransitionNode_26
   }
   {
    dst=DodgeLeftEnd
    rule=AnimGraph.AnimGraphNode_StateMachine_1.WingsAirStatemachine.AnimStateTransitionNode_25.Transition
    src=DodgeLeftStart
    via=AnimStateTransitionNode_25
   }
   {
    dst=Conduit
    rule=AnimGraph.AnimGraphNode_StateMachine_1.WingsAirStatemachine.AnimStateTransitionNode_27.Transition
    src=DodgeRightEnd
    via=AnimStateTransitionNode_27
   }
   {
    dst=DodgeRightEnd
    rule=AnimGraph.AnimGraphNode_StateMachine_1.WingsAirStatemachine.AnimStateTransitionNode_24.Transition
    src=DodgeRightStart
    via=AnimStateTransitionNode_24
   }
   {
    dst=Conduit
    rule=AnimGraph.AnimGraphNode_StateMachine_1.WingsAirStatemachine.AnimStateTransitionNode_9.Transition
    src=Glide
    via=AnimStateTransitionNode_9
   }
   {
    dst=Conduit
    rule=AnimGraph.AnimGraphNode_StateMachine_1.WingsAirStatemachine.AnimStateTransitionNode_18.Transition
    src=Leap
    via=AnimStateTransitionNode_18
   }
   {
    dst=Conduit
    rule=AnimGraph.AnimGraphNode_StateMachine_1.WingsAirStatemachine.AnimStateTransitionNode_4.Transition
    src=Leap
    via=AnimStateTransitionNode_4
   }
   {
    dst=BoostingStart
    rule=AnimGraph.AnimGraphNode_StateMachine_1.WingsAirStatemachine.AnimStateTransitionNode_8.Transition
    src=ToBoost
    via=AnimStateTransitionNode_8
   }
   {
    dst=DivinStart
    rule=AnimGraph.AnimGraphNode_StateMachine_1.WingsAirStatemachine.AnimStateTransitionNode_3.Transition
    src=ToDive
    via=AnimStateTransitionNode_3
   }
   {
    dst=DodgeLeftStart
    rule=AnimGraph.AnimGraphNode_StateMachine_1.WingsAirStatemachine.AnimStateTransitionNode_14.Transition
    src=ToDodge
    via=AnimStateTransitionNode_14
   }
   {
    dst=DodgeRightStart
    rule=AnimGraph.AnimGraphNode_StateMachine_1.WingsAirStatemachine.AnimStateTransitionNode_13.Transition
    src=ToDodge
    via=AnimStateTransitionNode_13
   }
  ]
 }
 Conduit {path=AnimGraph.AnimGraphNode_StateMachine_1.WingsAirStatemachine.AnimStateConduitNode_0.Conduit, nodes [Result(bCanEnterTransition=true)]}
 Leap {
  path=AnimGraph.AnimGraphNode_StateMachine_1.WingsAirStatemachine.AnimStateNode_0.Leap
  nodes [Output Animation Pose(Result=Sequence Player 'Olympus_Wings_Takeoff'() {bLoopAnimation=false, sequence=/SunRoseFlyingGameplay/Items/Wings/Animations/Wings/Olympus_Wings_Takeoff})]
 }
 Glide {
  path=AnimGraph.AnimGraphNode_StateMachine_1.WingsAirStatemachine.AnimStateNode_1.Glide
  nodes [
   AnimGraphNode_StateMachine_0 = GlideStateMachine() {is=title AnimGraphNode_StateMachine}
   n1 = Blend Poses by bool(
    BlendPose_0=Sequence Player 'Olympus_Wings_Flap_Large'() {sequence=/SunRoseFlyingGameplay/Items/Wings/Animations/Wings/Olympus_Wings_Flap_Large}
    BlendPose_1=@AnimGraphNode_StateMachine_0
    BlendTime_0=0.5
    BlendTime_1=0.4
    bActiveValue=Get Is Allow Wing Flap() -> (Is Allow Wing Flap:bool)
   ) {blendType=Sinusoidal, transitionType=Inertialization}
   n2 = Sequence Player 'Olympus_Wings_NoFuel'() {sequence=/SunRoseFlyingGameplay/Items/Wings/Animations/Wings/Olympus_Wings_NoFuel}
   n3 = Get IsOutOfGas() -> (IsOutOfGas:bool)
   n4 = Apply Additive(
    Base=Two Way Blend(
     A=Blend Poses by bool(BlendPose_0=@n2, BlendPose_1=@n1, BlendTime_0=0.4, BlendTime_1=0.5, bActiveValue=@n3) {blendType=Sinusoidal, transitionType=Inertialization}
     B=Sequence Player 'Olympus_Wings_Glide_Pitch_Down'() {bLoopAnimation=false, sequence=/SunRoseFlyingGameplay/Items/Wings/Animations/Wings/Olympus_Wings_Glide_Pitch_Down}
     Alpha=Get PlayerGlide Pitch Alpha() -> (PlayerGlide Pitch Alpha:double)
    ) {blendNode {alphaBoolBlend {blendOption=Linear}}}
    Additive="Sequence Player (additive) 'Olympus_Wings_Feather_Noise'"() {sequence=/SunRoseFlyingGameplay/Items/Wings/Animations/Wings/Olympus_Wings_Feather_Noise}
   ) {alphaBoolBlend {blendOption=Linear}}
   n5 = "Sequence Player (additive) 'Olympus_Wings_Landing_Pose'"() {bLoopAnimation=false, sequence=/SunRoseFlyingGameplay/Items/Wings/Animations/Wings/Olympus_Wings_Landing_Pose}
   n6 = Get WingSlown Down or Land Alpha() -> (WingSlown Down or Land Alpha:double)
   Output Animation Pose(Result=Blend Poses by bool(
     BlendPose_0=Sequence Player 'Olympus_Wings_Ascend_Loop'() {groupName=Flapping, method=SyncGroup, sequence=/SunRoseFlyingGameplay/Items/Wings/Animations/Wings/Olympus_Wings_Ascend_Loop}
     BlendPose_1=Apply Additive(Base=@n4, Additive=@n5, Alpha=@n6) {alphaBoolBlend {blendOption=Linear}}
     BlendTime_0=0.3
     BlendTime_1=0.6
     bActiveValue=Get IsAscending() -> (IsAscending:bool)
    ) {blendType=Linear, transitionType=Inertialization})
  ]
 }
 GlideStateMachine {
  entry=GlideState
  path=AnimGraph.AnimGraphNode_StateMachine_1.WingsAirStatemachine.AnimStateNode_1.Glide.AnimGraphNode_StateMachine_0.GlideStateMachine
  ids=id
  nodes [
   AnimStateNode_0 = GlideState() {is=title AnimStateNode}
   AnimStateNode_1 = RandomFlap() {is=title AnimStateNode}
   AnimStateTransitionNode_0 = AnimStateTransitionNode() {crossfadeDuration=0.25, logicType=TLT_Inertialization}
   AnimStateTransitionNode_1 = AnimStateTransitionNode() {bAutomaticRuleBasedOnSequencePlayerInState=true, crossfadeDuration=0.3, logicType=TLT_Inertialization}
  ]
  links [
   {
    dst=RandomFlap
    rule=AnimGraph.AnimGraphNode_StateMachine_1.WingsAirStatemachine.AnimStateNode_1.Glide.AnimGraphNode_StateMachine_0.GlideStateMachine.AnimStateTransitionNode_0.Transition
    src=GlideState
    via=AnimStateTransitionNode_0
   }
   {
    dst=GlideState
    rule=AnimGraph.AnimGraphNode_StateMachine_1.WingsAirStatemachine.AnimStateNode_1.Glide.AnimGraphNode_StateMachine_0.GlideStateMachine.AnimStateTransitionNode_1.Transition
    src=RandomFlap
    via=AnimStateTransitionNode_1
   }
  ]
 }
 GlideState {
  path=AnimGraph.AnimGraphNode_StateMachine_1.WingsAirStatemachine.AnimStateNode_1.Glide.AnimGraphNode_StateMachine_0.GlideStateMachine.AnimStateNode_0.GlideState
  nodes [Output Animation Pose(Result=Blendspace Player 'Olympus_Wings_Glide_Turbulence_BS'!(X=Get PlayerFuel() -> (PlayerFuel:double)) -> (Pose) {blendSpace=/SunRoseFlyingGameplay/Items/Wings/Animations/Wings/Olympus_Wings_Glide_Turbulence_BS})]
 }
 RandomFlap {
  path=AnimGraph.AnimGraphNode_StateMachine_1.WingsAirStatemachine.AnimStateNode_1.Glide.AnimGraphNode_StateMachine_0.GlideStateMachine.AnimStateNode_1.RandomFlap
  nodes [Output Animation Pose(Result=Sequence Player 'Olympus_Wings_Flap_Small'() {bLoopAnimation=false, sequence=/SunRoseFlyingGameplay/Items/Wings/Animations/Wings/Olympus_Wings_Flap_Small})]
 }
 Transition {
  path=AnimGraph.AnimGraphNode_StateMachine_1.WingsAirStatemachine.AnimStateNode_1.Glide.AnimGraphNode_StateMachine_0.GlideStateMachine.AnimStateTransitionNode_0.Transition
  nodes [Result(bCanEnterTransition=Get IsRandomFlap() -> (IsRandomFlap:bool))]
 }
 Transition {
  path=AnimGraph.AnimGraphNode_StateMachine_1.WingsAirStatemachine.AnimStateNode_1.Glide.AnimGraphNode_StateMachine_0.GlideStateMachine.AnimStateTransitionNode_1.Transition
  nodes [Result()]
 }
 DodgeLeftEnd {
  path=AnimGraph.AnimGraphNode_StateMachine_1.WingsAirStatemachine.AnimStateNode_10.DodgeLeftEnd
  nodes [Output Animation Pose(Result=Sequence Player 'Olympus_Wings_Dodge_Left_Recover'() {bLoopAnimation=false, sequence=/SunRoseFlyingGameplay/Items/Wings/Animations/Wings/Olympus_Wings_Dodge_Left_Recover})]
 }
 DivinStart {
  path=AnimGraph.AnimGraphNode_StateMachine_1.WingsAirStatemachine.AnimStateNode_2.DivinStart
  nodes [Output Animation Pose(Result=Blend Poses by bool(
     BlendPose_0=Sequence Player 'Olympus_Wings_Dive_Start_HitReact'() {bLoopAnimation=false, sequence=/SunRoseFlyingGameplay/Items/Wings/Animations/Wings/Olympus_Wings_Dive_Start_HitReact}
     BlendPose_1=Sequence Player 'Olympus_Wings_Dive'() {bLoopAnimation=false, sequence=/SunRoseFlyingGameplay/Items/Wings/Animations/Wings/Olympus_Wings_Dive}
     bActiveValue=Get PlayerIs Force Dive() -> (PlayerIs Force Dive:bool)
    ) {blendType=Linear, transitionType=Inertialization})]
 }
 DodgeRightStart {
  path=AnimGraph.AnimGraphNode_StateMachine_1.WingsAirStatemachine.AnimStateNode_3.DodgeRightStart
  nodes [Output Animation Pose(Result=Modify Curve(
     SourcePose=Sequence Player 'Olympus_Wings_Dodge_Right'() {bLoopAnimation=false, sequence=/SunRoseFlyingGameplay/Items/Wings/Animations/Wings/Olympus_Wings_Dodge_Right}
     CurveValues_0=1
     Alpha=1
    ) {curveNames [Stop], curveValues [0]})]
 }
 DodgeLeftStart {
  path=AnimGraph.AnimGraphNode_StateMachine_1.WingsAirStatemachine.AnimStateNode_4.DodgeLeftStart
  nodes [Output Animation Pose(Result=Modify Curve(
     SourcePose=Sequence Player 'Olympus_Wings_Dodge_Left'() {bLoopAnimation=false, sequence=/SunRoseFlyingGameplay/Items/Wings/Animations/Wings/Olympus_Wings_Dodge_Left}
     CurveValues_0=1
     Alpha=1
    ) {curveNames [Stop], curveValues [0]})]
 }
 BoostingStart {
  path=AnimGraph.AnimGraphNode_StateMachine_1.WingsAirStatemachine.AnimStateNode_5.BoostingStart
  nodes [Output Animation Pose(Result=Modify Curve(
     SourcePose=Sequence Player 'Olympus_Wings_Boost_Start'() {bLoopAnimation=false, sequence=/SunRoseFlyingGameplay/Items/Wings/Animations/Wings/Olympus_Wings_Boost_Start}
     CurveValues_0=1
     Alpha=1
    ) {curveNames [Stop], curveValues [0]})]
 }
 BoostingLoop {
  path=AnimGraph.AnimGraphNode_StateMachine_1.WingsAirStatemachine.AnimStateNode_6.BoostingLoop
  nodes [
   Apply Additive(Additive=Blendspace Player 'Olympus_Wings_Boost_Roll_BS'!(X=Get PlayerGlide Roll() -> (PlayerGlide Roll:double)) -> (Pose) {blendSpace=/SunRoseFlyingGameplay/Items/Wings/Animations/Wings/Olympus_Wings_Boost_Roll_BS}) {alphaBoolBlend {blendOption=Linear}}
   Output Animation Pose(Result=Modify Curve(
     SourcePose=Sequence Player 'Olympus_Wings_Boost_Loop'() {sequence=/SunRoseFlyingGameplay/Items/Wings/Animations/Wings/Olympus_Wings_Boost_Loop}
     CurveValues_0=1
     Alpha=1
    ) {curveNames [Stop], curveValues [0]})
  ]
 }
 DiveLoop {
  path=AnimGraph.AnimGraphNode_StateMachine_1.WingsAirStatemachine.AnimStateNode_7.DiveLoop
  nodes [Output Animation Pose(Result=Sequence Player 'Olympus_Wings_Dive_Loop'() {bLoopAnimation=false, sequence=/SunRoseFlyingGameplay/Items/Wings/Animations/Wings/Olympus_Wings_Dive_Loop})]
 }
 BoostEnd {
  path=AnimGraph.AnimGraphNode_StateMachine_1.WingsAirStatemachine.AnimStateNode_8.BoostEnd
  nodes [Output Animation Pose(Result=Sequence Player 'Olympus_Wings_Boost_End'() {bLoopAnimation=false, sequence=/SunRoseFlyingGameplay/Items/Wings/Animations/Wings/Olympus_Wings_Boost_End})]
 }
 DodgeRightEnd {
  path=AnimGraph.AnimGraphNode_StateMachine_1.WingsAirStatemachine.AnimStateNode_9.DodgeRightEnd
  nodes [Output Animation Pose(Result=Sequence Player 'Olympus_Wings_Dodge_Right_Recover'() {bLoopAnimation=false, sequence=/SunRoseFlyingGameplay/Items/Wings/Animations/Wings/Olympus_Wings_Dodge_Right_Recover})]
 }
 Transition {
  path=AnimGraph.AnimGraphNode_StateMachine_1.WingsAirStatemachine.AnimStateTransitionNode_0.Transition
  nodes [Result(bCanEnterTransition=Get IsBoosting())]
 }
 Transition {
  path=AnimGraph.AnimGraphNode_StateMachine_1.WingsAirStatemachine.AnimStateTransitionNode_1.Transition
  nodes [Result(bCanEnterTransition=Get IsWingDodgeLeft() -> (IsWingDodgeLeft:bool))]
 }
 Transition {
  path=AnimGraph.AnimGraphNode_StateMachine_1.WingsAirStatemachine.AnimStateTransitionNode_10.Transition
  nodes [Result(bCanEnterTransition=Get IsCrashing())]
 }
 Transition {
  path=AnimGraph.AnimGraphNode_StateMachine_1.WingsAirStatemachine.AnimStateTransitionNode_11.Transition
  nodes [Result(bCanEnterTransition=true)]
 }
 Transition {
  path=AnimGraph.AnimGraphNode_StateMachine_1.WingsAirStatemachine.AnimStateTransitionNode_12.Transition
  nodes [Result(bCanEnterTransition=NOT Boolean(A=Get IsBoosting()))]
 }
 Transition {
  path=AnimGraph.AnimGraphNode_StateMachine_1.WingsAirStatemachine.AnimStateTransitionNode_13.Transition
  nodes [Result(bCanEnterTransition=Get IsWingDodgeRight() -> (IsWingDodgeRight:bool))]
 }
 Transition {
  path=AnimGraph.AnimGraphNode_StateMachine_1.WingsAirStatemachine.AnimStateTransitionNode_14.Transition
  nodes [Result(bCanEnterTransition=Get IsWingDodgeLeft())]
 }
 Transition {path=AnimGraph.AnimGraphNode_StateMachine_1.WingsAirStatemachine.AnimStateTransitionNode_16.Transition, nodes [Result()]}
 Transition {
  path=AnimGraph.AnimGraphNode_StateMachine_1.WingsAirStatemachine.AnimStateTransitionNode_17.Transition
  nodes [Result(bCanEnterTransition=NOT Boolean(A=Get IsCrashing()))]
 }
 Transition {
  path=AnimGraph.AnimGraphNode_StateMachine_1.WingsAirStatemachine.AnimStateTransitionNode_18.Transition
  nodes [Result(bCanEnterTransition=Get PlayerIs Leap Interruptible() -> (PlayerIs Leap Interruptible:bool))]
 }
 Transition {
  path=AnimGraph.AnimGraphNode_StateMachine_1.WingsAirStatemachine.AnimStateTransitionNode_19.Transition
  nodes [Result(bCanEnterTransition=NOT Boolean(A=Get IsBoosting()))]
 }
 Transition {
  path=AnimGraph.AnimGraphNode_StateMachine_1.WingsAirStatemachine.AnimStateTransitionNode_2.Transition
  nodes [Result(bCanEnterTransition=Get IsWingDodgeRight())]
 }
 Transition {path=AnimGraph.AnimGraphNode_StateMachine_1.WingsAirStatemachine.AnimStateTransitionNode_20.Transition, nodes [Result()]}
 Transition {
  path=AnimGraph.AnimGraphNode_StateMachine_1.WingsAirStatemachine.AnimStateTransitionNode_21.Transition
  nodes [Result(bCanEnterTransition=Get IsBoostEnd() -> (IsBoostEnd:bool))]
 }
 Transition {path=AnimGraph.AnimGraphNode_StateMachine_1.WingsAirStatemachine.AnimStateTransitionNode_24.Transition, nodes [Result()]}
 Transition {path=AnimGraph.AnimGraphNode_StateMachine_1.WingsAirStatemachine.AnimStateTransitionNode_25.Transition, nodes [Result()]}
 Transition {path=AnimGraph.AnimGraphNode_StateMachine_1.WingsAirStatemachine.AnimStateTransitionNode_26.Transition, nodes [Result()]}
 Transition {path=AnimGraph.AnimGraphNode_StateMachine_1.WingsAirStatemachine.AnimStateTransitionNode_27.Transition, nodes [Result()]}
 Transition {
  path=AnimGraph.AnimGraphNode_StateMachine_1.WingsAirStatemachine.AnimStateTransitionNode_3.Transition
  nodes [Result(bCanEnterTransition=Get IsCrashing())]
 }
 Transition {path=AnimGraph.AnimGraphNode_StateMachine_1.WingsAirStatemachine.AnimStateTransitionNode_4.Transition, nodes [Result()]}
 Transition {
  path=AnimGraph.AnimGraphNode_StateMachine_1.WingsAirStatemachine.AnimStateTransitionNode_5.Transition
  nodes [Result(bCanEnterTransition=true)]
 }
 Transition {
  path=AnimGraph.AnimGraphNode_StateMachine_1.WingsAirStatemachine.AnimStateTransitionNode_6.Transition
  nodes [Result(bCanEnterTransition=Get PlayerIs Boost Start End() -> (PlayerIs Boost Start End:bool))]
 }
 Transition {
  path=AnimGraph.AnimGraphNode_StateMachine_1.WingsAirStatemachine.AnimStateTransitionNode_7.Transition
  nodes [Result(bCanEnterTransition=Get IsWingLeaping() -> (IsWingLeaping:bool))]
 }
 Transition {
  path=AnimGraph.AnimGraphNode_StateMachine_1.WingsAirStatemachine.AnimStateTransitionNode_8.Transition
  nodes [Result(bCanEnterTransition=Get IsToBoostStart() -> (IsToBoostStart:bool))]
 }
 Transition {
  path=AnimGraph.AnimGraphNode_StateMachine_1.WingsAirStatemachine.AnimStateTransitionNode_9.Transition
  nodes [Result(bCanEnterTransition=true)]
 }
 WingsGroundAdditive {
  entry=AdditiveIdentity
  path=AnimGraph.AnimGraphNode_StateMachine_2.WingsGroundAdditive
  ids=id
  nodes [
   AnimStateNode_0 = AdditiveIdentity() {is=title AnimStateNode}
   AnimStateNode_1 = LandAdditive() {is=title AnimStateNode}
   AnimStateNode_2 = WingsReplenish() {is=title AnimStateNode}
   AnimStateTransitionNode_0 = AnimStateTransitionNode() {logicType=TLT_Inertialization}
   AnimStateTransitionNode_1 = AnimStateTransitionNode() {automaticRuleTriggerTime=0, bAutomaticRuleBasedOnSequencePlayerInState=true, logicType=TLT_Inertialization}
   AnimStateTransitionNode_2 = AnimStateTransitionNode() {crossfadeDuration=0.1, logicType=TLT_Inertialization, priorityOrder=2}
   AnimStateTransitionNode_3 = AnimStateTransitionNode() {automaticRuleTriggerTime=0, bAutomaticRuleBasedOnSequencePlayerInState=true}
  ]
  links [
   {
    dst=LandAdditive
    rule=AnimGraph.AnimGraphNode_StateMachine_2.WingsGroundAdditive.AnimStateTransitionNode_0.Transition
    src=AdditiveIdentity
    via=AnimStateTransitionNode_0
   }
   {
    dst=WingsReplenish
    rule=AnimGraph.AnimGraphNode_StateMachine_2.WingsGroundAdditive.AnimStateTransitionNode_2.Transition
    src=AdditiveIdentity
    via=AnimStateTransitionNode_2
   }
   {
    dst=AdditiveIdentity
    rule=AnimGraph.AnimGraphNode_StateMachine_2.WingsGroundAdditive.AnimStateTransitionNode_1.Transition
    src=LandAdditive
    via=AnimStateTransitionNode_1
   }
   {
    dst=AdditiveIdentity
    rule=AnimGraph.AnimGraphNode_StateMachine_2.WingsGroundAdditive.AnimStateTransitionNode_3.Transition
    src=WingsReplenish
    via=AnimStateTransitionNode_3
   }
  ]
 }
 AdditiveIdentity {
  path=AnimGraph.AnimGraphNode_StateMachine_2.WingsGroundAdditive.AnimStateNode_0.AdditiveIdentity
  nodes [Output Animation Pose(Result=Additive Identity Pose())]
 }
 LandAdditive {
  path=AnimGraph.AnimGraphNode_StateMachine_2.WingsGroundAdditive.AnimStateNode_1.LandAdditive
  nodes [Output Animation Pose(Result="Sequence Player (additive) 'Olympus_Wings_LandingAdditive'"() {bLoopAnimation=false, sequence=/SunRoseFlyingGameplay/Items/Wings/Animations/Wings/Olympus_Wings_LandingAdditive})]
 }
 WingsReplenish {
  path=AnimGraph.AnimGraphNode_StateMachine_2.WingsGroundAdditive.AnimStateNode_2.WingsReplenish
  nodes [Output Animation Pose(Result="Sequence Player (additive) 'Olympus_Wings_Cooldown_Refresh'"() {bLoopAnimation=false, sequence=/SunRoseFlyingGameplay/Items/Wings/Animations/Wings/Olympus_Wings_Cooldown_Refresh})]
 }
 Transition {
  path=AnimGraph.AnimGraphNode_StateMachine_2.WingsGroundAdditive.AnimStateTransitionNode_0.Transition
  nodes [Result(bCanEnterTransition=Get PlayerIs Play Land Additive() -> (PlayerIs Play Land Additive:bool))]
 }
 Transition {path=AnimGraph.AnimGraphNode_StateMachine_2.WingsGroundAdditive.AnimStateTransitionNode_1.Transition, nodes [Result()]}
 Transition {
  path=AnimGraph.AnimGraphNode_StateMachine_2.WingsGroundAdditive.AnimStateTransitionNode_2.Transition
  nodes [Result(bCanEnterTransition=Get Is Wing Replenish() -> (Is Wing Replenish:bool))]
 }
 Transition {path=AnimGraph.AnimGraphNode_StateMachine_2.WingsGroundAdditive.AnimStateTransitionNode_3.Transition, nodes [Result()]}
 EventGraph {nodes [
   n1 = Event BlueprintUpdateAnimation()
   > Set Delta Time X(Delta Time X=@n1.DeltaTimeX)
   > n2 = Sequence!(execute) -> (then_0, then_1, then_2:Exec, then_3:Exec, then_4:Exec, then_5:Exec, then_6:Exec)
   > n3 = Cast To FortPlayerPawn(Object=TryGetPawnOwner())
   > Set As Fort Player Pawn(As Fort Player Pawn:Fort Player Pawn Object Reference=@n3) -> (Output_Get:Fort Player Pawn Object Reference)
   n4 = Get As Fort Player Pawn() -> (As Fort Player Pawn:Fort Player Pawn Object Reference)
   > n5 = Cast To SunRose_Player_AnimBP!(
    execute
    Object=GetLinkedAnimLayerInstanceByClass!(
     self:Skeletal Mesh Component Object Reference=Get Mesh(self=@n4)
     InClass:Anim Instance Class Reference=&/SunRose_Player_AnimBP.SunRose_Player_AnimBP_C
    ) -> (ReturnValue:Anim Instance Object Reference)
   ) -> (then, CastFailed, AsSun Rose Player Anim BP:Sun Rose Player Anim BP Object Reference) {is=title K2Node_DynamicCast}
   > Set As Sun Rose Player Anim BP(As Sun Rose Player Anim BP:Sun Rose Player Anim BP Object Reference=@n5) -> (Output_Get:Sun Rose Player Anim BP Object Reference)
   > Set PlayerPawn Rotation Rate(PlayerPawn Rotation Rate:Rotator=Get PawnRotationRate!(self:Fort Player Anim Instance Proxy Object Reference=@n5) -> (PawnRotationRate:Rotator)) -> (Output_Get:Rotator)
   > Set PlayerIs Accelerating 2D(PlayerIs Accelerating 2D:bool=Get bIsAccelerating2D!(self:Fort Player Anim Instance Proxy Object Reference=@n5) -> (bIsAccelerating2D)) -> (Output_Get:bool)
   > Set PlayerGlide Pitch(PlayerGlide Pitch:double=Get GlidePitch!(self:Sun Rose Flying Player Anim Instance Object Reference=@n5) -> (GlidePitch:float)) -> (Output_Get:double)
   > Set PlayerGlide Roll(PlayerGlide Roll:double=Get GlideRoll!(self:Sun Rose Flying Player Anim Instance Object Reference=@n5) -> (GlideRoll:float)) -> (Output_Get:double)
   > Set PlayerSpeed Alpha(PlayerSpeed Alpha:float=Get SlowSpeedAlpha!(self:Sun Rose Flying Player Anim Instance Object Reference=@n5) -> (SlowSpeedAlpha:float)) -> (Output_Get:float)
   > Set PlayerFuel(PlayerFuel:double=Get Fuel!(self:Sun Rose Flying Player Anim Instance Object Reference=@n5) -> (Fuel:float)) -> (Output_Get:double)
   > Set PlayerIsJumping(PlayerIsJumping:bool=Get bIsJumping!(self:Fort Player Anim Instance Proxy Object Reference=@n5) -> (bIsJumping)) -> (Output_Get:bool)
   > Set PlayerIsFalling(PlayerIsFalling:bool=Get bIsFalling!(self:Fort Player Anim Instance Proxy Object Reference=@n5) -> (bIsFalling)) -> (Output_Get:bool)
   > Set PlayerRoll Alpha(PlayerRoll Alpha:float=Get RollAlpha!(self:Sun Rose Flying Player Anim Instance Object Reference=@n5) -> (RollAlpha:float)) -> (Output_Get:float)
   > Set PlayerPitch Alpha(PlayerPitch Alpha:double=Get PitchAlpha!(self:Sun Rose Flying Player Anim Instance Object Reference=@n5) -> (PitchAlpha:float)) -> (Output_Get:double)
   > Set PlayerYaw Alpha(PlayerYaw Alpha:double=Get YawAlpha!(self:Sun Rose Flying Player Anim Instance Object Reference=@n5) -> (YawAlpha:float)) -> (Output_Get:double)
   > Set PlayerIs Tactical Sprint(PlayerIs Tactical Sprint:bool=Get bIsTacticalSprint!(self:Fort Player Anim Instance Proxy Object Reference=@n5) -> (bIsTacticalSprint:bool)) -> (Output_Get:bool)
   > Set PlayerIs Tactical Jump(PlayerIs Tactical Jump:bool=Get bIsTacticalJump!(self:Fort Player Anim Instance Proxy Object Reference=@n5) -> (bIsTacticalJump:bool)) -> (Output_Get:bool)
   > Set Player Is Sliding(Player Is Sliding:bool=Get bIsSliding!(self:Fort Player Anim Instance Proxy Object Reference=@n5) -> (bIsSliding)) -> (Output_Get:bool)
   > Set PlayerIs Leap Interruptible(PlayerIs Leap Interruptible:bool=Get bIsLeapInterruptible!(self:Sun Rose Flying Player Anim Instance Object Reference=@n5) -> (bIsLeapInterruptible:bool)) -> (Output_Get:bool)
   > Set PlayerIs Crouching(PlayerIs Crouching:bool=Get bIsCrouching!(self:Fort Player Anim Instance Proxy Object Reference=@n5) -> (bIsCrouching)) -> (Output_Get:bool)
   > Set PlayerIs Crouch Moving(PlayerIs Crouch Moving:bool=Get bIsCrouchMoving!(self:Fort Player Anim Instance Proxy Object Reference=@n5) -> (bIsCrouchMoving:bool)) -> (Output_Get:bool)
   > Set PlayerIs Ziplining(PlayerIs Ziplining:bool=Break Fort Anim Input Zipline!(FortAnimInput_Zipline:"Fort Anim Input Zipline Structure (by ref)"=Get ZiplineInput!(self:Fort Player Anim Instance Proxy Object Reference=@n5) -> (ZiplineInput:Fort Anim Input Zipline Structure)) -> (
     bIsZiplining:bool=~
     bIsOnAscender:bool=~
     bShouldPlayPivotTransition:bool=~
     LeanYaw:float=~
     PivotBlendDelayRemaining:float=~
     PivotBlendDelay:float=~
     PivotCardinalDirection:EFortCardinalDirection Enum
     WorldVelocityLastTick:Vector=~
    ) {is=title K2Node_BreakStruct}) -> (Output_Get:bool)
   > Set PlayerIs Surface Swimming(PlayerIs Surface Swimming:bool=Get bIsSurfaceSwimming!(self:Fort Player Anim Instance Proxy Object Reference=@n5) -> (bIsSurfaceSwimming)) -> (Output_Get:bool)
   > Set PlayerIs Boost Start End(PlayerIs Boost Start End:bool=Get bIsBoostStartEnd!(self:Sun Rose Flying Player Anim Instance Object Reference=@n5) -> (bIsBoostStartEnd:bool)) -> (Output_Get:bool)
   > Set PlayerGlide Roll Combined(PlayerGlide Roll Combined:double=Get GlideRollCombined!(self:Sun Rose Flying Player Anim Instance Object Reference=@n5) -> (GlideRollCombined:float)) -> (Output_Get:double)
   > Set Pelvis!(execute, Pelvis:float=GetCurveValue(CurveName=Air base), self:Sun Rose Flying Player Anim Instance Object Reference=@n5) -> (then, Output_Get:float)
   > Set Is Allow Wing Flap(Is Allow Wing Flap:bool=Get bIsAllowWingFlap!(self:Sun Rose Flying Player Anim Instance Object Reference=@n5) -> (bIsAllowWingFlap:bool)) -> (Output_Get:bool)
   > Set PlayerIs Play Land Additive(PlayerIs Play Land Additive:bool=Get bIsPlayLandAdditive!(self:Sun Rose Flying Player Anim Instance Object Reference=@n5) -> (bIsPlayLandAdditive:bool)) -> (Output_Get:bool)
   > Set PlayerIs Play Dive Hit to Idle(PlayerIs Play Dive Hit to Idle:bool=Get bIsPlayDiveHitToIdle!(self:Sun Rose Flying Player Anim Instance Object Reference=@n5) -> (bIsPlayDiveHitToIdle:bool)) -> (Output_Get:bool)
   > Set PlayerIs on Grind Rail(PlayerIs on Grind Rail:bool=Get bIsOnGrindRail!(self:Sun Rose Flying Player Anim Instance Object Reference=@n5) -> (bIsOnGrindRail)) -> (Output_Get:bool)
   > Set PlayerIs Force Dive(PlayerIs Force Dive:bool=Get bIsForceDive!(self:Sun Rose Flying Player Anim Instance Object Reference=@n5) -> (bIsForceDive:bool)) -> (Output_Get:bool)
   > Set PlayerGlide Pitch Alpha(PlayerGlide Pitch Alpha:double=Get GlidePitchAlpha!(self:Sun Rose Flying Player Anim Instance Object Reference=@n5) -> (GlidePitchAlpha:float)) -> (Output_Get:double)
   > Set PlayerGround Blend Time(PlayerGround Blend Time:double=Get GroundBlendTime!(self:Sun Rose Flying Player Anim Instance Object Reference=@n5) -> (GroundBlendTime:float)) -> (Output_Get:double)
   > Set PlayerCurrent Height Above World(PlayerCurrent Height Above World:double=Get CurrentHeightAboveWorld!(self:Sun Rose Flying Player Anim Instance Object Reference=@n5) -> (CurrentHeightAboveWorld:float)) -> (Output_Get:double)
   > Set PlayerIs on Ascender(PlayerIs on Ascender:bool=Get bIsOnAscender!(self:Sun Rose Flying Player Anim Instance Object Reference=@n5) -> (bIsOnAscender:bool)) -> (Output_Get:bool)
   > Set Is Wing Replenish(Is Wing Replenish:bool=Get bIsWingReplenish!(self:Sun Rose Flying Player Anim Instance Object Reference=@n5) -> (bIsWingReplenish:bool)) -> (Output_Get:bool)
   n6 = Set FlyingPawnComponent(
    execute=@n2.then_1
    FlyingPawnComponent:Flying Pawn Component Object Reference=GetComponentByClass!(
     self:Actor Object Reference=Get As Fort Player Pawn()
     ComponentClass:Actor Component Class Reference=&/FlyingPawnComponent.FlyingPawnComponent_C
    ) -> (ReturnValue:Flying Pawn Component Object Reference)
   ) -> (Output_Get:Flying Pawn Component Object Reference)
   n7 = Get FlyingStatus!(self:Fort Pawn Component Winged Flying Object Reference=@n6) -> (FlyingStatus:EWingedFlyingStatus Enum=Default)
   > Set IsBoosting(IsBoosting:bool=FlyingStatusHasAnyFlags!(Status:EWingedFlyingStatus Enum=@n7, Flags:Bitmask Integer=1) -> (ReturnValue:bool)) -> (Output_Get:bool)
   > Set IsWingDodgeLeft(IsWingDodgeLeft:bool=FlyingStatusHasAnyFlags(Status=@n7, Flags=8)) -> (Output_Get:bool)
   > Set IsWingDodgeRight(IsWingDodgeRight:bool=FlyingStatusHasAnyFlags(Status=@n7, Flags=16)) -> (Output_Get:bool)
   > Set IsAscending(IsAscending:bool=FlyingStatusHasAnyFlags(Status=@n7, Flags=2)) -> (Output_Get:bool)
   > Set IsCrashing(IsCrashing:bool=FlyingStatusHasAnyFlags(Status=@n7, Flags=4)) -> (Output_Get:bool)
   n8 = Get As Fort Player Pawn()
   > Set IsWingLeaping(IsWingLeaping:bool=HasMatchingGameplayTag!(self:Gameplay Tag Asset Interface Interface=@n8, TagToCheck:Gameplay Tag Structure=(TagName="Granted.Athena.SunRose.Wings.Leaping")) -> (ReturnValue:bool)) -> (Output_Get:bool)
   > Set IsFlying(IsFlying:bool=HasMatchingGameplayTag(self=@n8, TagToCheck=(TagName="MovementMode.Flying.WingedFlight"))) -> (Output_Get:bool)
   n9 = Get As Sun Rose Player Anim BP()
   n10 = Get Speed2D(self=@n9)
   n11 = "float < float"(A=@n10, B=800)
   n12 = "float >= float"(A=@n10, B=300)
   n13 = Get PlayerFuel()
   n14 = Get CurrentHeightAboveWorld(self=@n9)
   n15 = Branch(
    execute=@n2.then_2
    Condition=AND Boolean!(A=AND Boolean!(A=@n11, B=@n12) -> (ReturnValue), B="float >= float"(A=@n14, B=225), C="float > float"(A=@n13)) -> (ReturnValue)
   )
   > Set WingSlown Down or Land AlphaTarget(WingSlown Down or Land AlphaTarget:double=0.6) -> (Output_Get:double)
   Set WingSlown Down or Land AlphaTarget(
    execute=@n15.else
    WingSlown Down or Land AlphaTarget=Get SlowDownOrLandAlpha!(self:Sun Rose Flying Player Anim Instance Object Reference=@n9) -> (SlowDownOrLandAlpha:float)
   )
   Set WingSlown Down or Land Alpha(
    execute=@n2.then_3
    WingSlown Down or Land Alpha:double=FInterpTo(
     Current=Get WingSlown Down or Land Alpha()
     Target=Get WingSlown Down or Land AlphaTarget() -> (WingSlown Down or Land AlphaTarget:double)
     DeltaTime=Get Delta Time X()
     InterpSpeed=3
    )
   ) -> (Output_Get:double)
   n16 = Get As Fort Player Pawn()
   n17 = Cast To FortAnimInstance!(execute=@n2.then_4, Object=GetAnimInstance(self=Get Mesh(self=@n16))) -> (then, CastFailed, AsFort Anim Instance:Fort Anim Instance Object Reference) {is=title K2Node_DynamicCast}
   > RequestFullBodyPBALayer!(execute, InAnimInstance:Fort Anim Instance Object Reference=@n17, ReasonForRequest:Name=FlyingWings) -> (then)
   n18 = Get IsCrashing()
   n19 = Get IsBoosting()
   n20 = Get IsWingLeaping()
   n21 = Get PlayerIs Force Dive()
   > SetProperty!(
    execute
    InAnimInstance:Fort Anim Instance Object Reference=@n17
    InPropertyName:Name=FlyingWings_Flying
    InValue:bool=AND Boolean!(A=Get IsFlying(), B=NOT Boolean(A=@n19), C:bool=NOT Boolean(A=@n21), D:bool=NOT Boolean(A=@n18), E:bool=NOT Boolean(A=@n20)) -> (ReturnValue)
   ) -> (then)
   > SetProperty(InAnimInstance=@n17, InPropertyName=FlyingWings_Boosting, InValue=@n19)
   Set IsFlyingStart(execute=@n2.then_5, IsFlyingStart:bool=OR Boolean(A=Get IsWingLeaping(), B=Get IsFlying())) -> (Output_Get:bool)
   > Set IsProceduralDeactivated(IsProceduralDeactivated:bool=OR Boolean(A=Get PlayerIs Ziplining() -> (PlayerIs Ziplining:bool), B=Get PlayerIs on Ascender())) -> (Output_Get:bool)
   n22 = Get PlayerIs Tactical Sprint()
   n23 = Get IsFlying()
   n24 = OR Boolean(
    A=Get PlayerIs Tactical Sprint()
    B=AND Boolean!(A=Get PlayerIs Accelerating 2D(), B=NOT Boolean(A=@n23), C=NOT Boolean(A=@n22)) -> (ReturnValue)
   )
   n25 = Get Player Is Sliding()
   > Set IsToSliding(IsToSliding:bool=OR Boolean(A=Get Player Is Sliding(), B=AND Boolean!(A=@n25, B=@n24) -> (ReturnValue))) -> (Output_Get:bool)
   n26 = Get PlayerIsJumping()
   n27 = Get PlayerIsFalling()
   > Set IsFallLoopToJumpLand(IsFallLoopToJumpLand:bool=AND Boolean(A=NOT Boolean(A=@n27), B=NOT Boolean(A=@n26))) -> (Output_Get:bool)
   n28 = Get PlayerIsJumping()
   n29 = Get PlayerIsFalling()
   > Set TacticalJumpLoopToTacticalLand(TacticalJumpLoopToTacticalLand:bool=AND Boolean!(A=Get PlayerIs Tactical Sprint(), B=NOT Boolean(A=@n28), C:bool=NOT Boolean(A=@n29)) -> (ReturnValue)) -> (Output_Get:bool)
   n30 = Get!() -> (PlayerPawn Rotation Rate_Roll:float, PlayerPawn Rotation Rate_Pitch:float, PlayerPawn Rotation Rate_Yaw:float)
   > Set IsCrouchIdleToCrouchMoving(IsCrouchIdleToCrouchMoving:bool=OR Boolean!(
     A=Get PlayerIs Crouch Moving()
     B="float < float"(A=@n30.PlayerPawn Rotation Rate_Yaw, B=-0.185)
     C:bool="float >= float"(A=@n30.PlayerPawn Rotation Rate_Yaw, B=0.185)
    ) -> (ReturnValue)) -> (Output_Get:bool)
   > Set IsOnZipline(IsOnZipline:bool=OR Boolean!(A=Get PlayerIs Ziplining(), B=Get PlayerIs on Ascender()) -> (ReturnValue)) -> (Output_Get:bool)
   n31 = Get PlayerIs Crouch Moving()
   n32 = Get()
   > Set IsCrouchMovingToCrouchIdle(IsCrouchMovingToCrouchIdle:bool=AND Boolean!(A=NOT Boolean(A=@n31), B="Equal (Float)"(A=@n32.PlayerPawn Rotation Rate_Yaw)) -> (ReturnValue)) -> (Output_Get:bool)
   n33 = Get PlayerFuel()
   n34 = Get PlayerCurrent Height Above World() -> (PlayerCurrent Height Above World:double)
   > Set IsOutOfGas(IsOutOfGas:bool=AND Boolean(A="Equal (Float)"(A=@n33), B="float >= float"(A=@n34, B=150))) -> (Output_Get:bool)
   n35 = Get PlayerGlide Roll Combined()
   n36 = "float < float"(A=@n35, B=2.2)
   n37 = "float >= float"(A=@n35, B=-2.2)
   n38 = "float >= float"(A=Get PlayerGlide Pitch() -> (PlayerGlide Pitch:double), B=-21)
   n39 = Get RandomInteger!(self:Sun Rose Flying Player Anim Instance Object Reference=Get As Sun Rose Player Anim BP()) -> (RandomInteger:int=0)
   > Set IsRandomFlap(IsRandomFlap:bool=AND Boolean(A="Equal (Integer)"(A=@n39, B=2), B=AND Boolean!(A=@n36, B=@n37, C:bool=@n38) -> (ReturnValue))) -> (Output_Get:bool)
   n40 = Get IsWingDodgeRight()
   n41 = Get IsWingDodgeLeft()
   > Set IsToBoostStart(IsToBoostStart:bool=AND Boolean(A=NOT Boolean(A=@n40), B=NOT Boolean(A=@n41), C=Get IsBoosting())) -> (Output_Get:bool)
   > Set IsBoostEnd(IsBoostEnd:bool=OR Boolean!(A=Get IsCrashing(), B=Get IsBoosting(), C:bool=Get IsWingDodgeLeft(), D:bool=Get IsWingDodgeRight()) -> (ReturnValue)) -> (Output_Get:bool)
   n42 = Get PlayerIsFalling()
   n43 = Get PlayerIs Tactical Sprint()
   n44 = Get PlayerIsJumping()
   > Set IsTacticalJumpLoopEnd(IsTacticalJumpLoopEnd:bool=AND Boolean(A=NOT Boolean(A=@n44), B=NOT Boolean(A=@n42), C=NOT Boolean(A=@n43))) -> (Output_Get:bool)
   n45 = Get PlayerIs on Ascender()
   > Set ToJumpStart(ToJumpStart:bool=AND Boolean!(A=Get PlayerIsJumping(), B=NOT Boolean(A=@n45)) -> (ReturnValue)) -> (Output_Get:bool)
   n46 = Get PlayerIs Accelerating 2D()
   > Set IsNotMoving(IsNotMoving:bool=OR Boolean!(A=NOT Boolean(A=@n46), B=Get IsFlying(), C=Get PlayerIs Tactical Sprint(), D=Get PlayerIs Crouching(), E:bool=Get PlayerIs Crouch Moving()) -> (ReturnValue)) -> (Output_Get:bool)
   n47 = Get PlayerIs Crouch Moving()
   n48 = Get PlayerIs Crouching()
   > Set IsNotCrouchMoving(IsNotCrouchMoving:bool=AND Boolean(A=NOT Boolean(A=@n47), B=NOT Boolean(A=@n48))) -> (Output_Get:bool)
   n49 = Get PlayerIs Crouch Moving()
   n50 = Get PlayerIs Crouching()
   > Set IsNotCrouchIdle(IsNotCrouchIdle:bool=AND Boolean(A=NOT Boolean(A=@n49), B=NOT Boolean(A=@n50))) -> (Output_Get:bool)
   n51 = Get PlayerIs on Ascender()
   n52 = Get PlayerIs Ziplining()
   > Set IsNotOnZipline(IsNotOnZipline:bool=AND Boolean(A=NOT Boolean(A=@n52), B=NOT Boolean(A=@n51))) -> (Output_Get:bool)
  ]}
]
variables [
 As Fort Player Pawn {type=FortPlayerPawn}
 As Sun Rose Player Anim BP {type=SunRose_Player_AnimBP_C, default=null}
 Delta Time X {type=number}
 FlyingPawnComponent {type=FlyingPawnComponent_C, default=null}
 Is Allow Wing Flap {type=boolean}
 Is Wing Replenish {type=boolean}
 IsAscending {type=boolean}
 IsBoostEnd {type=boolean}
 IsBoosting {type=boolean}
 IsCrashing {type=boolean}
 IsCrouchIdleToCrouchMoving {type=boolean}
 IsCrouchMovingToCrouchIdle {type=boolean}
 IsFallLoopToJumpLand {type=boolean}
 IsFlying {type=boolean}
 IsFlyingStart {type=boolean}
 IsNotCrouchIdle {type=boolean}
 IsNotCrouchMoving {type=boolean}
 IsNotMoving {type=boolean}
 IsNotOnZipline {type=boolean}
 IsOnZipline {type=boolean}
 IsOutOfGas {type=boolean}
 IsProceduralDeactivated {type=boolean}
 IsRandomFlap {type=boolean}
 IsTacticalJumpLoopEnd {type=boolean}
 IsToBoostStart {type=boolean}
 IsToSliding {type=boolean}
 IsWingDodgeLeft {type=boolean}
 IsWingDodgeRight {type=boolean}
 IsWingLeaping {type=boolean}
 Player Is Sliding {type=boolean}
 PlayerCurrent Height Above World {type=number}
 PlayerFuel {type=number}
 PlayerGlide Pitch {type=number}
 PlayerGlide Pitch Alpha {type=number}
 PlayerGlide Roll {type=number}
 PlayerGlide Roll Combined {type=number}
 PlayerGround Blend Time {type=number}
 PlayerIs Accelerating 2D {type=boolean}
 PlayerIs Boost Start End {type=boolean}
 PlayerIs Crouch Moving {type=boolean}
 PlayerIs Crouching {type=boolean}
 PlayerIs Force Dive {type=boolean}
 PlayerIs Interruptible {type=boolean}
 PlayerIs Leap Interruptible {type=boolean}
 PlayerIs on Ascender {type=boolean}
 PlayerIs on Grind Rail {type=boolean}
 PlayerIs Play Dive Hit to Idle {type=boolean}
 PlayerIs Play Land Additive {type=boolean}
 PlayerIs Surface Swimming {type=boolean}
 PlayerIs Tactical Jump {type=boolean}
 PlayerIs Tactical Sprint {type=boolean}
 PlayerIs Ziplining {type=boolean}
 PlayerIsFalling {type=boolean}
 PlayerIsJumping {type=boolean}
 PlayerPawn Rotation Rate {type=Rotator}
 PlayerPitch Alpha {type=number}
 PlayerRoll Alpha {type=number}
 PlayerSpeed Alpha {type=number}
 PlayerYaw Alpha {type=number}
 PlayerYaw Blend {type=number}
 TacticalJumpLoopToTacticalLand {type=boolean}
 ToJumpStart {type=boolean}
 Wing Flap Alpha {type=number}
 WingSlown Down or Land Alpha {type=number}
 WingSlown Down or Land AlphaTarget {type=number}
]
