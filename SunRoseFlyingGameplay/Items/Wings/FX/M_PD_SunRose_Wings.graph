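# Additive, unlit, two-sided Niagara sprite and mesh particle material for SunRose wings FX in the flying gameplay plugin.
# Emissive scales particle color by an animated 3D band from two subtracted sphere gradients, driven by looping time and a curve atlas, or optionally Dynamic Parameters.
# Alpha combines a scene-normal Fresnel, distance-behind-translucency fade and optional sprite mask, boosted by dual panning noise.
# Shading path switches substitute constants for the Fresnel on forward and mobile.
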Material /SunRoseFlyingGameplay/Items/Wings/FX/M_PD_SunRose_Wings
graph {
 properties {
  bUsedWithNiagaraMeshParticles=true
  bUsedWithNiagaraSprites=true
  blendMode=BLEND_Additive
  pixelDepthOffsetMode=PDOM_Legacy
  shadingModel=MSM_Unlit
  twoSided=true
  usedShadingModels=Unlit
 }
 nodes [
  n1 = ParticleColor()
  n2 = ScalarParameter() {defaultValue=10, parameterName=Emissive}
  n3 = ScalarParameter() {defaultValue=110, parameterName=SphereMask_A_Min}
  n4 = ScalarParameter() {defaultValue=250, parameterName=SphereMask_A_Max}
  n5 = Frac(None=Divide(A=Time(), B=ScalarParameter() {defaultValue=1.2, parameterName=Frequency}))
  n6 = CurveAtlasRowParameter(CurveTime=@n5) {
   atlas=/Game/Effects/Fort_Effects/Materials/Energy/CA_FX_Energy
   curve=/Game/Effects/Fort_Effects/Materials/Energy/Curve_FX_Energy_01
   parameterName=GradientCurve
  }
  n7 = DynamicParameter() {paramNames [FresnelExponent, PanRandom, AnimSphereMaskOffset, Param4]}
  n8 = LinearInterpolate(A=@n6.R, B=@n7.AnimSphereMaskOffset, Alpha=ScalarParameter() {parameterName=UseDPtoAnimate})
  n9 = LinearInterpolate(Alpha=OneMinus(None=Multiply(A=@n8))) {constB=2}
  n10 = VectorParameter() {parameterName=SphereMask_Offset, defaultValue=<rgba 0 0 60 1>}
  n11 = ScalarParameter() {parameterName=Z_Offset_Anim}
  n12 = MaterialFunctionCall(R=@n10.R, G=@n10.R, B=Add(A=@n10.B, B=LinearInterpolate(B=@n11, Alpha=@n8) {constB=100})) {materialFunction=&AppendMany}
  n13 = Power(
   Base=MaterialFunctionCall!(Location, Offset=@n12.RGB, Radius=LinearInterpolate(A=@n3, B=@n4, Alpha=@n9), Calculate Camera Inside, Depth Biased Alpha, FadeDistance) {materialFunction=/Engine/Functions/Engine_MaterialFunctions01/Gradient/SphereGradient-3D}.Result 0-1
   Exp=ScalarParameter() {defaultValue=4, parameterName=SphereMask_A_Power}
  )
  n14 = ScalarParameter() {defaultValue=250, parameterName=SphereMask_B_Min}
  n15 = ScalarParameter() {defaultValue=250, parameterName=SphereMask_B_Max}
  n16 = Power(
   Base=MaterialFunctionCall(Offset=@n12.RGB, Radius=LinearInterpolate(A=@n14, B=@n15, Alpha=@n9)) {materialFunction=&SphereGradient-3D}.Result 0-1
   Exp=ScalarParameter() {defaultValue=8, parameterName=SphereMask_B_Power}
  )
  n17 = ScalarParameter() {parameterName=UseDPtoAnimate}
  n18 = Saturate(None=Multiply(A=Subtract(A=@n13, B=@n16), B=LinearInterpolate(A=@n5, B=@n7.AnimSphereMaskOffset, Alpha=@n17)))
  n19 = ScalarParameter() {defaultValue=1, parameterName=UseWorld3DMask}
  MaterialFunctionCall(Emissive=Multiply(A=Multiply(A=@n1.RGB, B=@n2), B=LinearInterpolate(B=@n18, Alpha=@n19) {constA=1})) {materialFunction=&MF_FX_EmissiveAdjustments}
  n20 = Constant() {r=0.5}
  n21 = SphereMask(A=TextureCoordinate(), B=@n20, Radius=@n20) {hardnessPercent=0}
  n22 = SceneTexture() {sceneTextureId=PPI_WorldNormal}
  n23 = Fresnel(
   ExponentIn=Multiply(A=ScalarParameter() {defaultValue=1, group=Alpha, parameterName=FresnelExponent}, B=@n7.FresnelExponent)
   Normal=MaterialFunctionCall(
    Normal=ComponentMask(None=@n22.Color) {b=true, g=true, r=true}
    Flatness=ScalarParameter() {defaultValue=0.2, group=Alpha, parameterName=FlattenNormal}
   ) {materialFunction=&FlattenNormal}.Result
  ) {baseReflectFraction=0, exponent=1}
  n24 = Constant3Vector() {constant=<rgba 0.5 0.5 0.5 1>}
  n25 = ShadingPathSwitch(Default=@n23, Deferred=@n23, Forward=@n24, Mobile=@n24)
  n26 = Constant() {r=0.3}
  n27 = ScalarParameter() {defaultValue=60, parameterName=CameraOffsetBias}
  n28 = ScalarParameter() {defaultValue=60, parameterName=DistanceDivisor}
  n29 = Multiply(
   A=LinearInterpolate(B=Power(Base=@n21), Alpha=ScalarParameter() {parameterName=UseSingleSpriteBottomFade}) {constA=1}
   B=Multiply(
    A=ShadingPathSwitch(Default=@n25, Deferred=@n25, Mobile=@n26) {desc=Migrated from FeatureLevelSwitch}
    B=MaterialFunctionCall!(Distance Bias=@n27, Distance Divisor=@n28, Scene Depth, Pixel Depth) {materialFunction=/Game/Effects/Fort_Effects/Materials/_Functions/MF_DistanceBehindTranslucency_01}.Inverted
   )
  )
  n30 = Multiply(
   A=TextureCoordinate()
   B=AppendVector(A=ScalarParameter() {defaultValue=1, parameterName=Param_1}, B=ScalarParameter() {defaultValue=1, parameterName=Param_2})
  )
  n31 = Divide(A=@n30, B=ScalarParameter() {defaultValue=0.5, parameterName=Param_12})
  n32 = Multiply(
   A=TextureSample(UVs=Panner(Coordinate=@n30) {speedY=0.5}) {samplerType=SAMPLERTYPE_LinearGrayscale, texture=/Game/Global/Textures/Default/Noise/T_SandTime2}.R
   B=TextureSample(UVs=Panner(Coordinate=@n31) {speedY=0.5}) {samplerType=SAMPLERTYPE_LinearGrayscale, texture=&T_SandTime2}.R
  )
  n33 = ScalarParameter() {defaultValue=1, parameterName=Param_11}
  n34 = Multiply(
   A=Saturate(None=MaterialFunctionCall(In=@n32, Contrast=@n33) {materialFunction=&CheapContrast}.Result)
   B=ScalarParameter() {defaultValue=15, parameterName=Param_3}
  )
  Multiply(A=ParticleColor().A, B=Add(A=Multiply(A=@n29, B=@n34), B=@n29))
  n35 = Constant() {r=0.5}
  SphereMask(A=TextureCoordinate(), B=@n35, Radius=@n35) {hardnessPercent=0}
 ]
}
