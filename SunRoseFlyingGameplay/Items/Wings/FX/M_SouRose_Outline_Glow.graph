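# Translucent, two-sided unlit Niagara sprite and mesh material for a glowing outline on the SunRose wings.
# A FXCommon_SoftOutline custom-depth outline with inner glow is depth-faded, scaled by color alphas, optionally view-angle faded, then boosted by a panning gold-sand mask.
# Emissive color times particle color gains a wash from a curve-atlas lookup, driven by V coordinate, particle-randomized noise and the AnimPanCurve dynamic parameter.
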Material /SunRoseFlyingGameplay/Items/Wings/FX/M_SouRose_Outline_Glow
graph {
 properties {
  bUsedWithNiagaraMeshParticles=true
  bUsedWithNiagaraSprites=true
  blendMode=BLEND_Translucent
  pixelDepthOffsetMode=PDOM_Legacy
  shadingModel=MSM_Unlit
  twoSided=true
  usedShadingModels=Unlit
 }
 nodes [
  n1 = StaticBool() {value=true}
  n2 = MaterialFunctionCall!(
   GlowMeshStencilIndex
   UseCustomDepth=ShadingPathSwitch(Default=@n1, Deferred=@n1, Mobile=StaticBool())
   UseCustomStencil
   OutlineThickness=ScalarParameter() {defaultValue=2, parameterName="OutlineThickness (S)"}
   OutlineStrength
   PixelDepthOffset
   MaxBlurDepth=ScalarParameter() {defaultValue=200, parameterName="MaxBlurDepth (S)"}
   ------EnableInnerGlow------=ScalarParameter() {defaultValue=1, parameterName=EnableInnerGlow}
   InnerGlowMaxArea=ScalarParameter() {defaultValue=100, parameterName="InnerGlowMaxArea (S)"}
   InnerGlowPower=ScalarParameter() {defaultValue=5, parameterName="InnerGlowPower (S)"}
   InnerGlowStrength=ScalarParameter() {defaultValue=0.1, parameterName="InnerGlowStrength (S)"}
   ------LessUse------
   MeshPosition
   Distance Steps=Constant() {r=6}
   Radial Steps=Constant() {r=2}
  ) {materialFunction=/Game/Effects/Fort_Effects/Effects/Atlas/Materials/FXCommon_SoftOutline}
  n3 = ScalarParameter() {defaultValue=100, parameterName=DepthFade}
  n4 = VectorParameter() {parameterName=EmissiveColor, defaultValue=<rgba 1 0.52 0.1706 1>}
  n5 = ParticleColor()
  n6 = Multiply(A=Multiply(A=DepthFade(Opacity=@n2.Result, FadeDistance=@n3), B=@n4.A), B=@n5.A)
  n7 = ScalarParameter() {defaultValue=2, parameterName=FadeBasedOnViewAngleSensitivity}
  n8 = LinearInterpolate(
   A=@n6
   B=Multiply(
    A=@n6
    B=MaterialFunctionCall("FadeSensitivity (optional)"=@n7) {materialFunction=&FadeBasedOnViewAngle}.Result
   )
   Alpha=ScalarParameter() {parameterName=UseFadeBasedOnViewAngle}
  )
  n9 = ScalarParameter() {defaultValue=8, parameterName=GoldSand_U_Tile}
  n10 = ScalarParameter() {defaultValue=3, parameterName=GoldSand_V_Tile}
  n11 = ScalarParameter() {defaultValue=6, parameterName=GoldSandIntensity}
  n12 = Constant()
  n13 = ScalarParameter() {defaultValue=0.45, parameterName=SandPanSpeedV}
  n14 = Time()
  Add(
   A=@n8
   B=Multiply(
    A=@n8
    B=MaterialFunctionCall!(GoldSand_U_Tile=@n9, GoldSand_V_Tile=@n10, GoldSandIntensity=@n11, PannerSpeedU=@n12, PannerSpeedV=@n13, PannerTimeInput=@n14) {materialFunction=&/MF_SunRose_Wings_GoldSand}.Result
   )
  )
  n15 = TextureSample(UVs=Add(A=TextureCoordinate() {uTiling=2}, B=ParticleRandom())) {texture=&T_NoisePack_01}
  ComponentMask(None=@n15.RGB) {g=true, r=true}
  n16 = Multiply(A=@n15.R) {constB=0.3}
  n17 = MaterialFunctionCall(Float2=TextureCoordinate()) {materialFunction=&BreakOutFloat2Components}
  n18 = DynamicParameter() {paramNames [AnimPanCurve,Param2,Param3,Param4]}
  n19 = CurveAtlasRowParameter(CurveTime=Add(A=Add(A=@n16, B=@n17.G), B=LinearInterpolate(Alpha=@n18.AnimPanCurve) {constA=-0.72})) {
   atlas=/Game/Effects/Fort_Effects/Materials/Energy/CA_FX_Energy
   curve=/Game/Effects/Fort_Effects/Materials/Energy/Curve_FX_Energy_01
   parameterName=Curve
  }
  n20 = Multiply(A=@n4.RGB, B=@n5.RGB)
  n21 = Multiply(A=Multiply(A=@n19.A) {constB=3}, B=@n20)
  n22 = ScalarParameter() {defaultValue=1, parameterName=WashFX}
  MaterialFunctionCall(Emissive=Add(A=Multiply(A=@n21, B=@n22), B=@n20)) {materialFunction=&MF_FX_EmissiveAdjustments}
  MaterialFunctionCall() {materialFunction=&DebugTimeSine}
  OneMinus(None=@n15.R)
  Time()
 ]
}
