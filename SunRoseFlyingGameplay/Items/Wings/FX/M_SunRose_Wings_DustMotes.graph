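# Translucent, lit Niagara sprite material for glowing dust motes in the SunRose wings effect.
# A soft sphere mask, panning noise, particle alpha and near-camera fade combine into a mote mask.
# Panning texture samples, FlashIntensity and dynamic-parameter randomness drive emissive flashes tinted by particle color.
# Named reroutes supply a Random value blendable between debug sine and dynamic parameter, plus a per-particle hue-shifted Color that WarmUp saturates.
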Material /SunRoseFlyingGameplay/Items/Wings/FX/M_SunRose_Wings_DustMotes
graph {
 properties {
  bEnableResponsiveAA=true
  bUsedWithNiagaraSprites=true
  blendMode=BLEND_Translucent
  pixelDepthOffsetMode=PDOM_Legacy
  translucencyLightingMode=TLM_VolumetricDirectional
  usedShadingModels=Default Lit
 }
 nodes [
  n1 = ParticleRandom()
  n2 = Multiply(A=Add(A=TextureCoordinate(), B=Multiply(A=@n1) {constB=1000})) {constB=0.001}
  n3 = Multiply(
   A=Multiply(A=NamedRerouteUsage(), B=ScalarParameter() {defaultValue=18, parameterName=FlashIntensity}) {constB=10}
   B=TextureSample(UVs=Panner(Coordinate=@n2) {speedX=0.05, speedY=0.05}) {texture=/Game/Athena/Deimos/Spawners/RiftSpawners/Assets/CubeGrowth/T_CubeGrowth02_occlusion}.R
  )
  n4 = DynamicParameter() {paramNames [Random,WarmUp,Param3,Param4]}
  MaterialFunctionCall(Emissive=Multiply(A=Multiply(A=@n3, B=@n4.Random), B=ParticleColor().RGB)) {materialFunction=&MF_FX_EmissiveAdjustments}
  n5 = Constant() {r=0.5}
  n6 = SphereMask(A=TextureCoordinate(), B=@n5, Radius=@n5, Hardness=LinearInterpolate(Alpha=NamedRerouteUsage()) {constA=0.3}) {hardnessPercent=20}
  n7 = ParticleColor()
  n8 = TextureCoordinate()
  n9 = Multiply(A=NamedRerouteUsage()) {constB=1000}
  n10 = Panner(Coordinate=Multiply(A=Add(A=@n8, B=@n9)) {constB=0.1}) {speedX=0.02, speedY=0.02}
  Multiply(
   A=Multiply(
    A=Multiply(A=@n6, B=@n7.A)
    B=TextureSample(UVs=@n10) {texture=/Game/Packages/Fortress_Transylvania/Haunted_Forest/Swamps/Textures/2ndInfectedMapLinear2}.R
   )
   B=MaterialFunctionCall(Distance=ScalarParameter() {defaultValue=50, parameterName=NearCamFade}) {materialFunction=&Near_Camera_Fade}.Result
  )
  n11 = Multiply(A=NamedRerouteUsage()) {constB=0.5}
  n12 = NamedRerouteUsage()
  Multiply(A=Multiply(A=LinearInterpolate(A=@n11, Alpha=@n12) {constA=0.7}, B=@n6), B=@n7.RGB)
  NamedRerouteDeclaration(None=LinearInterpolate(
    A=MaterialFunctionCall() {materialFunction=&DebugTimeSine}.Result
    B=DynamicParameter() {paramNames [Random,WarmUp,Param3,Param4]}.Random
    Alpha=ScalarParameter() {parameterName=UseDPtoUpdate}
   )) {name=Random, nodeColor=<rgba 0 1 0.6484 1>}
  n13 = ParticleRandom()
  n14 = VectorParameter() {parameterName=Color, defaultValue=<rgba 1 0.4582 0.05 1>}
  n15 = DynamicParameter() {paramNames [Random,WarmUp,Param3,Param4]}
  NamedRerouteDeclaration(None=Desaturation(
    None=MaterialFunctionCall(Hue Shift Percentage=@n13, Texture=@n14.RGB) {materialFunction=&HueShift}.Result
    Fraction=OneMinus(None=@n15.WarmUp)
   )) {name=ColorInput, nodeColor=<rgba 1 0.2812 0 1>}
  NamedRerouteUsage()
  ParticleRandom()
  ScalarParameter()
 ]
}
