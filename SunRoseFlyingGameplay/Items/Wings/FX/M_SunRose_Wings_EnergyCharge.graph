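# Additive unlit material for Niagara mesh particles that draws an energy charge effect for the SunRose wings.
# Noise-distorted, panning plasma is tinted by particle color and masked by a curve-atlas gradient along V, which the AnimOffset dynamic parameter slides.
# The mask also drives a scrolling gold sand layer before the shared emissive adjustment function.
# Particle alpha, soft noise and a V-based fade at both ends form a separate alpha mask.
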Material /SunRoseFlyingGameplay/Items/Wings/FX/M_SunRose_Wings_EnergyCharge
graph {
 properties {
  bUsedWithNiagaraMeshParticles=true
  blendMode=BLEND_Additive
  pixelDepthOffsetMode=PDOM_Legacy
  shadingModel=MSM_Unlit
  usedShadingModels=Unlit
 }
 nodes [
  n1 = ParticleColor()
  n2 = TextureCoordinate()
  n3 = AppendVector(A=ScalarParameter() {defaultValue=2, parameterName=BaseNoiseU_Tile}, B=ScalarParameter() {defaultValue=0.5, parameterName=BaseNoiseV_Tile})
  n4 = ScalarParameter() {defaultValue=1.5, parameterName=DistortionU_Tile}
  n5 = ScalarParameter() {defaultValue=0.4, parameterName=DistortionV_Tile}
  n6 = Panner(Coordinate=Multiply(A=TextureCoordinate(), B=AppendVector(A=@n4, B=@n5))) {speedY=-0.3}
  n7 = Divide(
   A=ComponentMask(None=TextureSample(UVs=@n6) {samplerType=SAMPLERTYPE_LinearColor, texture=&T_NoisePack_LinearColor}.RGB) {g=true, r=true}
   B=ScalarParameter() {defaultValue=8, parameterName=DistortionAmount}
  )
  n8 = Panner(Coordinate=Add(A=Multiply(A=@n2, B=@n3), B=@n7)) {speedY=0.5}
  n9 = TextureSample(UVs=@n8) {samplerType=SAMPLERTYPE_LinearColor, texture=/Game/Effects/Fort_Effects/Emotes/UltraLight/Texture/T_Plasma_A_C}
  n10 = Add(
   A=Divide(A=@n9.R, B=ScalarParameter() {defaultValue=8, parameterName=MaskDistortionAmount})
   B=Multiply(A=ComponentMask(None=@n7) {r=true}) {constB=2}
  )
  n11 = MaterialFunctionCall(Float2=TextureCoordinate()) {materialFunction=&BreakOutFloat2Components}
  n12 = DynamicParameter() {paramNames [AnimOffset,Param2,Param3,Param4]}
  n13 = CurveAtlasRowParameter(CurveTime=Add(A=Add(A=@n10, B=@n11.G), B=LinearInterpolate(Alpha=@n12.AnimOffset) {constA=-0.85, constB=0.7}) {constB=0}) {
   atlas=/Game/Effects/Fort_Effects/Materials/Energy/CA_FX_Energy
   curve=/Game/Effects/Fort_Effects/Materials/Energy/Curve_FX_Energy_01
   parameterName=GradientCurve
  }
  n14 = ChannelMaskParameter(None=MaterialFunctionCall(R=@n13.R, G=@n13.R, B=@n13.R, A=@n13.R) {materialFunction=&AppendMany}.RGBA) {maskChannel=Alpha, parameterName=GradientChannelMask, defaultValue=<rgba 0 0 0 1>}
  n15 = Multiply(
   A=@n14.R
   B=MaterialFunctionCall!(
    GoldSand_U_Tile=ScalarParameter() {defaultValue=10, parameterName=GoldSand_U_Tile}
    GoldSand_V_Tile=ScalarParameter() {defaultValue=4, parameterName=GoldSand_V_Tile}
    GoldSandIntensity=ScalarParameter() {defaultValue=10, parameterName=GoldSandIntensity}
    PannerSpeedU=Constant()
    PannerSpeedV=Constant() {r=-1}
    PannerTimeInput=Time()
   ) {materialFunction=&/MF_SunRose_Wings_GoldSand}.Result
  )
  MaterialFunctionCall(Emissive=Multiply(A=Multiply(A=@n1.RGB, B=@n9.RGB), B=Add(A=@n14.R, B=@n15))) {materialFunction=&MF_FX_EmissiveAdjustments}
  n16 = TextureSample(UVs=@n8) {texture=/Game/Effects/Fort_Effects/Emotes/UltraLight/Texture/T_SoftNoise_BC7}
  n17 = Multiply(B=@n11.G) {constA=1}
  n18 = Multiply(A=OneMinus(None=@n17)) {constB=3}
  Multiply(A=@n1.A, B=Multiply(A=Multiply(A=@n16.R, B=@n17), B=Saturate(None=@n18)))
 ]
}
