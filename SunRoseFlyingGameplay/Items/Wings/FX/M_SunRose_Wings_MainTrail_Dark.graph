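# Translucent unlit Niagara ribbon material for the dark main trail of the SunRose wings.
# Opacity is built from a UV falloff and energy curve that mask noise-distorted panning noise and a gold sand function.
# Panning time takes an offset from a dynamic parameter.
# Particle alpha, near-camera and view-angle fades scale the result, with depth fade skipped on forward and mobile paths.
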Material /SunRoseFlyingGameplay/Items/Wings/FX/M_SunRose_Wings_MainTrail_Dark
graph {
 properties {
  bEnableResponsiveAA=true
  bUsedWithNiagaraRibbons=true
  blendMode=BLEND_Translucent
  pixelDepthOffsetMode=PDOM_Legacy
  shadingModel=MSM_Unlit
  usedShadingModels=Unlit
 }
 nodes [
  Constant()
  n1 = ParticleColor()
  n2 = MaterialFunctionCall(Float2=TextureCoordinate()) {materialFunction=&BreakOutFloat2Components}
  n3 = OneMinus(None=@n2.R)
  n4 = CurveAtlasRowParameter(CurveTime=@n2.G) {atlas=/Game/Effects/Fort_Effects/Materials/Energy/CA_FX_Energy, curve=/Game/Effects/Fort_Effects/Materials/Energy/Curve_FX_Energy_01}
  n5 = Multiply(A=Multiply(A=@n3, B=@n3), B=Multiply(A=@n4.G))
  n6 = TextureCoordinate()
  n7 = Multiply(A=TextureCoordinate(), B=AppendVector(A=Constant() {r=0.2}, B=Constant() {r=0.5}))
  n8 = NamedRerouteUsage()
  n9 = Divide(
   A=TextureSample(UVs=Panner(Coordinate=@n7, Time=@n8) {speedX=-0.25}) {texture=/Game/Global/Textures/Default/Noise/T_Tiling_Noise_01}.R
   B=Constant() {r=1}
  )
  n10 = Constant() {r=0.5}
  n11 = Constant() {r=0.5}
  n12 = Panner(Coordinate=Multiply(A=Add(A=@n6, B=@n9), B=AppendVector(A=@n10, B=@n11)), Time=NamedRerouteUsage()) {speedX=-0.5}
  n13 = Subtract(
   A=Multiply(
    A=@n5
    B=TextureSample(UVs=@n12) {samplerType=SAMPLERTYPE_LinearColor, texture=&T_NoisePack_LinearColor}.G
   )
   B=Constant() {r=0.23}
  ) {constB=0.2}
  n14 = ScalarParameter() {defaultValue=1, parameterName=GoldSand_U_Tile}
  n15 = ScalarParameter() {defaultValue=0.5, parameterName=GoldSand_V_Tile}
  n16 = ScalarParameter() {defaultValue=20, parameterName=GoldSandIntensity}
  n17 = Constant() {r=-0.5}
  n18 = Constant()
  n19 = NamedRerouteUsage()
  n20 = Divide(A=@n12, B=Constant() {r=1})
  n21 = Add(
   A=Multiply(
    A=Saturate(None=@n13)
    B=MaterialFunctionCall!(GoldSand_U_Tile=@n14, GoldSand_V_Tile=@n15, GoldSandIntensity=@n16, PannerSpeedU=@n17, PannerSpeedV=@n18, PannerTimeInput=@n19) {materialFunction=&/MF_SunRose_Wings_GoldSand}.Result
   )
   B=Multiply(
    A=@n5
    B=TextureSample(UVs=@n20) {samplerType=SAMPLERTYPE_LinearColor, texture=&T_NoisePack_LinearColor}.R
   )
  )
  n22 = DepthFade(Opacity=@n21, FadeDistance=ScalarParameter() {defaultValue=100, parameterName=DepthFade})
  n23 = ShadingPathSwitch(Default=@n22, Deferred=@n22, Forward=@n21, Mobile=@n21)
  n24 = MaterialFunctionCall(Distance=Constant() {r=500}) {materialFunction=&Near_Camera_Fade}
  n25 = ScalarParameter() {defaultValue=1, parameterName=FadeBasedOnViewAngle}
  n26 = Saturate(None=Multiply(A=MaterialFunctionCall("FadeSensitivity (optional)"=@n25) {materialFunction=&FadeBasedOnViewAngle}.Result) {constB=5})
  Multiply(A=Multiply(A=Multiply(A=@n1.A, B=@n23), B=Multiply(A=@n24.Result, B=@n26)), B=ScalarParameter() {defaultValue=2, parameterName=Opacity})
  NamedRerouteDeclaration(None=Add(A=Time(), B=DynamicParameter() {paramNames [TimeOffset,Param2,Param3,Param4]}.TimeOffset)) {name=TimeOffset, nodeColor=<rgba 1 0.1641 0 1>}
 ]
}
