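# Niagara system for SunRose wing dust motes: a GPU sprite burst of short-lived warm orange particles spawned in a box volume.
# A curve-skewed random value per particle makes smaller motes more opaque and also feeds a dynamic material parameter.
# A spherical-radius region mask computed at spawn scales sprite size toward zero as its value rises.
# Motes fade out over their lifetime while curl noise and very weak downward gravity drift them.
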NiagaraSystem /SunRoseFlyingGameplay/Items/Wings/FX/NS_SunRose_Wings_DustMotes
dependencies {
 usedDataInterfaces [&NiagaraDataInterfaceCurve, &NiagaraDataInterfaceColorCurve]
 usedDynamicInputs []
 usedModules [&SystemState]
}
emitters {DustMotes {
  simTarget=GPUComputeSim
  emitterUpdateScript [
   &EmitterState()
   &SpawnBurst_Instantaneous(Spawn Count=100, Use Spawn Probability=false, UseLoopCountLimit=false)
  ]
  particleSpawnScript [
   /SetVariables(Particles.Random=&FloatFromCurve(
     FloatCurve=DataInterface() {propertyValues {Curve {keys [
         {arriveTangent=0.0328, interpMode=RCIM_Cubic, leaveTangent=0.0328, tangentMode=RCTM_User, time=0, value=0}
         {arriveTangent=3.369, interpMode=RCIM_Cubic, leaveTangent=3.369, tangentMode=RCTM_User, time=1, value=1}
        ]}}}
     CurveIndex=&RandomRangeFloat(Minimum=0, Maximum=1)
    )) {isSetParameters=true}
   &InitializeParticle(
    Color=&MakeLinearColor(
     R:NiagaraFloat=1
     G:NiagaraFloat=1
     B:NiagaraFloat=1
     A:NiagaraFloat=&Lerp_Float(A=0.05, B=0.2, Alpha=$Particles.Random)
    )
    Lifetime Max=1
    Lifetime Min=0.5
    Lifetime Mode=::NewEnumerator1
    Sprite Size Mode=::NewEnumerator3
    Uniform Sprite Size=&Multiply_Float(A=15, B=&Lerp_Float(A=1, B=0.1, Alpha=$Particles.Random))
    UsePositionOffset=false
    Write Material Random=true
   )
   &ShapeLocation(Box Size=<20 180 100>, MaskSpawnGroup=false, Override Seed=false, Shape Primitive=::NewEnumerator2, Surface Only=false, UseOwnerScale=true)
   &DynamicMaterialParameters(
    Index 0 Param 1=$Particles.Random
    Index 0 Param 2=0.5
    Module.Index0Param0Write=true
    Module.Index0Param1Write=true
    Module.Index0Param2Write=true
    Module.Index0Param3Write=true
   )
   &RegionMask(Region Radius=50, UseOffset=false, UseRotation=false)
  ]
  particleUpdateScript [
   &ParticleState()
   &ScaleSpriteSize(
    Scale Sprite Size Mode=::NewEnumerator0
    Uniform Scale Factor=&Lerp_Float(A=1, B=0, Alpha=$Particles.RegionMask.Float)
   )
   &ScaleColor(
    Linear Color Curve=DataInterface() {propertyValues {
      AlphaCurve {keys [<at 0 1>,<at 1 0>]}
      BlueCurve {keys [<at 0 0.0999>,<at 1 0.2226>]}
      GreenCurve {keys [<at 0 0.4584>,<at 1 0.5297>]}
      RedCurve {keys [<at 0 1>,<at 1 1>]}
     }}
    Scale Mode=::NewEnumerator2
   ) {module=ScaleColor001}
   &CurlNoiseForce(Noise Strength=1.8, Pan Noise=false)
   &GravityForce(Gravity=<0 0 -1>)
   &SolveForcesAndVelocity(Clamp Velocity=false, Limit Acceleration=false)
  ]
  rendererClasses [&NiagaraSpriteRendererProperties]
  renderers [{rendererClass=&NiagaraSpriteRendererProperties}]
 }}
summary {
 systemName=NS_SunRose_Wings_DustMotes
 emitters [{emitterName=DustMotes, simTarget=GPUComputeSim, rendererClasses [&NiagaraSpriteRendererProperties]}]
 userVariables []
}
