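# Niagara system for the SunRose wings item that draws speed-line streaks during flight.
# One GPU sprite emitter spawns short-lived, faint, elongated sprites in a sphere offset along X, with inherited velocity reversed.
# Sprite size scales with speed and a grow-then-shrink life curve, while alpha fades and a camera offset shifts depth.
# It exposes InitialVelocity and SpeedlineSpawn user variables, though their wiring is not visible.
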NiagaraSystem /SunRoseFlyingGameplay/Items/Wings/FX/NS_SunRose_Wings_Speedlines
dependencies {
 usedDataInterfaces [&NiagaraDataInterfaceCurve, &NiagaraDataInterfaceVector2DCurve]
 usedDynamicInputs [
  &ClampFloat
  &RemapRange
  &VectorLength
  &MakeFloatFromVector
 ]
 usedModules [
  &self:SystemSpawnScript.SystemScriptSource.SystemScriptGraph.NiagaraNodeAssignment_1.SetVariables_#1
  &self:SystemSpawnScript.SystemScriptSource.SystemScriptGraph.NiagaraNodeAssignment_0.SetVariables_#2
  &self:SystemSpawnScript.SystemScriptSource.SystemScriptGraph.NiagaraNodeAssignment_2.SetVariables_#3
  &SystemState
 ]
}
emitters {FN_Sprite_SpeedLines001 {
  simTarget=GPUComputeSim
  emitterUpdateScript [
   &EmitterState()
   &SpawnRate(SpawnRate=600, Use Spawn Probability=false)
  ]
  particleSpawnScript [
   &InitializeParticle(
    Color Maximum=<rgba 1 1 1 0.1>
    Color Minimum=<rgba 1 1 1 0.05>
    Color Mode=::NewEnumerator2
    Lifetime=&UniformRangedFloat(Minimum=0.2, Maximum=0.3)
    Sprite Size Max=<70 1500>
    Sprite Size Min=<30 500>
    Sprite Size Mode=::NewEnumerator2
    UsePositionOffset=false
    Write Material Random=true
   ) {module=InitializeParticle001}
   &InheritVelocity(Inherited Velocity Amount Scale=<-2 -2 -2>, Inherited Velocity Speed Limit=5000, Speed Limit=true, UseSpeedThreshold=false)
   &ShapeLocation(MaskSpawnGroup=false, Offset=<500 0 0>, Offset Mode=::NewEnumerator0, Override Seed=false, Sphere Radius=1500, UseOwnerScale=true)
  ]
  particleUpdateScript [
   &ParticleState()
   &ScaleColor(
    Scale Alpha=&FloatFromCurve(FloatCurve=DataInterface() {propertyValues {Curve {keys [{interpMode=RCIM_Cubic, tangentMode=RCTM_User, time=0, value=1}, {interpMode=RCIM_Cubic, tangentMode=RCTM_User, time=1, value=0}]}}})
    Scale RGB=&VectorFromFloat(Value=1)
    ScaleA=true
    ScaleRGB=true
   ) {module=ScaleColor001}
   &SolveForcesAndVelocity(Clamp Velocity=false, Limit Acceleration=false)
   &ScaleSpriteSizeBySpeed(Max Scale Factor=<0.5 1>, Min Scale Factor=<0 0>, Velocity Threshold=4250)
   &ScaleSpriteSize(Scale Factor=&Vector2DFromCurve(
     Vector2Curve=DataInterface() {propertyValues {
       XCurve {keys [<at -0.0069 0 RCIM_Cubic>,<at 0.2678 1.0118 RCIM_Cubic>,<at 1 0 RCIM_Cubic>]}
       YCurve {keys [
         <at 0 0.0245 RCIM_Cubic>
         {arriveTangent=-0.0245, interpMode=RCIM_Cubic, leaveTangent=-0.0245, time=0.2784, value=0.9882}
         <at 1 0 RCIM_Cubic>
        ]}
      }}
     ScaleCurve=<1 1>
    ))
   &CameraOffset(Camera Offset Amount=-150)
  ]
  rendererClasses [&NiagaraSpriteRendererProperties]
  renderers [{rendererClass=&NiagaraSpriteRendererProperties}]
 }}
summary {
 systemName=NS_SunRose_Wings_Speedlines
 emitters [{emitterName=FN_Sprite_SpeedLines001, simTarget=GPUComputeSim, rendererClasses [&NiagaraSpriteRendererProperties]}]
 userVariables [
  User.InitialVelocity {type=Vector3f, description="", defaultValue {struct=Vector3f, value=<1 0 0>}}
  User.SpeedlineSpawn {type=NiagaraFloat, defaultValue=0, description=""}
 ]
}
