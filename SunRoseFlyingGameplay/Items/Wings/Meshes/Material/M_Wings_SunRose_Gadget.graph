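# Masked character material for the SunRose wings gadget that dissolves feathers through energy-driven masks, with charred paper tips and glowing embers.
# Between shared character PreFX and PostFX functions, it blends in a dark rough layer and sets opacity and emissive.
# Radial pre-skinned coordinates and scrolling burn-pattern noise shape the effects, while Replenish and Boost act as timestamps compared with Time for timed flashes.
# A metallic dynamic-gold blend is built but left unconnected.
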Material /SunRoseFlyingGameplay/Items/Wings/Meshes/Material/M_Wings_SunRose_Gadget
graph {
 properties {
  bUseMaterialAttributes=true
  bUsedWithClothing=true
  bUsedWithMeshDeformer=true
  bUsedWithNanite=true
  bUsedWithSkeletalMesh=true
  blendMode=BLEND_Masked
  pixelDepthOffsetMode=PDOM_Legacy
  previewMesh=/SunRoseFlyingGameplay/Items/Wings/Meshes/Wings_SunRose_Gadget
  shadingModel=MSM_FromMaterialExpression
  usedShadingModels=Default Lit | Subsurface | Clear Coat | Subsurface Profile | Cloth
 }
 nodes [
  n1 = LinearInterpolate(Alpha=LinearInterpolate(A=NamedRerouteUsage(), B=NamedRerouteUsage())) {constA=0.65}
  n2 = LinearInterpolate(Alpha=NamedRerouteUsage()) {constA=0.3}
  n3 = NamedRerouteUsage()
  n4 = Multiply(A=@n3, B=@n3)
  n5 = Add(
   A=Multiply(A=NamedRerouteUsage(), B=Multiply(A=@n1, B=@n2))
   B=LinearInterpolate(
    A=ScalarParameter() {defaultValue=1.2, group=SunRose Wings, parameterName=GoldMax}
    B=ScalarParameter() {defaultValue=0.2, group=SunRose Wings, parameterName=GoldMin}
    Alpha=Multiply(A=@n4, B=@n4)
   ) {constA=1.2, constB=0.2}
  ) {constB=0.6}
  n6 = Saturate(None=Subtract(A=Multiply(A=@n5) {constB=200}) {constB=100})
  Add(A=@n6, B=NamedRerouteUsage())
  n7 = NamedRerouteUsage()
  n8 = Constant2Vector() {g=0.6, r=2}
  n9 = Multiply(A=TextureSample(UVs=Multiply(A=@n7, B=@n8) {constB=8}) {samplerType=SAMPLERTYPE_LinearColor, texture=/SunRoseFlyingGameplay/Items/Wings/Meshes/Texture/T_Sunrose_Wings_BurnPattern}.R) {constB=0.25}
  Add(A=@n9, B=OneMinus(None=NamedRerouteUsage()))
  n10 = MaterialFunctionCall() {materialFunction=&MF_CharacterParent_2023_PreFX}
  n11 = GetMaterialAttributes(None=@n10.Result) {attributeGetTypes [#1, #2, #3]}
  n12 = Saturate(None=Subtract(A=Multiply(A=@n11.Roughness) {constB=4}) {constB=3})
  n13 = Multiply(A=Multiply(A=@n11.Specular, B=NamedRerouteDeclaration(None=@n12) {name=Curvature, nodeColor=<rgba 0.0781 1 0 1>})) {constB=4}
  n14 = SetMaterialAttributes!(
   MaterialAttributes=@n11.MaterialAttributes
   Base Color=LinearInterpolate(
    A=VectorParameter() {group=SunRose Wings, parameterName=Dynamic Gold 1, defaultValue=<rgba 0.8698 0.5783 0.2214 1>}.RGB
    B=VectorParameter() {group=SunRose Wings, parameterName=Dynamic Gold 2, defaultValue=<rgba 1 0.8035 0.5365 1>}.RGB
    Alpha=Saturate(None=@n13)
   )
   Specular=Saturate(None=Multiply(A=@n11.Specular) {constB=2})
   Roughness=Saturate(None=Subtract(A=@n11.Roughness) {constB=0.4})
   Metallic=Constant() {r=1}
  ) {attributeSetTypes [
    #3
    #1
    #2
    #4
   ]}
  BlendMaterialAttributes(A=@n11.MaterialAttributes, B=@n14, Alpha=NamedRerouteUsage()) {vertexAttributeBlendType=UseA}
  n15 = SetMaterialAttributes!(
   MaterialAttributes=@n10.Result
   Base Color=Power(Base=NamedRerouteUsage()) {constExponent=5}
   Specular=Constant() {r=0.1}
   Metallic=Constant()
   Roughness=Constant() {r=0.85}
  ) {attributeSetTypes [
    #3
    #1
    #4
    #2
   ]}
  n16 = NamedRerouteUsage()
  n17 = Add(A=NamedRerouteUsage(), B=NamedRerouteUsage())
  n18 = NamedRerouteUsage()
  MaterialFunctionCall(
   Material=SetMaterialAttributes!(
    MaterialAttributes=BlendMaterialAttributes(A=@n10.Result, B=@n15, Alpha=@n16) {vertexAttributeBlendType=UseA}
    Opacity Mask=NamedRerouteUsage()
    Emissive Color=Add(A=@n17, B=@n18) {constB=0}
   ) {attributeSetTypes [#5, #6]}
   M=@n10.M
   Pre Skinned Local Position=@n10.M
   Distance Blend=@n10.M
   Fresnel=@n10.M
  ) {materialFunction=&MF_CharacterParent_2023_PostFX}
  n19 = TextureSample() {samplerType=SAMPLERTYPE_LinearGrayscale, texture=/SunRoseFlyingGameplay/Items/Wings/Meshes/Texture/T_SunRose_Wings_Mask}
  n20 = LinearInterpolate(
   A=ScalarParameter() {defaultValue=-0.15, group=SunRose Wings, parameterName=FeathersDissolveMin}
   B=ScalarParameter() {defaultValue=0.4, group=SunRose Wings, parameterName=FeathersDissolveMax}
   Alpha=NamedRerouteUsage()
  ) {constA=-0.15, constB=0.36}
  NamedRerouteDeclaration(None=Add(A=NamedRerouteDeclaration(None=@n19.R) {name=Opacity Texture, nodeColor=<rgba 0 0.6484 1 1>}, B=@n20)) {name=Opacity Mask, nodeColor=<rgba 0 1 0.6016 1>}
  n21 = ScalarParameter() {defaultValue=1, group=SunRose Wings, parameterName=FeathersDissolve, sliderMax=1}
  n22 = Multiply(A=NamedRerouteUsage(), B=ScalarParameter() {parameterName=ReplenishAnimation})
  NamedRerouteDeclaration(None=Saturate(None=Add(A=@n21, B=@n22))) {name=Energy, nodeColor=<rgba 0.0547 0 1 1>}
  n23 = Time()
  n24 = ScalarParameter() {defaultValue=1, parameterName=Replenish}
  n25 = Divide(A=Abs(None=Subtract(A=@n23, B=@n24))) {constB=1.6}
  n26 = Multiply(A=OneMinus(None=Saturate(None=@n25)), B=Saturate(None=@n24))
  n27 = Multiply(A=@n26, B=@n26)
  n28 = Multiply(A=@n27, B=Multiply(A=OneMinus(None=@n27)) {constB=4})
  n29 = PixelNormalWS()
  n30 = CameraVectorWS()
  n31 = OneMinus(None=Saturate(None=DotProduct(A=@n29, B=@n30)))
  n32 = Add(A=NamedRerouteUsage(), B=Time())
  n33 = Add(A=Frac(None=Multiply(A=@n32) {constB=2})) {constB=-0.5}
  n34 = OneMinus(None=Multiply(A=Abs(None=@n33)) {constB=2})
  n35 = Multiply(A=@n34, B=@n34)
  n36 = Multiply(A=Multiply(A=@n28, B=Multiply(A=@n31, B=@n31)) {constA=1}, B=Add(A=Multiply(A=@n35) {constB=4}) {constB=0.5})
  n37 = Multiply(A=@n28, B=NamedRerouteUsage())
  NamedRerouteDeclaration(None=Multiply(A=Add(A=@n36, B=@n37), B=Constant3Vector() {constant=<rgba 2 0.8238 0.1931 1>})) {name=Replenish, nodeColor=<rgba 0.4766 1 0 1>}
  n38 = VertexInterpolator(VS=ComponentMask(None=PreSkinnedPosition()) {b=true, r=true})
  n39 = Multiply(A=@n38.PS) {constB=0.01}
  n40 = Multiply(A=Time(), B=Constant2Vector() {g=0.009, r=-0.0082})
  n41 = Multiply(A=@n38.PS) {constB=0.005}
  n42 = Multiply(A=Time(), B=Constant2Vector() {g=0.014, r=0.0145})
  n43 = Multiply(
   A=TextureSample(UVs=Add(A=@n39, B=@n40)) {samplerType=SAMPLERTYPE_LinearColor, texture=&T_Sunrose_Wings_BurnPattern}.A
   B=TextureSample(UVs=Add(A=@n41, B=@n42)) {samplerType=SAMPLERTYPE_LinearColor, texture=&T_Sunrose_Wings_BurnPattern}.A
  )
  NamedRerouteDeclaration(None=Multiply(A=Multiply(A=@n43, B=@n43)) {constB=14}) {name=Sparks, nodeColor=<rgba 0 1 0.5781 1>}
  n44 = NamedRerouteUsage()
  n45 = Saturate(None=Add(A=Multiply(A=@n44) {constB=2}) {constB=-1})
  n46 = Multiply(A=@n45, B=@n45)
  n47 = Add(A=@n46, B=NamedRerouteUsage())
  n48 = ScalarParameter() {defaultValue=0.1, group=SunRose Wings, parameterName=Gold Fade 1}
  n49 = Saturate(None=Multiply(A=Add(A=@n47, B=@n48) {constB=0.01}, B=ScalarParameter() {defaultValue=20, group=SunRose Wings, parameterName=Gold Fade 2}) {constB=200})
  NamedRerouteDeclaration(None=Multiply(A=@n6, B=Multiply(A=@n49, B=@n46))) {name=Dynamic Gold, nodeColor=<rgba 0.1484 0 1 1>}
  n50 = TextureSample() {samplerType=SAMPLERTYPE_LinearColor, texture=/SunRoseFlyingGameplay/Items/Wings/Meshes/Texture/T_SunRose_Wings_ColorMask}
  NamedRerouteDeclaration(None=@n50.G) {name=Random Value, nodeColor=<rgba 0 1 0.6719 1>}
  NamedRerouteDeclaration(None=@n50.B) {name=Static Gold, nodeColor=<rgba 1 0.2578 0 1>}
  n51 = Add(A=NamedRerouteUsage()) {constB=-0.5}
  NamedRerouteDeclaration(None=Saturate(None=Multiply(A=@n51) {constB=2})) {name=Golden Energy, nodeColor=<rgba 0 0.8125 1 1>}
  n52 = NamedRerouteUsage()
  NamedRerouteDeclaration(None=Saturate(None=Multiply(A=@n52) {constB=2})) {name=Feathers Energy, nodeColor=<rgba 0.0547 1 0 1>}
  n53 = OneMinus(None=NamedRerouteUsage())
  n54 = Multiply(A=@n53) {constB=5}
  n55 = NamedRerouteUsage()
  n56 = OneMinus(None=NamedRerouteUsage())
  n57 = Add(A=NamedRerouteUsage(), B=Multiply(A=NamedRerouteUsage()) {constB=0.14})
  n58 = Add(A=Multiply(A=Multiply(A=@n54, B=@n55), B=@n56), B=@n57)
  n59 = NamedRerouteUsage()
  n60 = Constant3Vector() {constant=<rgba 0.474 0.1771 0.0568 1>}
  n61 = Constant3Vector() {constant=<rgba 7 1.9777 0.401 1>}
  n62 = Multiply(A=Multiply(A=@n56, B=@n56)) {constB=2}
  n63 = NamedRerouteUsage()
  n64 = Constant2Vector() {g=1.86, r=4.23}
  n65 = Time()
  n66 = Constant2Vector() {g=0.09, r=-0.082}
  n67 = Multiply(A=OneMinus(None=NamedRerouteUsage())) {constB=0.4}
  n68 = Add(A=Add(A=Multiply(A=@n63, B=@n64) {constB=8}, B=Multiply(A=@n65, B=@n66)), B=@n67)
  n69 = NamedRerouteUsage()
  n70 = Constant2Vector() {g=0.51, r=1.6}
  n71 = Time()
  n72 = Constant2Vector() {g=0.14, r=0.145}
  n73 = Add(A=@n67, B=Add(A=Multiply(A=@n69, B=@n70) {constB=8}, B=Multiply(A=@n71, B=@n72)))
  n74 = Multiply(A=Multiply(
    A=TextureSample(UVs=@n68) {samplerType=SAMPLERTYPE_LinearColor, texture=&T_Sunrose_Wings_BurnPattern}.R
    B=TextureSample(UVs=@n73) {samplerType=SAMPLERTYPE_LinearColor, texture=&T_Sunrose_Wings_BurnPattern}.R
   )) {constB=4}
  n75 = Multiply(
   A=Max(A=Subtract(A=@n62) {constB=0.2}, B=@n57) {constB=0}
   B=NamedRerouteDeclaration(None=@n74) {name=Embers Distortion, nodeColor=<rgba 1 0.75 0 1>}
  )
  n76 = Multiply(A=@n75, B=@n75)
  n77 = NamedRerouteUsage()
  n78 = Add(A=Multiply(A=@n76, B=@n76), B=Multiply(A=Power(Base=@n77) {constExponent=256}) {constB=0.3})
  NamedRerouteDeclaration(None=Multiply(A=Multiply(A=@n58, B=@n59), B=LinearInterpolate(A=@n60, B=@n61, Alpha=@n78))) {name=Burnt Emission, nodeColor=<rgba 0 1 0.1094 1>}
  n79 = Multiply(A=NamedRerouteUsage(), B=LinearInterpolate(Alpha=NamedRerouteUsage()) {constA=0.3})
  n80 = LinearInterpolate(Alpha=NamedRerouteUsage()) {constA=0.95}
  n81 = ScalarParameter() {defaultValue=-0.5, group=SunRose Wings, parameterName=Paper Tip Min}
  n82 = ScalarParameter() {defaultValue=-1, group=SunRose Wings, parameterName=Paper Tip Max}
  n83 = NamedRerouteUsage()
  n84 = Saturate(None=Add(A=Multiply(A=@n83) {constB=4}) {constB=-3})
  n85 = Multiply(A=@n84, B=@n84)
  n86 = Multiply(A=Add(A=Multiply(A=@n79, B=@n80), B=LinearInterpolate(A=@n81, B=@n82, Alpha=@n85))) {constB=30}
  NamedRerouteDeclaration(None=Saturate(None=@n86)) {name=Paper Tips Mask, nodeColor=<rgba 0.9453 0 1 1>}
  n87 = Saturate(None=Abs(None=@n86))
  NamedRerouteDeclaration(None=OneMinus(None=Multiply(A=@n87, B=@n87))) {name=Paper Tip Emission, nodeColor=<rgba 0 0.5781 1 1>}
  n88 = Time()
  n89 = ScalarParameter() {defaultValue=1, parameterName=Replenish}
  n90 = Saturate(None=Divide(A=Subtract(A=@n88, B=@n89)) {constB=0.5})
  NamedRerouteDeclaration(None=Multiply(A=@n90, B=@n90)) {name=Replenish Energy, nodeColor=<rgba 0 1 0.5078 1>}
  n91 = ComponentMask(None=PreSkinnedPosition()) {b=true, r=true}
  n92 = Constant2Vector() {g=-50}
  n93 = VertexInterpolator(VS=Divide(A=Add(A=@n91, B=@n92), B=Constant2Vector() {g=180, r=350}))
  n94 = ComponentMask(None=@n93.PS) {r=true}
  n95 = Abs(None=@n94)
  n96 = Add(A=Multiply(A=ComponentMask(None=@n93.PS) {g=true}, B=@n95), B=@n95)
  n97 = Multiply(A=@n93.PS, B=@n93.PS)
  n98 = Saturate(None=Add(A=ComponentMask(None=@n97) {r=true}, B=ComponentMask(None=@n97) {g=true}))
  n99 = Saturate(None=Sign(None=@n94))
  NamedRerouteDeclaration(None=Add(A=AppendVector(A=@n96, B=@n98), B=Multiply(A=@n99) {constB=0.16})) {name=Radial Coord, nodeColor=<rgba 1 0.8906 0 1>}
  NamedRerouteDeclaration(None=Add(A=NamedRerouteDeclaration(None=@n50.R) {name=Color Texture, nodeColor=<rgba 0.4062 0 1 1>}, B=@n20)) {name=Color Mask, nodeColor=<rgba 0.4531 1 0 1>}
  n100 = NamedRerouteDeclaration(None=@n98) {name=Radial Mask, nodeColor=<rgba 1 0 0.7138 1>}
  NamedRerouteDeclaration(None=Multiply(A=@n100, B=@n100)) {name=Radial Mask Squared, nodeColor=<rgba 0.3828 0 1 1>}
  n101 = Multiply(A=@n53, B=@n53)
  n102 = Multiply(A=Multiply(A=Multiply(A=@n101) {constB=5}, B=Add(A=@n55) {constB=0.02}), B=@n56)
  n103 = Power(Base=OneMinus(None=Saturate(None=@n102))) {constExponent=4}
  NamedRerouteDeclaration(None=Max(A=OneMinus(None=@n103), B=NamedRerouteUsage())) {name=Burnt Multiplier, nodeColor=<rgba 1 0 0.1875 1>}
  n104 = NamedRerouteUsage()
  n105 = Time()
  n106 = ScalarParameter() {group=SunRose Wings, parameterName=Boost}
  n107 = Divide(A=Abs(None=Subtract(A=@n105, B=@n106)), B=ScalarParameter() {defaultValue=1.6, parameterName=BoostSpeed}) {constB=1.6}
  n108 = Multiply(A=OneMinus(None=Saturate(None=@n107)), B=Saturate(None=@n106))
  n109 = Multiply(A=@n108, B=@n108)
  n110 = Add(A=Add(A=Add(A=@n9, B=@n104)) {constB=-0.5}, B=LinearInterpolate(Alpha=@n109) {constA=-1, constB=0.8})
  n111 = Add(A=Saturate(None=Multiply(A=@n110) {constB=2})) {constB=-0.5}
  n112 = OneMinus(None=Multiply(A=Abs(None=@n111)) {constB=2})
  n113 = Add(A=Multiply(A=Multiply(A=@n112, B=@n112)) {constB=2}, B=NamedRerouteUsage())
  NamedRerouteDeclaration(None=Multiply(A=Multiply(A=@n113, B=@n109), B=VectorParameter() {parameterName=BoostColor, defaultValue=<rgba 1 0.7059 0.1451 1>}.RGB)) {name=Boost, nodeColor=<rgba 1 0 0.3984 1>}
  NamedRerouteUsage()
  Reroute(None=@n74)
  ScalarParameter() {group=SunRose Wings, parameterName=Boost}
 ]
}
