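# Physics asset for the SunRose wings gadget mesh in the flying gameplay plugin, giving the wings simplified collision bodies.
# Four box bodies cover the left and right shoulder and wrist bones, with the right side roughly mirroring the left.
# One constraint per side joins wrist to shoulder with 45-degree swing and twist limit values.
# The two sides differ: only the left constraint explicitly sets all swing and twist motion to Free.
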PhysicsAsset /SunRoseFlyingGameplay/Items/Wings/Meshes/Wings_SunRose_Gadget_Physics
bodies {
 shoulder_l [{
   shapeType=Box
   extentX=26.01
   extentY=91
   extentZ=41.01
   shapeName=shoulder_l_box
   center=<50.7294 -2.1233 -4.1875>
   rotation=<rot -0.0005 80.0001 15>
  }]
 shoulder_r [{
   shapeType=Box
   extentX=26.01
   extentY=91
   extentZ=41.01
   shapeName=shoulder_l_box
   center=<-51.1464 2.1969 5.7678>
   rotation=<rot -0.0005 80.0001 -165>
  }]
 wrist_l [{
   shapeType=Box
   extentX=23.01
   extentY=71.01
   extentZ=44
   shapeName=wrist_l_box
   center=<29.6186 -1.9795 -24.2579>
   rotation=<rot 0.001 -89.9999 5>
  }]
 wrist_r [{
   shapeType=Box
   extentX=23.01
   extentY=71.01
   extentZ=44
   shapeName=wrist_l_box
   center=<-30.9704 1.9795 24.7886>
   rotation=<rot 0.001 -89.9999 -175>
  }]
}
constraints [
 {bone1Name=wrist_r, bone2Name=shoulder_r, swing1LimitDegrees=45, swing2LimitDegrees=45, twistLimitDegrees=45}
 {
  bone1Name=wrist_l
  bone2Name=shoulder_l
  swing1LimitDegrees=45
  swing1Motion=Free
  swing2LimitDegrees=45
  swing2Motion=Free
  twistLimitDegrees=45
  twistMotion=Free
 }
]
